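# Control Rig for the F_MED_JoltMosaic female character body, adding procedural correctives to skirt, limb and helper bones.
# A shared deform-rig function runs first, then crouch smoothing, center-skirt rotation averaging, skirt setup, adjustments and forearm/spine surface constraints.
# Spherical pose readers on both thighs weight layered skirt-bone offsets, blended with crouch, jump, DBNO, idle and rail-grind states.
# Further adjustments cover gliding, crouched knees, crouched pickaxe swings and the right upper arm.
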ControlRigBlueprint /BRCosmetics/Characters/Player/Female/Medium/Bodies/F_MED_JoltMosaic/Meshes/F_MED_JoltMosaic_CtrlRig
graphs [
 Multiply_11_ContainedGraph {nodes [
   n1 = RigVMFunctionEntryNode()
   RigVMFunctionReturnNode(Result=&RigVMFunction_MathDoubleMul(B=@n1.C, A=&RigVMFunction_MathDoubleMul(A=@n1.A, B=@n1.B).Result).Result)
  ]}
 Multiply_11_ContainedGraph {nodes [
   n1 = RigVMFunctionEntryNode()
   RigVMFunctionReturnNode(Result=&RigVMFunction_MathDoubleMul(B=@n1.C, A=&RigVMFunction_MathDoubleMul(A=@n1.A, B=@n1.B).Result).Result)
  ]}
 New Function_ContainedGraph {nodes [
   n1 = RigVMFunctionEntryNode()
   RigVMFunction_Sequence = &RigVMFunction_Sequence(ExecuteContext=@n1.ExecuteContext)
   RigVMFunction_MathBoolToFloat = &RigVMFunction_MathBoolToFloat(Value=RigVMVariableNode() {variable=CrouchingBool}.Value)
   AlphaInterp = &RigVMFunction_AlphaInterp(InterpSpeedDecreasing=10, InterpSpeedIncreasing=10, Scale=1, bClampResult=true, Value=@RigVMFunction_MathBoolToFloat.Result)
   RigVMVariableNode(Value=@AlphaInterp.Result, ExecuteContext=@RigVMFunction_Sequence.A) {variable=Crouching}
   RigVMFunctionReturnNode(ExecuteContext=@RigVMFunction_Sequence.B)
  ]}
 New Function_ContainedGraph {nodes [
   n1 = RigVMFunctionEntryNode()
   RigVMFunction_Sequence = &RigVMFunction_Sequence(ExecuteContext=@n1.ExecuteContext)
   &RigUnit_ParentConstraint(
    AdvancedSettings=(InterpolationType=Average,RotationOrderForFilter=XZY)
    Child=(Type=Bone,Name="dyn_skirt_fr_02_c")
    Filter=(TranslationFilter=(bX=False,bY=False,bZ=False),RotationFilter=(bX=True,bY=True,bZ=True),ScaleFilter=(bX=False,bY=False,bZ=False))
    Parents=((Item=(Type=Bone,Name="dyn_skirt_fr_02_l"),Weight=1),(Item=(Type=Bone,Name="dyn_skirt_fr_02_r"),Weight=1))
    Weight=1
    ExecutePin=@RigVMFunction_Sequence.A
   )
   > &RigUnit_ParentConstraint(
    AdvancedSettings=(InterpolationType=Average,RotationOrderForFilter=XZY)
    Child=(Type=Bone,Name="dyn_skirt_fr_03_c")
    Filter=(TranslationFilter=(bX=False,bY=False,bZ=False),RotationFilter=(bX=True,bY=True,bZ=True),ScaleFilter=(bX=False,bY=False,bZ=False))
    Parents=((Item=(Type=Bone,Name="dyn_skirt_fr_03_l"),Weight=1),(Item=(Type=Bone,Name="dyn_skirt_fr_03_r"),Weight=1))
    Weight=1
   )
   > &RigUnit_ParentConstraint(
    AdvancedSettings=(InterpolationType=Average,RotationOrderForFilter=XZY)
    Child=(Type=Bone,Name="dyn_skirt_fr_04_c")
    Filter=(TranslationFilter=(bX=False,bY=False,bZ=False),RotationFilter=(bX=True,bY=True,bZ=True),ScaleFilter=(bX=False,bY=False,bZ=False))
    Parents=((Item=(Type=Bone,Name="dyn_skirt_fr_04_l"),Weight=1),(Item=(Type=Bone,Name="dyn_skirt_fr_04_r"),Weight=1))
    Weight=1
   )
   > &RigUnit_ParentConstraint(
    AdvancedSettings=(InterpolationType=Average,RotationOrderForFilter=XZY)
    Child=(Type=Bone,Name="dyn_skirt_bk_02_c")
    Filter=(TranslationFilter=(bX=False,bY=False,bZ=False),RotationFilter=(bX=True,bY=True,bZ=True),ScaleFilter=(bX=False,bY=False,bZ=False))
    Parents=((Item=(Type=Bone,Name="dyn_skirt_bk_02_l"),Weight=1),(Item=(Type=Bone,Name="dyn_skirt_bk_02_r"),Weight=1))
    Weight=1
   )
   > &RigUnit_ParentConstraint(
    AdvancedSettings=(InterpolationType=Average,RotationOrderForFilter=XZY)
    Child=(Type=Bone,Name="dyn_skirt_bk_03_c")
    Filter=(TranslationFilter=(bX=False,bY=False,bZ=False),RotationFilter=(bX=True,bY=True,bZ=True),ScaleFilter=(bX=False,bY=False,bZ=False))
    Parents=((Item=(Type=Bone,Name="dyn_skirt_bk_03_l"),Weight=1),(Item=(Type=Bone,Name="dyn_skirt_bk_03_r"),Weight=1))
    Weight=1
   )
   > &RigUnit_ParentConstraint(
    AdvancedSettings=(InterpolationType=Average,RotationOrderForFilter=XZY)
    Child=(Type=Bone,Name="dyn_skirt_bk_04_c")
    Filter=(TranslationFilter=(bX=False,bY=False,bZ=False),RotationFilter=(bX=True,bY=True,bZ=True),ScaleFilter=(bX=False,bY=False,bZ=False))
    Parents=((Item=(Type=Bone,Name="dyn_skirt_bk_04_l"),Weight=1),(Item=(Type=Bone,Name="dyn_skirt_bk_04_r"),Weight=1))
    Weight=1
   )
   RigVMFunctionReturnNode(ExecuteContext=@RigVMFunction_Sequence.B)
  ]}
 New Function_ContainedGraph {nodes [
   RigVMFunctionEntryNode()
   > RigVMFunction_Sequence = RigVMAggregateNode()
   n1 = &RigUnit_SphericalPoseReader(
    ActiveRegionScaleFactors=(PositiveWidth=1,NegativeWidth=1,PositiveHeight=1,NegativeHeight=1)
    ActiveRegionSize=0.05
    Debug=(bDrawDebug=false,bDraw2D=False,bDrawLocalAxes=False,DebugScale=25,DebugSegments=20,DebugThickness=0.25)
    DriverAxis=(X=1,Y=0,Z=0)
    DriverItem=(Type=Bone,Name="thigh_r")
    FalloffRegionScaleFactors=(PositiveWidth=1,NegativeWidth=1.1,PositiveHeight=1,NegativeHeight=1)
    FalloffSize=0.55
    FlipHeightScaling=true
    FlipWidthScaling=true
    OptionalParentItem=(Type=Bone,Name="None")
    RotationOffset=(X=-90,Y=0,Z=-20)
    ExecutePin=@RigVMFunction_Sequence.A
   )
   > n2 = &RigUnit_SphericalPoseReader(
    ActiveRegionScaleFactors=(PositiveWidth=1,NegativeWidth=1,PositiveHeight=1,NegativeHeight=1)
    ActiveRegionSize=0.05
    Debug=(bDrawDebug=false,bDraw2D=False,bDrawLocalAxes=False,DebugScale=25,DebugSegments=20,DebugThickness=0.25)
    DriverAxis=(X=1,Y=0,Z=0)
    DriverItem=(Type=Bone,Name="thigh_r")
    FalloffRegionScaleFactors=(PositiveWidth=1,NegativeWidth=1,PositiveHeight=1,NegativeHeight=1)
    FalloffSize=0.25
    FlipHeightScaling=false
    FlipWidthScaling=false
    OptionalParentItem=(Type=Bone,Name="None")
    RotationOffset=(X=-40,Y=30,Z=0)
   )
   > n3 = &RigUnit_SphericalPoseReader(
    ActiveRegionScaleFactors=(PositiveWidth=1,NegativeWidth=1,PositiveHeight=1,NegativeHeight=1)
    ActiveRegionSize=0.05
    Debug=(bDrawDebug=false,bDraw2D=False,bDrawLocalAxes=False,DebugScale=25,DebugSegments=20,DebugThickness=0.25)
    DriverAxis=(X=1,Y=0,Z=0)
    DriverItem=(Type=Bone,Name="thigh_r")
    FalloffRegionScaleFactors=(PositiveWidth=1,NegativeWidth=1.2,PositiveHeight=1,NegativeHeight=1)
    FalloffSize=0.55
    FlipHeightScaling=true
    FlipWidthScaling=true
    OptionalParentItem=(Type=Bone,Name="None")
    RotationOffset=(X=0,Y=20,Z=0)
   )
   > n4 = &RigUnit_SphericalPoseReader(
    ActiveRegionScaleFactors=(PositiveWidth=1,NegativeWidth=1,PositiveHeight=1,NegativeHeight=1)
    ActiveRegionSize=0.05
    Debug=(bDrawDebug=false,bDraw2D=False,bDrawLocalAxes=False,DebugScale=25,DebugSegments=20,DebugThickness=0.25)
    DriverAxis=(X=1,Y=0,Z=0)
    DriverItem=(Type=Bone,Name="thigh_r")
    FalloffRegionScaleFactors=(PositiveWidth=1,NegativeWidth=1,PositiveHeight=1,NegativeHeight=1.25)
    FalloffSize=0.2
    FlipHeightScaling=true
    FlipWidthScaling=true
    OptionalParentItem=(Type=Bone,Name="None")
    RotationOffset=(X=40,Y=20,Z=0)
   )
   > n5 = &RigUnit_SphericalPoseReader(
    ActiveRegionScaleFactors=(PositiveWidth=1,NegativeWidth=1,PositiveHeight=1,NegativeHeight=1)
    ActiveRegionSize=0.05
    Debug=(bDrawDebug=false,bDraw2D=false,bDrawLocalAxes=False,DebugScale=25,DebugSegments=20,DebugThickness=0.25)
    DriverAxis=(X=1,Y=0,Z=0)
    DriverItem=(Type=Bone,Name="thigh_r")
    FalloffRegionScaleFactors=(PositiveWidth=1,NegativeWidth=1,PositiveHeight=1,NegativeHeight=1)
    FalloffSize=0.55
    FlipHeightScaling=true
    FlipWidthScaling=true
    OptionalParentItem=(Type=Bone,Name="None")
    RotationOffset=(X=90,Y=0,Z=20)
   )
   > n6 = &RigUnit_SphericalPoseReader(
    ActiveRegionScaleFactors=(PositiveWidth=1,NegativeWidth=1,PositiveHeight=1,NegativeHeight=1)
    ActiveRegionSize=0.05
    Debug=(bDrawDebug=false,bDraw2D=false,bDrawLocalAxes=False,DebugScale=25,DebugSegments=20,DebugThickness=0.25)
    DriverAxis=(X=1,Y=0,Z=0)
    DriverItem=(Type=Bone,Name="thigh_r")
    FalloffRegionScaleFactors=(PositiveWidth=2,NegativeWidth=1.5,PositiveHeight=1,NegativeHeight=1)
    FalloffSize=0.22
    FlipHeightScaling=true
    FlipWidthScaling=true
    OptionalParentItem=(Type=Bone,Name="None")
    RotationOffset=(X=-130,Y=0,Z=0)
   )
   > n7 = &RigUnit_SphericalPoseReader(
    ActiveRegionScaleFactors=(PositiveWidth=1,NegativeWidth=1,PositiveHeight=1,NegativeHeight=1)
    ActiveRegionSize=0.05
    Debug=(bDrawDebug=false,bDraw2D=False,bDrawLocalAxes=False,DebugScale=25,DebugSegments=20,DebugThickness=0.25)
    DriverAxis=(X=-1,Y=0,Z=0)
    DriverItem=(Type=Bone,Name="thigh_l")
    FalloffRegionScaleFactors=(PositiveWidth=1,NegativeWidth=1.1,PositiveHeight=1,NegativeHeight=1)
    FalloffSize=0.55
    FlipHeightScaling=False
    FlipWidthScaling=False
    OptionalParentItem=(Type=Bone,Name="None")
    RotationOffset=(X=90,Y=0,Z=-20)
   )
   > n8 = &RigUnit_SphericalPoseReader(
    ActiveRegionScaleFactors=(PositiveWidth=1,NegativeWidth=1,PositiveHeight=1,NegativeHeight=1)
    ActiveRegionSize=0.05
    Debug=(bDrawDebug=false,bDraw2D=False,bDrawLocalAxes=False,DebugScale=25,DebugSegments=20,DebugThickness=0.25)
    DriverAxis=(X=-1,Y=0,Z=0)
    DriverItem=(Type=Bone,Name="thigh_l")
    FalloffRegionScaleFactors=(PositiveWidth=1,NegativeWidth=1,PositiveHeight=1,NegativeHeight=1)
    FalloffSize=0.25
    FlipHeightScaling=False
    FlipWidthScaling=False
    OptionalParentItem=(Type=Bone,Name="None")
    RotationOffset=(X=140,Y=30,Z=0)
   )
   > n9 = &RigUnit_SphericalPoseReader(
    ActiveRegionScaleFactors=(PositiveWidth=1,NegativeWidth=1,PositiveHeight=1,NegativeHeight=1)
    ActiveRegionSize=0.05
    Debug=(bDrawDebug=false,bDraw2D=False,bDrawLocalAxes=False,DebugScale=25,DebugSegments=20,DebugThickness=0.25)
    DriverAxis=(X=-1,Y=0,Z=0)
    DriverItem=(Type=Bone,Name="thigh_l")
    FalloffRegionScaleFactors=(PositiveWidth=1,NegativeWidth=1.12,PositiveHeight=1,NegativeHeight=1)
    FalloffSize=0.55
    FlipHeightScaling=False
    FlipWidthScaling=False
    OptionalParentItem=(Type=Bone,Name="None")
    RotationOffset=(X=180,Y=20,Z=0)
   )
   > n10 = &RigUnit_SphericalPoseReader(
    ActiveRegionScaleFactors=(PositiveWidth=1,NegativeWidth=1,PositiveHeight=1,NegativeHeight=1)
    ActiveRegionSize=0.05
    Debug=(bDrawDebug=false,bDraw2D=False,bDrawLocalAxes=False,DebugScale=25,DebugSegments=20,DebugThickness=0.25)
    DriverAxis=(X=-1,Y=0,Z=0)
    DriverItem=(Type=Bone,Name="thigh_l")
    FalloffRegionScaleFactors=(PositiveWidth=1,NegativeWidth=1,PositiveHeight=1,NegativeHeight=1.25)
    FalloffSize=0.2
    FlipHeightScaling=true
    FlipWidthScaling=false
    OptionalParentItem=(Type=Bone,Name="None")
    RotationOffset=(X=-140,Y=20,Z=0)
   )
   > n11 = &RigUnit_SphericalPoseReader(
    ActiveRegionScaleFactors=(PositiveWidth=1,NegativeWidth=1,PositiveHeight=1,NegativeHeight=1)
    ActiveRegionSize=0.05
    Debug=(bDrawDebug=false,bDraw2D=False,bDrawLocalAxes=False,DebugScale=25,DebugSegments=20,DebugThickness=0.25)
    DriverAxis=(X=-1,Y=0,Z=0)
    DriverItem=(Type=Bone,Name="thigh_l")
    FalloffRegionScaleFactors=(PositiveWidth=1,NegativeWidth=1,PositiveHeight=1,NegativeHeight=1)
    FalloffSize=0.55
    FlipHeightScaling=False
    FlipWidthScaling=False
    OptionalParentItem=(Type=Bone,Name="None")
    RotationOffset=(X=-90,Y=0,Z=20)
   )
   > n12 = &RigUnit_SphericalPoseReader(
    ActiveRegionScaleFactors=(PositiveWidth=1,NegativeWidth=1,PositiveHeight=1,NegativeHeight=1)
    ActiveRegionSize=0.05
    Debug=(bDrawDebug=false,bDraw2D=False,bDrawLocalAxes=False,DebugScale=25,DebugSegments=20,DebugThickness=0.25)
    DriverAxis=(X=-1,Y=0,Z=0)
    DriverItem=(Type=Bone,Name="thigh_l")
    FalloffRegionScaleFactors=(PositiveWidth=2,NegativeWidth=2,PositiveHeight=1,NegativeHeight=2)
    FalloffSize=0.2
    FlipHeightScaling=true
    FlipWidthScaling=true
    OptionalParentItem=(Type=Bone,Name="None")
    RotationOffset=(X=50,Y=0,Z=0)
   )
   Multiply_3 = RigVMAggregateNode(
    C=&RigVMFunction_MathDoubleRemap(SourceMaximum=1, SourceMinimum=0, TargetMaximum=0, TargetMinimum=1, Value=RigVMVariableNode() {variable=Crouching}.Value).Result
    A=RigVMVariableNode() {variable=Leg_Front_R}.Value
    B=RigVMVariableNode() {variable=Leg_Front_L}.Value
   )
   RigVMFunction_AnimEvalRichCurve = &RigVMFunction_AnimEvalRichCurve(
    Curve=(EditorCurveData=(Keys=((),(InterpMode=RCIM_Cubic,TangentMode=RCTM_Break,Time=0.2218,Value=0.2795,ArriveTangent=1.0223,LeaveTangent=0.3085),(InterpMode=RCIM_Cubic,TangentMode=RCTM_User,Time=0.9885,Value=0.1705,ArriveTangent=-1.575,LeaveTangent=-1.575))))
    SourceMinimum=0
    TargetMinimum=0
    Value=@Multiply_3.Result
   )
   > &RigUnit_ModifyTransforms(
    ItemToModify="((Item=(Type=Bone,Name=\"base_skirt_fr_r\"),Transform=(Rotation=(X=0,Y=0,Z=0,W=1),Translation=(X=0,Y=0,Z=0),Scale3D=(X=1,Y=1,Z=1))),(Item=(Type=Bone,Name=\"base_skirt_sd_r\"),Transform=(Rotation=(X=0,Y=0,Z=0,W=1),Translation=(X=0,Y=0,Z=0),Scale3D=(X=1,Y=1,Z=1) ...(truncated)"
    WeightMinimum=0
    Weight=@RigVMFunction_AnimEvalRichCurve.Result
   )
   Multiply_11 = RigVMAggregateNode(
    A=RigVMVariableNode() {variable=Leg_Front_R}.Value
    B=RigVMVariableNode() {variable=Leg_Front_L}.Value
    C=RigVMVariableNode() {variable=Crouching}.Value
   )
   n13 = &RigVMFunction_AnimEvalRichCurve(
    Curve=(EditorCurveData=(Keys=((),(InterpMode=RCIM_Cubic,Time=0.2661,Value=0.4716,ArriveTangent=1,LeaveTangent=1),(Time=1,Value=1))))
    SourceMinimum=0
    TargetMinimum=0
    Value=@Multiply_11.Result
   )
   > &RigUnit_ModifyTransforms(
    ItemToModify="((Item=(Type=Bone,Name=\"base_skirt_fr_r\"),Transform=(Rotation=(X=0,Y=0.2508,Z=-0.0001,W=0.968),Translation=(X=0,Y=0,Z=0),Scale3D=(X=1,Y=1,Z=1))),(Item=(Type=Bone,Name=\"base_skirt_fr_l\"),Transform=(Rotation=(X=0.0236,Y=0.1171,Z=-0.0127,W=0.9928),Translation ...(truncated)"
    WeightMinimum=0
    Weight=@n13.Result
   )
   > &RigUnit_ModifyTransforms(
    ItemToModify="((Item=(Type=Bone,Name=\"base_skirt_bk_r\"),Transform=(Rotation=(X=0.0003,Y=0.0202,Z=-0.0397,W=0.999),Translation=(X=0,Y=0,Z=0),Scale3D=(X=1,Y=1,Z=1))),(Item=(Type=Bone,Name=\"base_skirt_bk_l\"),Transform=(Rotation=(X=0.0015,Y=0.0401,Z=-0.0366,W=0.9985),Transl ...(truncated)"
    WeightMinimum=0
    Weight=&RigVMFunction_MathDoubleMul(B=RigVMVariableNode() {variable=Leg_Back_R}.Value, A=RigVMVariableNode() {variable=Leg_Back_L}.Value).Result
   )
   n14 = &RigVMFunction_AlphaInterp(
    InterpSpeedDecreasing=4
    InterpSpeedIncreasing=50
    Scale=1
    bClampResult=true
    Value=&RigVMFunction_MathBoolToFloat(Value=RigVMVariableNode() {variable=Jumping}.Value).Result
   )
   n15 = RigVMVariableNode() {variable=Leg_FrontInside_R}
   > &RigUnit_ModifyTransforms(
    ItemToModify="((Item=(Type=Bone,Name=\"base_skirt_fr_r\"),Transform=(Rotation=(X=0.1408,Y=0.1426,Z=-0.2055,W=0.9579),Translation=(X=0,Y=0,Z=0),Scale3D=(X=1,Y=1,Z=1))),(Item=(Type=Bone,Name=\"base_skirt_fr_c\"),Transform=(Rotation=(X=-0.0763,Y=0.0162,Z=0.282,W=-0.9562),Trans ...(truncated)"
    WeightMinimum=0
    Weight=&RigVMFunction_MathDoubleRemap(
     SourceMaximum=0.5
     SourceMinimum=0
     TargetMaximum=1
     TargetMinimum=0
     Value=&RigVMFunction_MathDoubleMul(B=@n14.Result, A=@n15.Value).Result
    ).Result
   )
   RigVMVariableNode(ExecuteContext=@RigVMFunction_Sequence.B, Value=@n1.OutputParam) {variable=Leg_Front_R}
   > RigVMVariableNode(Value=@n2.OutputParam) {variable=Leg_FrontSide_R}
   > RigVMVariableNode(Value=@n3.OutputParam) {variable=Leg_Side_R}
   > RigVMVariableNode(Value=@n4.OutputParam) {variable=Leg_BackSide_R}
   > RigVMVariableNode(Value=@n5.OutputParam) {variable=Leg_Back_R}
   > RigVMVariableNode(Value=@n6.OutputParam) {variable=Leg_FrontInside_R}
   > RigVMVariableNode(Value=@n7.OutputParam) {variable=Leg_Front_L}
   > RigVMVariableNode(Value=@n8.OutputParam) {variable=Leg_FrontSide_L}
   > RigVMVariableNode(Value=@n9.OutputParam) {variable=Leg_Side_L}
   > RigVMVariableNode(Value=@n10.OutputParam) {variable=Leg_BackSide_L}
   > RigVMVariableNode(Value=@n11.OutputParam) {variable=Leg_Back_L}
   > RigVMVariableNode(Value=@n12.OutputParam) {variable=Leg_FrontInside_L}
   &RigUnit_ModifyTransforms(
    ItemToModify="((Item=(Type=Bone,Name=\"base_skirt_fr_r\"),Transform=(Rotation=(X=-0.1877,Y=0.704,Z=0.3485,W=-0.5897),Translation=(X=-1.4588,Y=0.3088,Z=-0.0019),Scale3D=(X=0.8659,Y=1,Z=1))),(Item=(Type=Bone,Name=\"base_skirt_sd_r\"),Transform=(Rotation=(X=-0.0228,Y=0.1108,Z= ...(truncated)"
    WeightMinimum=0
    ExecutePin=@RigVMFunction_Sequence.C
    Weight=RigVMVariableNode() {variable=Leg_Front_R}.Value
   )
   > &RigUnit_ModifyTransforms(
    ItemToModify="((Item=(Type=Bone,Name=\"base_skirt_fr_r\"),Transform=(Rotation=(X=0,Y=0.2037,Z=0,W=-0.979),Translation=(X=0,Y=0,Z=0),Scale3D=(X=1,Y=1,Z=1))),(Item=(Type=Bone,Name=\"base_skirt_sd_r\"),Transform=(Rotation=(X=-0.2402,Y=0.03,Z=0.0887,W=-0.9662),Translation=(X=0, ...(truncated)"
    WeightMinimum=0
    Weight=RigVMVariableNode() {variable=Leg_FrontSide_R}.Value
   )
   > &RigUnit_ModifyTransforms(
    ItemToModify="((Item=(Type=Bone,Name=\"base_skirt_fr_r\"),Transform=(Rotation=(X=0.0683,Y=-0.2679,Z=0.1121,W=0.9545),Translation=(X=0,Y=0,Z=0),Scale3D=(X=0.9107,Y=1,Z=1))),(Item=(Type=Bone,Name=\"base_skirt_sd_r\"),Transform=(Rotation=(X=-0.1067,Y=0.658,Z=0.0688,W=-0.7422), ...(truncated)"
    WeightMinimum=0
    Weight=RigVMVariableNode() {variable=Leg_Side_R}.Value
   )
   > &RigUnit_ModifyTransforms(
    ItemToModify="((Item=(Type=Bone,Name=\"base_skirt_fr_r\"),Transform=(Rotation=(X=0,Y=0,Z=0,W=1),Translation=(X=0,Y=0,Z=0),Scale3D=(X=1,Y=1,Z=1))),(Item=(Type=Bone,Name=\"base_skirt_sd_r\"),Transform=(Rotation=(X=0.0289,Y=-0.2165,Z=0.0764,W=0.9729),Translation=(X=0,Y=0,Z=0), ...(truncated)"
    WeightMinimum=0
    Weight=RigVMVariableNode() {variable=Leg_BackSide_R}.Value
   )
   > &RigUnit_ModifyTransforms(
    ItemToModify="((Item=(Type=Bone,Name=\"base_skirt_sd_r\"),Transform=(Rotation=(X=0,Y=0,Z=0.3928,W=0.9196),Translation=(X=0,Y=0,Z=0),Scale3D=(X=1,Y=1,Z=1))),(Item=(Type=Bone,Name=\"base_skirt_bk_r\"),Transform=(Rotation=(X=-0.1964,Y=-0.6896,Z=0.1685,W=0.6764),Translation=(X= ...(truncated)"
    WeightMinimum=0
    Weight=RigVMVariableNode() {variable=Leg_Back_R}.Value
   )
   n16 = RigVMVariableNode() {variable=Crouching}
   > &RigUnit_ModifyTransforms(
    ItemToModify="((Item=(Type=Bone,Name=\"base_skirt_fr_r\"),Transform=(Rotation=(X=0,Y=0.0158,Z=0,W=0.9999),Translation=(X=0,Y=0,Z=0),Scale3D=(X=1,Y=1,Z=1))),(Item=(Type=Bone,Name=\"base_skirt_fr_l\"),Transform=(Rotation=(X=-0.0504,Y=-0.5515,Z=0.0746,W=0.8293),Translation=(X= ...(truncated)"
    WeightMinimum=0
    Weight=&RigVMFunction_MathDoubleMul(
     B=&RigVMFunction_MathDoubleRemap(SourceMaximum=1, SourceMinimum=0, TargetMaximum=0, TargetMinimum=1, Value=@n16.Value).Result
     A=RigVMVariableNode() {variable=Leg_FrontInside_R}.Value
    ).Result
   )
   > &RigUnit_ModifyTransforms(
    ItemToModify="((Item=(Type=Bone,Name=\"base_skirt_fr_l\"),Transform=(Rotation=(X=-0.3071,Y=-0.7644,Z=0.2068,W=0.5278),Translation=(X=0,Y=0,Z=0),Scale3D=(X=1,Y=1,Z=1))),(Item=(Type=Bone,Name=\"base_skirt_sd_l\"),Transform=(Rotation=(X=0.0498,Y=0.1139,Z=-0.3917,W=-0.9117),Tra ...(truncated)"
    WeightMinimum=0
    Weight=RigVMVariableNode() {variable=Leg_Front_L}.Value
   )
   > &RigUnit_ModifyTransforms(
    ItemToModify="((Item=(Type=Bone,Name=\"base_skirt_fr_l\"),Transform=(Rotation=(X=0.0264,Y=-0.1336,Z=-0.0865,W=0.9869),Translation=(X=0,Y=0,Z=0),Scale3D=(X=1,Y=1,Z=1))),(Item=(Type=Bone,Name=\"base_skirt_sd_l\"),Transform=(Rotation=(X=-0.2485,Y=-0.1505,Z=0.1284,W=0.9482),Tra ...(truncated)"
    WeightMinimum=0
    Weight=RigVMVariableNode() {variable=Leg_FrontSide_L}.Value
   )
   > &RigUnit_ModifyTransforms(
    ItemToModify="((Item=(Type=Bone,Name=\"base_skirt_fr_l\"),Transform=(Rotation=(X=0.0932,Y=0.2491,Z=0.3443,W=-0.9004),Translation=(X=0,Y=0,Z=0),Scale3D=(X=1,Y=1,Z=1))),(Item=(Type=Bone,Name=\"base_skirt_sd_l\"),Transform=(Rotation=(X=0.0466,Y=0.6785,Z=-0.0502,W=-0.7314),Tran ...(truncated)"
    WeightMinimum=0
    Weight=RigVMVariableNode() {variable=Leg_Side_L}.Value
   )
   > &RigUnit_ModifyTransforms(
    ItemToModify="((Item=(Type=Bone,Name=\"base_skirt_fr_l\"),Transform=(Rotation=(X=0,Y=0,Z=0,W=1),Translation=(X=0,Y=0,Z=0),Scale3D=(X=1,Y=1,Z=1))),(Item=(Type=Bone,Name=\"base_skirt_sd_l\"),Transform=(Rotation=(X=-0.0477,Y=-0.0993,Z=-0.2173,W=0.9699),Translation=(X=0,Y=0,Z=0 ...(truncated)"
    WeightMinimum=0
    Weight=RigVMVariableNode() {variable=Leg_BackSide_L}.Value
   )
   > &RigUnit_ModifyTransforms(
    ItemToModify="((Item=(Type=Bone,Name=\"base_skirt_sd_l\"),Transform=(Rotation=(X=0,Y=0,Z=0,W=1),Translation=(X=0,Y=0,Z=0),Scale3D=(X=1,Y=1,Z=1))),(Item=(Type=Bone,Name=\"base_skirt_bk_c\"),Transform=(Rotation=(X=-0.0162,Y=-0.6905,Z=0.0168,W=0.723),Translation=(X=0,Y=0,Z=0), ...(truncated)"
    WeightMinimum=0
    Weight=RigVMVariableNode() {variable=Leg_Back_L}.Value
   )
   n17 = RigVMVariableNode() {variable=Leg_FrontInside_L}
   > &RigUnit_ModifyTransforms(
    ItemToModify="((Item=(Type=Bone,Name=\"base_skirt_fr_l\"),Transform=(Rotation=(X=0,Y=0,Z=0,W=1),Translation=(X=0,Y=0,Z=0),Scale3D=(X=1,Y=1,Z=1))),(Item=(Type=Bone,Name=\"base_skirt_fr_r\"),Transform=(Rotation=(X=-0.0955,Y=0.2325,Z=0.087,W=-0.964),Translation=(X=0,Y=0,Z=0),S ...(truncated)"
    WeightMinimum=0
    Weight=@n17.Value
   )
   n18 = &RigVMFunction_AlphaInterp(
    InterpSpeedDecreasing=10
    InterpSpeedIncreasing=10
    Scale=1
    bClampResult=true
    Value=&RigVMFunction_MathBoolToFloat(Value=RigVMVariableNode() {variable=DBNOBool}.Value).Result
   )
   > &RigUnit_ModifyTransforms(
    ItemToModify="((Item=(Type=Bone,Name=\"base_skirt_bk_c\"),Transform=(Rotation=(X=0,Y=0.1736,Z=0,W=0.9848),Translation=(X=0.1829,Y=0,Z=0.7229),Scale3D=(X=1,Y=1,Z=1))),(Item=(Type=Bone,Name=\"base_skirt_bk_l\"),Transform=(Rotation=(X=0,Y=0.1736,Z=0,W=0.9848),Translation=(X=0. ...(truncated)"
    WeightMinimum=0
    Weight=@n18.Result
   )
   RigVMFunctionReturnNode(ExecuteContext=@RigVMFunction_Sequence.E)
   Multiply = &RigVMFunction_MathDoubleMul(A=RigVMVariableNode() {variable=Leg_Front_L}.Value, B=RigVMVariableNode() {variable=Crouching}.Value)
   &RigUnit_ModifyTransforms(
    ItemToModify="((Item=(Type=Bone,Name=\"base_skirt_fr_l\"),Transform=(Rotation=(X=-0.0464,Y=-0.0272,Z=0.0223,W=-0.9983),Translation=(X=-0.873,Y=-0.9131,Z=0.3991),Scale3D=(X=1,Y=1,Z=1))),(Item=(Type=Bone,Name=\"base_skirt_sd_l\"),Transform=(Rotation=(X=0.2856,Y=0.094,Z=0.0568 ...(truncated)"
    WeightMinimum=0
    Weight=@Multiply.Result
    ExecutePin=@RigVMFunction_Sequence.D
   )
   > &RigUnit_ModifyTransforms(
    ItemToModify="((Item=(Type=Bone,Name=\"base_skirt_fr_r\"),Transform=(Rotation=(X=0.0765,Y=-0.0874,Z=-0.0307,W=0.9928),Translation=(X=0,Y=0,Z=0),Scale3D=(X=1,Y=1,Z=1))),(Item=(Type=Bone,Name=\"base_skirt_fr_c\"),Transform=(Rotation=(X=-0.4066,Y=0.2965,Z=0.0611,W=-0.862),Tran ...(truncated)"
    WeightMinimum=0
    Weight=&RigVMFunction_MathDoubleMul(B=RigVMVariableNode() {variable=Crouching}.Value, A=RigVMVariableNode() {variable=Leg_Front_R}.Value).Result
   )
   n19 = &RigVMFunction_AlphaInterp(
    InterpSpeedDecreasing=10
    InterpSpeedIncreasing=10
    Scale=1
    bClampResult=true
    Value=&RigVMFunction_MathDoubleMul(A=RigVMVariableNode() {variable=Idle}.Value, B=RigVMVariableNode() {variable=Crouching}.Value).Result
   )
   > &RigUnit_ModifyTransforms(
    ItemToModify=((Item=(Type=Bone,Name="base_skirt_bk_l"),Transform=(Rotation=(X=0,Y=0.1732,Z=0,W=0.9849),Translation=(X=0,Y=0,Z=0),Scale3D=(X=1,Y=1,Z=1))))
    WeightMinimum=0
    Weight=@n19.Result
   )
   > &RigUnit_ModifyTransforms(
    ItemToModify="((Item=(Type=Bone,Name=\"base_skirt_fr_c\"),Transform=(Rotation=(X=-0.1663,Y=0.1106,Z=0.4098,W=-0.89),Translation=(X=0,Y=0,Z=0),Scale3D=(X=1,Y=1,Z=1))),(Item=(Type=Bone,Name=\"base_skirt_fr_l\"),Transform=(Rotation=(X=-0.2518,Y=0.1788,Z=0.1221,W=-0.9432),Trans ...(truncated)"
    WeightMinimum=0
    Weight=&RigVMFunction_MathDoubleMul(B=RigVMVariableNode() {variable=Leg_FrontInside_R}.Value, A=RigVMVariableNode() {variable=Crouching}.Value).Result
   )
   > &RigUnit_ModifyTransforms(
    ItemToModify="((Item=(Type=Bone,Name=\"base_skirt_fr_l\"),Transform=(Rotation=(X=0.0009,Y=0.0063,Z=0.0414,W=0.9991),Translation=(X=0,Y=0,Z=0),Scale3D=(X=1,Y=1,Z=1))),(Item=(Type=Bone,Name=\"base_skirt_fr_r\"),Transform=(Rotation=(X=-0.514,Y=-0.1869,Z=-0.1147,W=0.8292),Trans ...(truncated)"
    WeightMinimum=0
    Weight=&RigVMFunction_MathDoubleMul(A=RigVMVariableNode() {variable=Leg_FrontInside_L}.Value, B=RigVMVariableNode() {variable=Crouching}.Value).Result
   )
   RigVMFunction_MathBoolToFloat = &RigVMFunction_MathBoolToFloat(Value=RigVMVariableNode() {variable=GrindRail}.Value)
   AlphaInterp = &RigVMFunction_AlphaInterp(InterpSpeedDecreasing=10, InterpSpeedIncreasing=10, Scale=1, bClampResult=true, Value=@RigVMFunction_MathBoolToFloat.Result)
   > &RigUnit_ParentConstraint(
    AdvancedSettings=(InterpolationType=Average,RotationOrderForFilter=XZY)
    Child=(Type=Bone,Name="base_skirt_fr_l")
    Filter=(TranslationFilter=(bX=False,bY=False,bZ=False),RotationFilter=(bX=True,bY=True,bZ=True),ScaleFilter=(bX=True,bY=True,bZ=True))
    Parents=((Item=(Type=Bone,Name="thigh_l"),Weight=1))
    Weight=@AlphaInterp.Result
   )
   > &RigUnit_ParentConstraint(
    AdvancedSettings=(InterpolationType=Average,RotationOrderForFilter=XZY)
    Child=(Type=Bone,Name="base_skirt_sd_l")
    Filter=(TranslationFilter=(bX=False,bY=False,bZ=False),RotationFilter=(bX=True,bY=True,bZ=True),ScaleFilter=(bX=True,bY=True,bZ=True))
    Parents=((Item=(Type=Bone,Name="thigh_l"),Weight=1))
    Weight=@AlphaInterp.Result
   )
   > &RigUnit_ModifyTransforms(
    ItemToModify="((Item=(Type=Bone,Name=\"base_skirt_fr_l\"),Transform=(Rotation=(X=-0.0196,Y=0.5004,Z=0.0708,W=-0.8627),Translation=(X=-6.6444,Y=-0.3841,Z=2.7277),Scale3D=(X=1,Y=1,Z=1))),(Item=(Type=Bone,Name=\"base_skirt_sd_l\"),Transform=(Rotation=(X=0,Y=0.1439,Z=0,W=-0.989 ...(truncated)"
    WeightMinimum=0
    Weight=&RigVMFunction_MathDoubleMul(A=1, B=@AlphaInterp.Result).Result
   )
   &RigVMFunction_MathDoubleMul(A=@n17.Value, B=RigVMVariableNode() {variable=Crouching}.Value)
  ]}
 New Function_ContainedGraph {nodes [
   n1 = RigVMFunctionEntryNode()
   RigVMFunction_Sequence = &RigVMFunction_Sequence(ExecuteContext=@n1.ExecuteContext)
   AlphaInterp = &RigVMFunction_AlphaInterp(InterpSpeedDecreasing=10, InterpSpeedIncreasing=10, Scale=1, bClampResult=true, Value=RigVMVariableNode() {variable=HangGlider}.Value)
   &RigUnit_ModifyTransforms(
    ItemToModify="((Item=(Type=Bone,Name=\"deform_upperarm_up_l\"),Transform=(Rotation=(X=0,Y=-0.2164,Z=0,W=0.9763),Translation=(X=0,Y=0,Z=-2.5),Scale3D=(X=1,Y=1,Z=1))),(Item=(Type=Bone,Name=\"deform_upperarm_up_r\"),Transform=(Rotation=(X=0,Y=-0.2164,Z=0,W=0.9763),Translation= ...(truncated)"
    WeightMinimum=0
    Weight=@AlphaInterp.Result
    ExecutePin=@RigVMFunction_Sequence.A
   )
   > &RigUnit_ModifyTransforms(
    ItemToModify="((Item=(Type=Bone,Name=\"knee_l_outer_dyn\"),Transform=(Rotation=(X=0,Y=0,Z=0,W=1),Translation=(X=-3,Y=0,Z=1),Scale3D=(X=1,Y=1,Z=1))),(Item=(Type=Bone,Name=\"knee_l_inner_dyn\"),Transform=(Rotation=(X=0,Y=0,Z=0,W=-1),Translation=(X=-3,Y=-1,Z=0),Scale3D=(X=1,Y= ...(truncated)"
    WeightMinimum=0
    Weight=RigVMVariableNode() {variable=Crouching}.Value
   )
   > &RigUnit_ModifyTransforms(
    ItemToModify="((Item=(Type=Bone,Name=\"knee_r_outer_dyn\"),Transform=(Rotation=(X=0,Y=0,Z=0,W=-1),Translation=(X=2,Y=0,Z=-4),Scale3D=(X=1,Y=1,Z=1))),(Item=(Type=Bone,Name=\"knee_r_inner_dyn\"),Transform=(Rotation=(X=0,Y=0,Z=0,W=-1),Translation=(X=3,Y=4,Z=0),Scale3D=(X=1,Y=1 ...(truncated)"
    WeightMinimum=0
    Weight=RigVMVariableNode() {variable=Crouching}.Value
   )
   > &RigUnit_ParentConstraint(
    AdvancedSettings=(RotationOrderForFilter=XZY)
    Child=(Type=Bone,Name="deform_upperarm_Helper_base")
    Parents="((Item=(Type=Bone,Name=\"spine_03\"),Weight=0.0003),(Item=(Type=Bone,Name=\"upperarm_r\"),Weight=0.0006),(Item=(Type=Bone,Name=\"upperarm_twist_02_r\"),Weight=0.0011),(Item=(Type=Bone,Name=\"spine_04\"),Weight=0.0028),(Item=(Type=Bone,Name=\"clavicle_r_helper\"),Wei ...(truncated)"
    Weight=1
   )
   > n2 = &RigUnit_SphericalPoseReader(
    ActiveRegionScaleFactors=(PositiveWidth=1,NegativeWidth=1,PositiveHeight=1,NegativeHeight=0.8)
    ActiveRegionSize=0.1
    Debug=(bDrawDebug=false,bDraw2D=False,bDrawLocalAxes=False,DebugScale=25,DebugSegments=20,DebugThickness=0.25)
    DriverAxis=(X=1,Y=0,Z=0)
    DriverItem=(Type=Bone,Name="upperarm_r")
    FalloffRegionScaleFactors=(PositiveWidth=1,NegativeWidth=1,PositiveHeight=1.7,NegativeHeight=0.5)
    FalloffSize=0.25
    FlipHeightScaling=False
    FlipWidthScaling=False
    OptionalParentItem=(Type=Bone,Name="None")
    RotationOffset=(X=51.9395,Y=8.7023,Z=22.4256)
   )
   > &RigUnit_ModifyTransforms(
    ItemToModify="((Item=(Type=Bone,Name=\"upperarm_twist_01_r\"),Transform=(Rotation=(X=-0.0436,Y=0,Z=0,W=0.999),Translation=(X=0,Y=0,Z=0),Scale3D=(X=1,Y=1,Z=1))),(Item=(Type=Bone,Name=\"clavicle_r_helper\"),Transform=(Rotation=(X=0.0008,Y=-0.0436,Z=0.0174,W=0.9989),Translatio ...(truncated)"
    WeightMinimum=0
    Weight=&RigVMFunction_MathDoubleMul(
     B=&RigVMFunction_MathDoubleRemap(SourceMaximum=1, SourceMinimum=0, TargetMaximum=0, TargetMinimum=1, Value=0, bClamp=False).Result
     A=@n2.OutputParam
    ).Result
   )
   RigVMFunction_MathBoolToFloat = &RigVMFunction_MathBoolToFloat(Value=RigVMVariableNode() {variable=CrouchingBool}.Value)
   Multiply = &RigVMFunction_MathDoubleMul(
    A=@RigVMFunction_MathBoolToFloat.Result
    B=&RigVMFunction_MathBoolToFloat(Value=RigVMVariableNode() {variable=swingPickaxe}.Value).Result
   )
   n3 = &RigVMFunction_AlphaInterp(InterpSpeedDecreasing=1, InterpSpeedIncreasing=5, Scale=1, bClampResult=False, Value=@Multiply.Result)
   > &RigUnit_ModifyTransforms(
    ItemToModify="((Item=(Type=Bone,Name=\"base_skirt_fr_l\"),Transform=(Rotation=(X=0.0023,Y=-0.0647,Z=-0.0158,W=0.9978),Translation=(X=-0.8659,Y=1.3574,Z=5.4733),Scale3D=(X=1,Y=1,Z=1))),(Item=(Type=Bone,Name=\"base_skirt_fr_c\"),Transform=(Rotation=(X=-0.0056,Y=-0.2099,Z=-0.1 ...(truncated)"
    WeightMinimum=0
    Weight=@n3.Result
   )
   > &RigUnit_ModifyTransforms(
    ItemToModify="((Item=(Type=Bone,Name=\"dyn_skirt_fr_01_l\"),Transform=(Rotation=(X=0,Y=0.0239,Z=0,W=-0.9997),Translation=(X=0,Y=0,Z=1.5654),Scale3D=(X=1,Y=1,Z=1))),(Item=(Type=Bone,Name=\"dyn_skirt_sd_02_r\"),Transform=(Rotation=(X=0,Y=-0.0125,Z=0,W=0.9999),Translation=(X=0 ...(truncated)"
    WeightMinimum=0
    Weight=RigVMVariableNode() {variable=Crouching}.Value
   )
   RigVMFunctionReturnNode(ExecuteContext=@RigVMFunction_Sequence.B)
  ]}
 New Function_ContainedGraph {nodes [
   n1 = RigVMFunctionEntryNode()
   RigVMFunction_Sequence = &RigVMFunction_Sequence(ExecuteContext=@n1.ExecuteContext)
   &RigUnit_ParentConstraint(
    AdvancedSettings=(RotationOrderForFilter=XZY)
    Child=(Type=Bone,Name="Bottom_l_01_jnt")
    Parents=((Item=(Type=Bone,Name="lowerarm_l"),Weight=0.5),(Item=(Type=Bone,Name="lowerarm_twist_01_l"),Weight=0.5))
    Weight=1
    ExecutePin=@RigVMFunction_Sequence.A
   )
   > &RigUnit_ParentConstraint(
    AdvancedSettings=(RotationOrderForFilter=XZY)
    Child=(Type=Bone,Name="Bottom_l_02_jnt")
    Parents=((Item=(Type=Bone,Name="lowerarm_l"),Weight=0.5),(Item=(Type=Bone,Name="lowerarm_twist_01_l"),Weight=0.5))
    Weight=1
   )
   > &RigUnit_ParentConstraint(
    AdvancedSettings=(RotationOrderForFilter=XZY)
    Child=(Type=Bone,Name="Bottom_r_01_jnt")
    Parents=((Item=(Type=Bone,Name="lowerarm_r"),Weight=0.5),(Item=(Type=Bone,Name="lowerarm_twist_01_r"),Weight=0.5))
    Weight=1
   )
   > &RigUnit_ParentConstraint(
    AdvancedSettings=(RotationOrderForFilter=XZY)
    Child=(Type=Bone,Name="Bottom_r_02_jnt")
    Parents=((Item=(Type=Bone,Name="hand_r"),Weight=0.0139),(Item=(Type=Bone,Name="lowerarm_twist_02_r"),Weight=0.1073),(Item=(Type=Bone,Name="lowerarm_r"),Weight=0.1253),(Item=(Type=Bone,Name="lowerarm_twist_01_r"),Weight=0.7534))
    Weight=1
   )
   > &RigUnit_ParentConstraint(
    AdvancedSettings=(RotationOrderForFilter=XZY)
    Child=(Type=Bone,Name="Spine_Bottom_01")
    Parents=((Item=(Type=Bone,Name="spine_03"),Weight=0.0293),(Item=(Type=Bone,Name="spine_04"),Weight=0.0985),(Item=(Type=Bone,Name="spine_01"),Weight=0.1885),(Item=(Type=Bone,Name="spine_02"),Weight=0.3203),(Item=(Type=Bone,Name="pelvis"),Weight=0.3634))
    Weight=1
   )
   > &RigUnit_ParentConstraint(
    AdvancedSettings=(RotationOrderForFilter=XZY)
    Child=(Type=Bone,Name="Spine_Bottom_02")
    Parents=((Item=(Type=Bone,Name="pelvis"),Weight=0.0011),(Item=(Type=Bone,Name="spine_01"),Weight=0.0148),(Item=(Type=Bone,Name="spine_02"),Weight=0.0424),(Item=(Type=Bone,Name="spine_03"),Weight=0.2258),(Item=(Type=Bone,Name="spine_04"),Weight=0.7159))
    Weight=1
   )
   RigVMFunctionReturnNode(ExecuteContext=@RigVMFunction_Sequence.B)
  ]}
 RigVMFunctionLibrary {nodes [
   AdditionalAdjustments = RigVMCollapseNode(ExecuteContext=())
   ConstraintSetup = RigVMCollapseNode(ExecuteContext=())
   Set variables = RigVMCollapseNode(ExecuteContext=())
   SkirtSetup = RigVMCollapseNode(ExecuteContext=())
   SurfaceConstraints = RigVMCollapseNode(ExecuteContext=())
  ]}
 RigVMFunction_Sequence_1_ContainedGraph {nodes [
   n1 = RigVMFunctionEntryNode()
   RigVMFunction_Sequence = &RigVMFunction_Sequence(ExecuteContext=@n1.ExecuteContext)
   n2 = &RigVMFunction_Sequence(ExecuteContext=@RigVMFunction_Sequence.B)
   n3 = &RigVMFunction_Sequence(ExecuteContext=@n2.B)
   n4 = &RigVMFunction_Sequence(ExecuteContext=@n3.B)
   RigVMFunctionReturnNode(A=@RigVMFunction_Sequence.A, B=@n2.A, C=@n3.A, D=@n4.A, E=@n4.B)
  ]}
 RigVMModel {nodes [
   n1 = &RigUnit_PrepareForExecution()
   &RigUnit_HierarchyImportFromSkeleton(NameSpace=None, bIncludeCurves=False, bIncludeMeshSockets=False, bIncludeVirtualBones=true, ExecutePin=@n1.ExecutePin)
   > addFKCtrl = RigVMFunctionReferenceNode(
    Custom_FKCtrl=()
    Options_FKCtrl="(pelvis_1_#1=True,spine_3_#2=True,clavicle_5_#3=True,arm_7_#4=True,leg_9_#5=True,head_11_46DC15EF465CD2979 ...(truncated)"
    Shape_Settings="(pelvis_2_#6=(bVisible=True,Name=\"Hexagon_Thin\",Color=(R=0.0117,G=0.8281,B=1,A=1),Transform=(Rotation=(X=0,Y=-0.7071,Z=0,W=0.7071),Translation=(X=2,Y=0,Z=0),Scale3D=(X=5,Y=5,Z=7))),spine_4_#7=(bVi ...(truncated)"
   )
   > addNullObject = RigVMFunctionReferenceNode(
    Custom_Joints=()
    Options_Null="(pelvis_1_#1=true,spine_3_#2=False,clavicle_5_#3=False,arm_7_#4=False,leg_9_#5=true,head_11_46DC15EF465CD2 ...(truncated)"
   )
   RigUnit_BeginExecution = &RigUnit_BeginExecution()
   setFKCtrl = RigVMFunctionReferenceNode(setFK_Ctrl=false, ExecuteContext=@RigUnit_BeginExecution.ExecutePin)
   > FN_DeformRig_FNC = RigVMFunctionReferenceNode(
    DeformRig_ Curve="(deformRig_l_upperarm_up_alpha_72_#8=False,deformRig_r_upperarm_up_alpha_73_#9=False,deformRig_l_upperarm_fr_alpha_74_#10=False,deformRig_r_upperarm_fr_alpha_75_CB51B8 ...(truncated)"
    DeformRig_Bool="(upperarm_1_#11=True,elbow_3_#12=True,wrist_7_#13=True,groin_8_#14=True,glute_17_#15=True,knee_9_40C408324CA261 ...(truncated)"
    DeformRig_Debug=(upperarm_debug_18_#11=False,wrist_debug_20_#13=False,lat_debug_26_#16=False,groinGlute_debug_27_#14=False)
    DeformRig_Value="(upperarm_up_18_#17=(Rotation=(X=0.0,Y=-0.0,Z=0.0,W=1.0),Translation=(X=0.0,Y=0.0,Z=8.0),Scale3D=(X=0.70,Y=1.0,Z=1.0)),upperarm_fwd_fr_19_#18=(Rotation=(X=0.0,Y=-0.0,Z=0.0,W=1.0),Translation=(X=0.0, ...(truncated)"
   )
   > Set variables = RigVMFunctionReferenceNode()
   > ConstraintSetup = RigVMFunctionReferenceNode()
   > SkirtSetup = RigVMFunctionReferenceNode()
   > AdditionalAdjustments = RigVMFunctionReferenceNode()
   > SurfaceConstraints = RigVMFunctionReferenceNode()
  ]}
]
hierarchy [
 root {loc=<0 0 0>, rot=<0 0 0 1>}
 pelvis {parent=root, loc=<0 -0.5301 98.0571>, rot=<0.0127 0.707 -0.0127 -0.707>}
 spine_01 {parent=pelvis, loc=<6.1208 -0.9328 0>, rot=<0 0 0.034 -0.9994>}
 spine_02 {parent=spine_01, loc=<6.5997 -0.2095 0>, rot=<0 0 0.0238 -0.9997>}
 spine_03 {parent=spine_02, loc=<6.4914 -0.1237 0>, rot=<0 0 0.0149 -0.9999>}
 spine_04 {parent=spine_03, loc=<6.9598 0.0794 0>, rot=<0 0 -0.0132 -0.9999>}
 spine_05 {parent=spine_04, loc=<6.9987 0 0>, rot=<0 0 -0.0015 -1>}
 neck_01 {parent=spine_05, loc=<11.3138 0.1529 0>, rot=<0 0 -0.2003 -0.9797>}
 neck_02 {parent=neck_01, loc=<4.5627 0 0>, rot=<0 0 0.0504 -0.9987>}
 head {parent=neck_02, loc=<5.3905 0 0.0001>, rot=<0 0 0.115 -0.9934>}
 base_hair_1_l {parent=head, loc=<1.2178 -5.8046 -5.6>, rot=<0.1334 -0.6944 0.6944 -0.1334>}
 dyn_hair_1_l {parent=base_hair_1_l, loc=<0 0 0>, rot=<0 -0.0463 0 -0.9989>}
 dyn_hair_2_l {parent=dyn_hair_1_l, loc=<7.3205 0 0>, rot=<0 0.0027 0 -1>}
 dyn_hair_3_l {parent=dyn_hair_2_l, loc=<7.3197 0 0>, rot=<0 -0.0833 0 -0.9965>}
 dyn_hair_4_l {parent=dyn_hair_3_l, loc=<7.2884 0 0>, rot=<0 -0.068 0 -0.9977>}
 dyn_hair_5_l {parent=dyn_hair_4_l, loc=<7.2939 0 0>, rot=<0 -0.0079 0 -1>}
 dyn_hair_6_l {parent=dyn_hair_5_l, loc=<7.3213 0 0>, rot=<0 0.026 0 -0.9997>}
 dyn_hair_7_l {parent=dyn_hair_6_l, loc=<7.3211 0 0>, rot=<0 0 0 1>}
 base_hair_1_r {parent=head, loc=<1.2178 -5.8046 5.6>, rot=<0.1334 -0.6944 0.6944 -0.1334>}
 dyn_hair_1_r {parent=base_hair_1_r, loc=<0 0 0>, rot=<0 -0.0463 0 -0.9989>}
 dyn_hair_2_r {parent=dyn_hair_1_r, loc=<7.3206 0 -0.0001>, rot=<0 0.0027 0 -1>}
 dyn_hair_3_r {parent=dyn_hair_2_r, loc=<7.3192 0 0.0002>, rot=<0 -0.0833 0 -0.9965>}
 dyn_hair_4_r {parent=dyn_hair_3_r, loc=<7.2887 0 -0.0001>, rot=<0 -0.068 0 -0.9977>}
 dyn_hair_5_r {parent=dyn_hair_4_r, loc=<7.2941 0 0.0001>, rot=<0 -0.0079 0 -1>}
 dyn_hair_6_r {parent=dyn_hair_5_r, loc=<7.321 0 -0.0001>, rot=<0 0.026 0 -0.9997>}
 dyn_hair_7_r {parent=dyn_hair_6_r, loc=<7.321 0 0>, rot=<0 0 0 1>}
 base_hair_1_c {parent=head, loc=<1.2187 -5.9142 0>, rot=<0.1334 -0.6944 0.6944 -0.1334>}
 dyn_hair_1_c {parent=base_hair_1_c, loc=<0 0 0>, rot=<0 -0.054 0 -0.9985>}
 dyn_hair_2_c {parent=dyn_hair_1_c, loc=<7.3205 0 0>, rot=<0 0.0118 0 -0.9999>}
 dyn_hair_3_c {parent=dyn_hair_2_c, loc=<7.3197 0 0>, rot=<0 -0.0861 0 -0.9963>}
 dyn_hair_4_c {parent=dyn_hair_3_c, loc=<7.2884 0 0>, rot=<0 -0.0666 0 -0.9978>}
 dyn_hair_5_c {parent=dyn_hair_4_c, loc=<7.2939 0 0>, rot=<0 -0.0071 0 -1>}
 dyn_hair_6_c {parent=dyn_hair_5_c, loc=<7.3213 0 0>, rot=<0 0.026 0 -0.9997>}
 dyn_hair_7_c {parent=dyn_hair_6_c, loc=<7.3211 0 0>, rot=<0 0 0 1>}
 hair_side_r_01_base {parent=head, loc=<2.0222 2.3085 7.3401>, rot=<-0.0131 -0.6232 0.7805 0.0464>}
 hair_side_r_01_dyn {parent=hair_side_r_01_base, loc=<0 0 0>, rot=<0 0 0 1>}
 hair_side_r_02_dyn {parent=hair_side_r_01_dyn, loc=<8.0277 0.0897 0.2753>, rot=<-0.0803 -0.0498 0.0003 -0.9955>}
 hair_side_r_03_dyn {parent=hair_side_r_02_dyn, loc=<7.1177 -0.369 -0.3548>, rot=<-0.0931 -0.0728 0.0901 -0.9889>}
 hair_side_r_04_dyn {parent=hair_side_r_03_dyn, loc=<6.6938 0.1257 -0.4153>, rot=<-0.1856 -0.0885 -0.0324 -0.9781>}
 hair_side_l_01_base {parent=head, loc=<2.0222 2.3085 -7.3401>, rot=<-0.0464 -0.7805 0.6232 0.0131>}
 hair_side_l_01_dyn {parent=hair_side_l_01_base, loc=<0 0 0>, rot=<0 0 0 1>}
 hair_side_l_02_dyn {parent=hair_side_l_01_dyn, loc=<8.0277 -0.0897 0.2753>, rot=<0.0803 -0.0498 -0.0003 -0.9955>}
 hair_side_l_03_dyn {parent=hair_side_l_02_dyn, loc=<7.1177 0.369 -0.3548>, rot=<0.0931 -0.0728 -0.0901 -0.9889>}
 hair_side_l_04_dyn {parent=hair_side_l_03_dyn, loc=<6.6938 -0.1257 -0.4153>, rot=<0.1856 -0.0885 0.0324 -0.9781>}
 front_hair_dyn {parent=head, loc=<8.5333 9.574 -0.4558>, rot=<0 0 0 1>}
 Back_side_l_base {parent=head, loc=<-5.747 -1.0487 -8.6195>, rot=<-0.1074 -0.7659 -0.0026 -0.6339>}
 Back_side_l_dyn {parent=Back_side_l_base, loc=<0 0 0>, rot=<0 0 0 1>}
 Back_side_r_base {parent=head, loc=<-5.747 -1.0487 8.9678>, rot=<-0.1074 -0.7659 -0.0026 -0.6339>}
 Back_side_r_dyn {parent=Back_side_r_base, loc=<0 0 0>, rot=<0 0 0 1>}
 Back_rear_base {parent=head, loc=<-1.3089 -9.2165 -0.2788>, rot=<0.5347 -0.4627 0.4627 -0.5347>}
 Back_rear_dyn {parent=Back_rear_base, loc=<0 0 0>, rot=<0 0 0 1>}
 clavicle_l {parent=spine_05, loc=<5.4831 -0.1053 -3.2946>, rot=<0.0465 -0.7289 -0.009 -0.6829>}
 upperarm_l {parent=clavicle_l, loc=<9.8658 -1.2136 -0.4105>, rot=<0.0164 -0.3245 -0.0269 -0.9454>}
 lowerarm_l {parent=upperarm_l, loc=<26.465 -0.8345 0.259>, rot=<0.0036 0.0161 -0.2298 -0.9731>}
 hand_l {parent=lowerarm_l, loc=<22.7355 -0.3203 -0.6105>, rot=<-0.6784 -0.0223 0.0735 -0.7307>}
 thumb_01_l {parent=hand_l, loc=<1.3118 -0.3191 -2.459>, rot=<0.6546 -0.4042 0.0664 -0.6354>}
 thumb_02_l {parent=thumb_01_l, loc=<5.9336 -0.0448 0.0401>, rot=<-0.0037 0.0213 0.298 -0.9543>}
 thumb_03_l {parent=thumb_02_l, loc=<2.2223 0.3682 -0.0907>, rot=<0.0118 -0.013 0.0066 -0.9998>}
 index_metacarpal_l {parent=hand_l, loc=<1.6535 0.0183 -2.96>, rot=<0.0158 -0.0124 -0.0173 -0.9996>}
 index_01_l {parent=index_metacarpal_l, loc=<8.3008 0.0357 -0.0524>, rot=<-0.0044 0.0485 0.0924 -0.9945>}
 index_02_l {parent=index_01_l, loc=<3.3587 -0.3004 0.0452>, rot=<-0.002 0.0027 0.3225 -0.9466>}
 index_03_l {parent=index_02_l, loc=<2.5578 -0.123 -0.0252>, rot=<-0.002 0.0067 0.2558 -0.9667>}
 middle_metacarpal_l {parent=hand_l, loc=<2.1856 -0.1725 -0.5984>, rot=<0.006 0.0168 -0.0329 -0.9993>}
 middle_01_l {parent=middle_metacarpal_l, loc=<7.3275 -0.2447 0.0129>, rot=<-0.0208 0.0042 0.0858 -0.9961>}
 middle_02_l {parent=middle_01_l, loc=<4.0123 0.1401 0.1175>, rot=<-0.0218 0.0258 0.3512 -0.9357>}
 middle_03_l {parent=middle_02_l, loc=<2.6077 0.0534 0.1373>, rot=<0.0082 0.0188 0.1708 -0.9851>}
 ring_metacarpal_l {parent=hand_l, loc=<2.2074 -0.3625 1.3638>, rot=<-0.0284 0.0722 -0.0458 -0.9959>}
 ring_01_l {parent=ring_metacarpal_l, loc=<7.0295 -0.6197 -0.1446>, rot=<-0.0489 -0.0252 0.1892 -0.9804>}
 ring_02_l {parent=ring_01_l, loc=<3.6174 0.4647 -0.1027>, rot=<0 0 0.2806 -0.9598>}
 ring_03_l {parent=ring_02_l, loc=<2.4152 0.114 0.0199>, rot=<0 0 0.1189 -0.9929>}
 pinky_metacarpal_l {parent=hand_l, loc=<1.7616 -0.9753 3.3814>, rot=<-0.0397 0.0814 -0.069 -0.9935>}
 pinky_01_l {parent=pinky_metacarpal_l, loc=<7.1709 -0.4831 -0.1094>, rot=<-0.0654 -0.0035 0.2242 -0.9723>}
 pinky_02_l {parent=pinky_01_l, loc=<2.8274 0.0893 -0.1043>, rot=<0.0037 0.0011 0.2854 -0.9584>}
 pinky_03_l {parent=pinky_02_l, loc=<2.2046 0.0307 -0.0078>, rot=<-0.0002 0.0098 0.1445 -0.9895>}
 deform_wrist_dn_base_l {parent=hand_l, loc=<0 -1.5326 0>, rot=<0 0 0 -1>}
 deform_wrist_dn_l {parent=deform_wrist_dn_base_l, loc=<0 0 0>, rot=<0 0 0 1>}
 deform_wrist_up_base_l {parent=hand_l, loc=<0 2.3009 0>, rot=<0 0 0 -1>}
 deform_wrist_up_l {parent=deform_wrist_up_base_l, loc=<0 0 0>, rot=<0 0 0 1>}
 hand_attach_l {parent=hand_l, loc=<8.5 -3 0>, rot=<0.7071 0.7071 0 0>}
 weapon_l {parent=hand_l, loc=<-0.1974 -3.4111 0.3811>, rot=<0 0 0 1>}
 lowerarm_twist_02_l {parent=lowerarm_l, loc=<10.6707 0 0>, rot=<0 0 0 1>}
 lowerarm_twist_01_l {parent=lowerarm_l, loc=<18.316 0 0>, rot=<0 0 0 1>}
 deform_elbow_in_base_l {parent=lowerarm_l, loc=<0.1834 3.6062 -0.2572>, rot=<0 0 0 -1>}
 deform_elbow_in_l {parent=deform_elbow_in_base_l, loc=<0 0 0>, rot=<0 0 0 1>}
 deform_elbow_out_base_l {parent=lowerarm_l, loc=<0 -4 0>, rot=<0 0 0 1>}
 deform_elbow_out_l {parent=deform_elbow_out_base_l, loc=<0 0 0>, rot=<0 0 0 1>}
 Bottom_l_01_jnt {parent=lowerarm_l, loc=<18.4648 -4.5821 0.8268>, rot=<0.1434 0.1073 0.6948 -0.6966>}
 Bottom_l_02_jnt {parent=lowerarm_l, loc=<16.1435 -4.6507 0.8574>, rot=<0.1443 0.1141 0.6831 -0.7068>}
 upperarm_twist_01_l {parent=upperarm_l, loc=<7.2135 -0.1134 0>, rot=<0 0 0 1>}
 deform_upperarm_fr_base_l {parent=upperarm_twist_01_l, loc=<-5.6312 5.092 0.3354>, rot=<0 0 0 -1>}
 deform_upperarm_fr_l {parent=deform_upperarm_fr_base_l, loc=<0 0 0>, rot=<0 0 0 1>}
 deform_upperarm_up_base_l {parent=upperarm_twist_01_l, loc=<-8.891 0.4825 3.9517>, rot=<0 0 0 -1>}
 deform_upperarm_up_l {parent=deform_upperarm_up_base_l, loc=<0 0 0>, rot=<0 0 0 1>}
 deform_upperarm_bk_base_l {parent=upperarm_twist_01_l, loc=<-6.4053 -4.3746 0.314>, rot=<0 0 0 -1>}
 deform_upperarm_bk_l {parent=deform_upperarm_bk_base_l, loc=<0 0 0>, rot=<0 0 0 1>}
 upperarm_twist_02_l {parent=upperarm_l, loc=<19.7998 -0.1134 0>, rot=<0 0 0 1>}
 clavicle_r {parent=spine_05, loc=<5.4831 -0.1053 3.2946>, rot=<0.7289 0.0465 -0.6829 0.009>}
 upperarm_r {parent=clavicle_r, loc=<-9.8658 1.2136 0.4105>, rot=<0.0164 -0.3245 -0.0269 -0.9454>}
 lowerarm_r {parent=upperarm_r, loc=<-26.465 0.8345 -0.259>, rot=<0.0036 0.0161 -0.2298 -0.9731>}
 hand_r {parent=lowerarm_r, loc=<-22.7355 0.3203 0.6106>, rot=<-0.6784 -0.0223 0.0735 -0.7307>}
 thumb_01_r {parent=hand_r, loc=<-1.3118 0.3191 2.459>, rot=<0.6546 -0.4042 0.0664 -0.6354>}
 thumb_02_r {parent=thumb_01_r, loc=<-5.9336 0.0448 -0.0401>, rot=<-0.0037 0.0213 0.298 -0.9543>}
 thumb_03_r {parent=thumb_02_r, loc=<-2.2223 -0.3682 0.0907>, rot=<0.0118 -0.013 0.0066 -0.9998>}
 index_metacarpal_r {parent=hand_r, loc=<-1.6535 -0.0183 2.96>, rot=<0.0158 -0.0124 -0.0173 -0.9996>}
 index_01_r {parent=index_metacarpal_r, loc=<-8.3008 -0.0357 0.0524>, rot=<-0.0044 0.0485 0.0924 -0.9945>}
 index_02_r {parent=index_01_r, loc=<-3.3587 0.3004 -0.0452>, rot=<-0.002 0.0027 0.3225 -0.9466>}
 index_03_r {parent=index_02_r, loc=<-2.5578 0.123 0.0252>, rot=<-0.002 0.0067 0.2558 -0.9667>}
 middle_metacarpal_r {parent=hand_r, loc=<-2.1856 0.1725 0.5984>, rot=<0.006 0.0168 -0.0329 -0.9993>}
 middle_01_r {parent=middle_metacarpal_r, loc=<-7.3275 0.2447 -0.0129>, rot=<-0.0208 0.0042 0.0858 -0.9961>}
 middle_02_r {parent=middle_01_r, loc=<-4.0123 -0.1401 -0.1175>, rot=<-0.0218 0.0258 0.3512 -0.9357>}
 middle_03_r {parent=middle_02_r, loc=<-2.6077 -0.0534 -0.1373>, rot=<0.0082 0.0188 0.1708 -0.9851>}
 ring_metacarpal_r {parent=hand_r, loc=<-2.2074 0.3625 -1.3638>, rot=<-0.0284 0.0722 -0.0458 -0.9959>}
 ring_01_r {parent=ring_metacarpal_r, loc=<-7.0295 0.6197 0.1446>, rot=<-0.0489 -0.0252 0.1892 -0.9804>}
 ring_02_r {parent=ring_01_r, loc=<-3.6174 -0.4647 0.1027>, rot=<0 0 0.2806 -0.9598>}
 ring_03_r {parent=ring_02_r, loc=<-2.4152 -0.114 -0.0199>, rot=<0 0 0.1189 -0.9929>}
 pinky_metacarpal_r {parent=hand_r, loc=<-1.7616 0.9753 -3.3814>, rot=<-0.0397 0.0814 -0.069 -0.9935>}
 pinky_01_r {parent=pinky_metacarpal_r, loc=<-7.1709 0.4831 0.1094>, rot=<-0.0654 -0.0035 0.2242 -0.9723>}
 pinky_02_r {parent=pinky_01_r, loc=<-2.8274 -0.0893 0.1043>, rot=<0.0037 0.0011 0.2854 -0.9584>}
 pinky_03_r {parent=pinky_02_r, loc=<-2.2046 -0.0307 0.0078>, rot=<-0.0002 0.0098 0.1445 -0.9895>}
 deform_wrist_dn_base_r {parent=hand_r, loc=<-0.0001 1.5328 0>, rot=<0 0 0 -1>}
 deform_wrist_dn_r {parent=deform_wrist_dn_base_r, loc=<0 0 0>, rot=<0 0 0 1>}
 deform_wrist_up_base_r {parent=hand_r, loc=<-0.0002 -2.3007 -0.0001>, rot=<0 0 0 -1>}
 deform_wrist_up_r {parent=deform_wrist_up_base_r, loc=<0 0 0>, rot=<0 0 0 1>}
 weapon_r {parent=hand_r, loc=<0.1974 3.4111 -0.3811>, rot=<0 0 0 1>}
 hand_attach_r {parent=hand_r, loc=<-8.5001 2.9999 0>, rot=<0 0 -0.7071 -0.7071>}
 lowerarm_twist_02_r {parent=lowerarm_r, loc=<-10.6707 0 0>, rot=<0 0 0 -1>}
 lowerarm_twist_01_r {parent=lowerarm_r, loc=<-18.316 0 0>, rot=<0 0 0 -1>}
 deform_elbow_in_base_r {parent=lowerarm_r, loc=<-0.1834 -3.6062 0.2572>, rot=<0 0 0 -1>}
 deform_elbow_in_r {parent=deform_elbow_in_base_r, loc=<0 0 0>, rot=<0 0 0 1>}
 deform_elbow_out_base_r {parent=lowerarm_r, loc=<0 4 -0.0001>, rot=<0 0 0 -1>}
 deform_elbow_out_r {parent=deform_elbow_out_base_r, loc=<0 0 0>, rot=<0 0 0 1>}
 Bottom_r_01_jnt {parent=lowerarm_r, loc=<-18.4649 4.5821 -0.8266>, rot=<0.1073 -0.1434 -0.6966 -0.6948>}
 Bottom_r_02_jnt {parent=lowerarm_r, loc=<-16.1434 4.6506 -0.8575>, rot=<0.1141 -0.1443 -0.7068 -0.6831>}
 upperarm_twist_01_r {parent=upperarm_r, loc=<-7.2135 0.1134 0>, rot=<0 0 0 -1>}
 deform_upperarm_fr_base_r {parent=upperarm_twist_01_r, loc=<5.6312 -5.092 -0.3354>, rot=<0 0 0 -1>}
 deform_upperarm_fr_r {parent=deform_upperarm_fr_base_r, loc=<0 0 0>, rot=<0 0 0 1>}
 deform_upperarm_up_base_r {parent=upperarm_twist_01_r, loc=<8.891 -0.4825 -3.9517>, rot=<0 0 0 -1>}
 deform_upperarm_up_r {parent=deform_upperarm_up_base_r, loc=<0 0 0>, rot=<0 0 0 1>}
 deform_upperarm_bk_base_r {parent=upperarm_twist_01_r, loc=<6.4053 4.3746 -0.314>, rot=<0 0 0 -1>}
 deform_upperarm_bk_r {parent=deform_upperarm_bk_base_r, loc=<0 0 0>, rot=<0 0 0 1>}
 upperarm_twist_02_r {parent=upperarm_r, loc=<-19.7998 0.1134 0>, rot=<0 0 0 -1>}
 clavicle_r_helper {parent=clavicle_r, loc=<0 0 0>, rot=<0 0 0 1>}
 deform_upperarm_Helper_base {parent=clavicle_r, loc=<-10.9705 3.6399 7.3769>, rot=<0.8381 0.2147 0.4913 0.1009>}
 deform_upperarm_Helper_dyn {parent=deform_upperarm_Helper_base, loc=<0 0 0>, rot=<0 0 0 1>}
 attach_backpack {parent=spine_05, loc=<-0.063 -13.4667 0.002>, rot=<-0.0092 -0.707 0.0093 -0.707>}
 Backpack_BR {parent=attach_backpack, type=Null, loc=<0 2.4191 0>, rot=<0 0 0 1>}
 attach_cape {parent=spine_05, loc=<9.017 -11.714 0>, rot=<0.0616 -0.7044 -0.0616 -0.7044>}
 cape {parent=attach_cape, type=Null, loc=<0 0 0>, rot=<0 0 0 1>}
 Bow_C_base {parent=spine_05, loc=<6.6805 6.6646 0.001>, rot=<-0.5877 -0.3932 -0.3932 -0.5877>}
 Bow_C_dyn {parent=Bow_C_base, loc=<0 0 0>, rot=<0 0 0 1>}
 Bow_l_dyn {parent=Bow_C_dyn, loc=<0.5921 -3.2569 -0.4827>, rot=<-0.5642 0.4135 0.6545 -0.2869>}
 Bow_r_dyn {parent=Bow_C_dyn, loc=<0.3163 2.7804 -0.4463>, rot=<-0.572 -0.3902 0.6178 0.3726>}
 Bow_side_l_dyn {parent=Bow_C_dyn, loc=<1.3178 -1.3924 -3.5735>, rot=<-0.6949 0.0897 0.7113 0.0557>}
 Bow_side_r_dyn {parent=Bow_C_dyn, loc=<1.0336 1.2109 -2.995>, rot=<0.7137 0.0392 -0.6987 -0.0308>}
 deform_pec_aim_base_l {parent=spine_04, loc=<3.2456 9.3309 -1.2237>, rot=<0.0603 -0.4033 0.3318 -0.8507>}
 deform_pec_aim_l {parent=deform_pec_aim_base_l, loc=<0 0 0>, rot=<0 0 0 1>}
 deform_pec_base_l {parent=deform_pec_aim_l, loc=<9.3665 0 0>, rot=<0 0 0 1>}
 deform_pec_l {parent=deform_pec_base_l, loc=<0 0 0>, rot=<0 0 0 1>}
 deform_pec_aim_base_r {parent=spine_04, loc=<3.2452 9.3309 1.2237>, rot=<-0.4033 -0.0603 0.8507 0.3318>}
 deform_pec_aim_r {parent=deform_pec_aim_base_r, loc=<0 0 0>, rot=<0 0 0 1>}
 deform_pec_base_r {parent=deform_pec_aim_r, loc=<-9.3667 -0.0003 -0.0004>, rot=<0 0 0 1>}
 deform_pec_r {parent=deform_pec_base_r, loc=<0 0 0>, rot=<0 0 0 1>}
 Bottom_C_01_jnt {parent=spine_03, loc=<0.0901 11.8863 0.4551>, rot=<0.4565 0.54 0.54 0.4565>}
 Spine_Bottom_02 {parent=spine_03, loc=<0.0513 12.0105 0.4553>, rot=<0 0 -0.7925 -0.6099>}
 Bottom_C_02_jnt {parent=spine_02, loc=<-1.8902 10.3287 0.5408>, rot=<0.464 0.5336 0.5336 0.464>}
 Spine_Bottom_01 {parent=spine_02, loc=<-1.9247 10.4525 0.5487>, rot=<0 0 -0.7347 -0.6783>}
 thigh_l {parent=pelvis, loc=<-1.0855 -0.2109 -8.7053>, rot=<0.0009 0.0355 0.0155 -0.9993>}
 calf_l {parent=thigh_l, loc=<-42.5284 -0.0019 -0.7297>, rot=<0.0006 -0.0161 -0.0391 -0.9991>}
 foot_l {parent=calf_l, loc=<-46.7401 -2.5014 -0.592>, rot=<0.0427 -0.0238 0.0353 -0.9982>}
 ball_l {parent=foot_l, loc=<-3.5658 10.8066 -0.1544>, rot=<0.0001 -0.0001 -0.7063 -0.7079>}
 Shoelaces_l_base {parent=foot_l, loc=<0.3665 6.9589 0.0473>, rot=<0.2367 0.6375 0.6994 0.2198>}
 Shoelaces_l_dyn {parent=Shoelaces_l_base, loc=<0 0 0>, rot=<0 0 0 1>}
 Showlace_left_r_dyn {parent=Shoelaces_l_dyn, loc=<1.5987 0.9325 0.3504>, rot=<-0.6927 0.0266 -0.1274 -0.7094>}
 Showlace_left_l_dyn {parent=Shoelaces_l_dyn, loc=<1.4929 -1.4078 0.1591>, rot=<0.8517 -0.2349 -0.067 0.4636>}
 Left_knot_r_dyn {parent=Shoelaces_l_dyn, loc=<0.8671 1.8678 0.2122>, rot=<-0.6699 -0.2435 -0.3061 -0.631>}
 Left_knot_l_dyn {parent=Shoelaces_l_dyn, loc=<0.5488 -1.8012 -0.1982>, rot=<0.8772 -0.4749 -0.0039 -0.0704>}
 calf_twist_02_l {parent=calf_l, loc=<-19.1546 -1.7943 -0.2638>, rot=<0.0237 0.0095 -0.0035 -0.9997>}
 calf_twist_01_l {parent=calf_l, loc=<-35.8483 -1.9197 -0.5794>, rot=<0.0237 0.0095 -0.0035 -0.9997>}
 deform_knee_in_base_l {parent=calf_l, loc=<0.8031 -5.0231 0.2456>, rot=<0 0 0 -1>}
 deform_knee_in_l {parent=deform_knee_in_base_l, loc=<0 0 0>, rot=<0 0 0 1>}
 deform_knee_out_base_l {parent=calf_l, loc=<0 4 0>, rot=<0 0 0 -1>}
 deform_knee_out_l {parent=deform_knee_out_base_l, loc=<0 0 0>, rot=<0 0 0 1>}
 knee_l_outer_base {parent=calf_l, loc=<-0.1001 -0.639 5.7317>, rot=<0.0665 0.77 -0.0016 -0.6346>}
 knee_l_outer_dyn {parent=knee_l_outer_base, loc=<0 0 0>, rot=<0 0 0 1>}
 knee_l_inner_base {parent=calf_l, loc=<0.87 -0.3696 -3.4354>, rot=<0.5363 0.4958 -0.5038 0.4612>}
 knee_l_inner_dyn {parent=knee_l_inner_base, loc=<0 0 0>, rot=<0 0 0 1>}
 thigh_twist_01_l {parent=thigh_l, loc=<-27.952 0 0>, rot=<0.0013 0 0 -1>}
 deform_groin_base_l {parent=thigh_l, loc=<0.4801 5.6776 -0.2486>, rot=<0 0 0 -1>}
 deform_groin_l {parent=deform_groin_base_l, loc=<0 0 0>, rot=<0 0 0 1>}
 thigh_r {parent=pelvis, loc=<-1.0855 -0.2109 8.7053>, rot=<0.0355 -0.0009 0.9993 0.0155>}
 calf_r {parent=thigh_r, loc=<42.5284 0.0019 0.7297>, rot=<0.0006 -0.0161 -0.0391 -0.9991>}
 foot_r {parent=calf_r, loc=<46.7401 2.5014 0.592>, rot=<0.0427 -0.0238 0.0353 -0.9982>}
 ball_r {parent=foot_r, loc=<3.5658 -10.8066 0.1544>, rot=<0.0001 -0.0001 -0.7063 -0.7079>}
 Shoelaces_r_base {parent=foot_r, loc=<-0.3665 -6.9589 -0.0472>, rot=<0.2367 0.6375 0.6994 0.2198>}
 Shoelaces_r_dyn {parent=Shoelaces_r_base, loc=<0 0 0>, rot=<0 0 0 1>}
 Showlace_right_l_dyn {parent=Shoelaces_r_dyn, loc=<-1.5987 -0.9325 -0.3504>, rot=<-0.6927 0.0266 -0.1274 -0.7094>}
 Showlace_right_r_dyn {parent=Shoelaces_r_dyn, loc=<-1.4929 1.4078 -0.1591>, rot=<0.8304 -0.2337 -0.0803 0.4993>}
 Right_knot_r_dyn {parent=Shoelaces_r_dyn, loc=<-0.8671 -1.8679 -0.2122>, rot=<-0.6699 -0.2435 -0.3061 -0.631>}
 Right_knot_r_dyn1 {parent=Shoelaces_r_dyn, loc=<-0.5488 1.8012 0.1982>, rot=<0.8772 -0.4749 -0.0039 -0.0704>}
 calf_twist_02_r {parent=calf_r, loc=<19.1546 1.7943 0.2638>, rot=<0.0237 0.0095 -0.0035 -0.9997>}
 calf_twist_01_r {parent=calf_r, loc=<35.8483 1.9197 0.5794>, rot=<0.0237 0.0095 -0.0035 -0.9997>}
 deform_knee_in_base_r {parent=calf_r, loc=<-0.8031 5.0231 -0.2457>, rot=<0 0 0 -1>}
 deform_knee_in_r {parent=deform_knee_in_base_r, loc=<0 0 0>, rot=<0 0 0 1>}
 deform_knee_out_base_r {parent=calf_r, loc=<0 -4 0>, rot=<0 0 0 -1>}
 deform_knee_out_r {parent=deform_knee_out_base_r, loc=<0 0 0>, rot=<0 0 0 1>}
 knee_r_outer_base {parent=calf_r, loc=<0.1001 0.639 -5.7317>, rot=<0.0665 0.77 -0.0016 -0.6346>}
 knee_r_outer_dyn {parent=knee_r_outer_base, loc=<0 0 0>, rot=<0 0 0 1>}
 knee_r_inner_base {parent=calf_r, loc=<-0.87 0.3696 3.4354>, rot=<0.5363 0.4958 -0.5038 0.4612>}
 knee_r_inner_dyn {parent=knee_r_inner_base, loc=<0 0 0>, rot=<0 0 0 1>}
 thigh_twist_01_r {parent=thigh_r, loc=<27.952 0 0>, rot=<0.0013 0 0 -1>}
 deform_groin_base_r {parent=thigh_r, loc=<-0.4801 -5.6776 0.2486>, rot=<0 0 0 -1>}
 deform_groin_r {parent=deform_groin_base_r, loc=<0 0 0>, rot=<0 0 0 1>}
 base_skirt_sd_r {parent=pelvis, loc=<5.6383 -0.217 12.3328>, rot=<0.145 -0.0026 0.9893 0.0177>}
 dyn_skirt_sd_01_r {parent=base_skirt_sd_r, loc=<0 0 0>, rot=<0 0 0 -1>}
 dyn_skirt_sd_02_r {parent=dyn_skirt_sd_01_r, loc=<7.9464 0 0>, rot=<0 0 0 -1>}
 dyn_skirt_sd_03_r {parent=dyn_skirt_sd_02_r, loc=<7.4038 0 0>, rot=<0 0 0 -1>}
 dyn_skirt_sd_04_r {parent=dyn_skirt_sd_03_r, loc=<7.7408 0 0>, rot=<0 0 0 1>}
 base_skirt_fr_r {parent=pelvis, loc=<5.1451 6.4518 8.1479>, rot=<0.0848 0.4133 0.9051 0.052>}
 dyn_skirt_fr_01_r {parent=base_skirt_fr_r, loc=<0 0 0>, rot=<0 0 0 1>}
 dyn_skirt_fr_02_r {parent=dyn_skirt_fr_01_r, loc=<7.9466 0 0>, rot=<0 0 0 -1>}
 dyn_skirt_fr_03_r {parent=dyn_skirt_fr_02_r, loc=<7.4037 0 0>, rot=<0 0 0 -1>}
 dyn_skirt_fr_04_r {parent=dyn_skirt_fr_03_r, loc=<7.7408 0 0>, rot=<0 0 0 1>}
 base_skirt_bk_r {parent=pelvis, loc=<6.173 -9.9002 7.89>, rot=<0.0758 -0.4762 0.8753 -0.0371>}
 dyn_skirt_bk_01_r {parent=base_skirt_bk_r, loc=<0 0 0>, rot=<0 0 0 1>}
 dyn_skirt_bk_02_r {parent=dyn_skirt_bk_01_r, loc=<7.9463 0 0>, rot=<0 0 0 1>}
 dyn_skirt_bk_03_r {parent=dyn_skirt_bk_02_r, loc=<7.6033 0 0>, rot=<0 0 0 1>}
 dyn_skirt_bk_04_r {parent=dyn_skirt_bk_03_r, loc=<7.7408 0 0>, rot=<0 0 0 1>}
 base_skirt_fr_c {parent=pelvis, loc=<4.8956 8.541 0>, rot=<0.0395 0.706 0.706 0.0395>}
 dyn_skirt_fr_01_c {parent=base_skirt_fr_c, loc=<0 0 0>, rot=<0 0 0 1>}
 dyn_skirt_fr_02_c {parent=dyn_skirt_fr_01_c, loc=<7.9463 0 0>, rot=<0 0 0 1>}
 dyn_skirt_fr_03_c {parent=dyn_skirt_fr_02_c, loc=<7.4037 0 0>, rot=<0 0 0 1>}
 dyn_skirt_fr_04_c {parent=dyn_skirt_fr_03_c, loc=<7.7408 0 0>, rot=<0 0 0 1>}
 base_skirt_fr_l {parent=pelvis, loc=<5.1449 6.4518 -8.1479>, rot=<-0.052 -0.9051 -0.4133 -0.0848>}
 dyn_skirt_fr_01_l {parent=base_skirt_fr_l, loc=<0 0 0>, rot=<0 0 0 1>}
 dyn_skirt_fr_02_l {parent=dyn_skirt_fr_01_l, loc=<7.9463 0 0>, rot=<0 0 0 1>}
 dyn_skirt_fr_03_l {parent=dyn_skirt_fr_02_l, loc=<7.4037 0 0>, rot=<0 0 0 1>}
 dyn_skirt_fr_04_l {parent=dyn_skirt_fr_03_l, loc=<7.7408 0 0>, rot=<0 0 0 1>}
 base_skirt_sd_l {parent=pelvis, loc=<5.6382 -0.217 -12.3328>, rot=<-0.0177 -0.9893 0.0026 -0.145>}
 dyn_skirt_sd_01_l {parent=base_skirt_sd_l, loc=<0 0 0>, rot=<0 0 0 1>}
 dyn_skirt_sd_02_l {parent=dyn_skirt_sd_01_l, loc=<7.9464 -0.0005 0.0028>, rot=<0.0002 -0.0023 -0.0004 -1>}
 dyn_skirt_sd_03_l {parent=dyn_skirt_sd_02_l, loc=<7.4037 0 0>, rot=<0 0 0 1>}
 dyn_skirt_sd_04_l {parent=dyn_skirt_sd_03_l, loc=<7.7408 0 0>, rot=<0 0 0 1>}
 base_skirt_bk_l {parent=pelvis, loc=<6.1731 -9.9002 -7.89>, rot=<0.0371 -0.8753 0.4762 -0.0758>}
 dyn_skirt_bk_01_l {parent=base_skirt_bk_l, loc=<0 0 0>, rot=<0 0 0 1>}
 dyn_skirt_bk_02_l {parent=dyn_skirt_bk_01_l, loc=<7.9463 0 0>, rot=<0 0 0 1>}
 dyn_skirt_bk_03_l {parent=dyn_skirt_bk_02_l, loc=<7.6034 0 0>, rot=<0 0 0 1>}
 dyn_skirt_bk_04_l {parent=dyn_skirt_bk_03_l, loc=<7.7408 0 0>, rot=<0 0 0 1>}
 base_skirt_bk_c {parent=pelvis, loc=<7.1067 -11.069 0>, rot=<0.0486 -0.7054 0.7054 -0.0486>}
 dyn_skirt_bk_01_c {parent=base_skirt_bk_c, loc=<0 0 0>, rot=<0 0 0 1>}
 dyn_skirt_bk_02_c {parent=dyn_skirt_bk_01_c, loc=<7.9463 0 0>, rot=<0 0 0 1>}
 dyn_skirt_bk_03_c {parent=dyn_skirt_bk_02_c, loc=<8.3193 0 0>, rot=<0 0 0 1>}
 dyn_skirt_bk_04_c {parent=dyn_skirt_bk_03_c, loc=<6.8252 0 0>, rot=<0 0 0 1>}
 deform_glute_base_l {parent=pelvis, loc=<-7.3466 -9.0213 -9.1245>, rot=<0.0172 0.031 -0.4477 -0.8935>}
 deform_glute_l {parent=deform_glute_base_l, loc=<0 0 0>, rot=<0 0 0 1>}
 deform_glute_base_r {parent=pelvis, loc=<-7.3464 -9.0213 9.1245>, rot=<0.031 -0.0172 0.8935 -0.4477>}
 deform_glute_r {parent=deform_glute_base_r, loc=<0 0 0>, rot=<0 0 0 1>}
 ik_hand_root {parent=root, loc=<0 0 0>, rot=<0 0 0 1>}
 ik_hand_gun {parent=ik_hand_root, loc=<-48.1743 3.3545 104.4381>, rot=<0.6276 0.1309 0.4183 -0.6434>}
 ik_hand_r {parent=ik_hand_gun, loc=<0 0 0>, rot=<0 0 0 1>}
 ik_hand_l {parent=ik_hand_gun, loc=<59.3271 67.6997 34.3524>, rot=<-0.6158 -0.7027 -0.3565 0>}
 ik_foot_root {parent=root, loc=<0 0 0>, rot=<0 0 0 1>}
 ik_foot_l {parent=ik_foot_root, loc=<14.8262 -7.3545 8.296>, rot=<0.0357 0.7045 0.0284 -0.7082>}
 ik_foot_r {parent=ik_foot_root, loc=<-14.8262 -7.3545 8.296>, rot=<0.7082 0.0284 -0.7045 0.0357>}
]
variables [
 Crouching {type=double, default=0}
 CrouchingBool {type=bool, default=False}
 DBNOBool {type=bool, default=False}
 DBNOCarried {type=bool, default=False}
 DBNOIdle {type=double, default=0}
 FallingOnGround {type=double, default=0}
 GrindRail {type=bool, default=False}
 HangGlider {type=double, default=0}
 Idle {type=double, default=0}
 Jumping {type=bool, default=False}
 Leg_BackSide_L {type=double, default=0}
 Leg_BackSide_R {type=double, default=0}
 Leg_Back_L {type=double, default=0}
 Leg_Back_R {type=double, default=0}
 Leg_FrontInside_L {type=double, default=0}
 Leg_FrontInside_R {type=double, default=0}
 Leg_FrontSide_L {type=double, default=0}
 Leg_FrontSide_R {type=double, default=0}
 Leg_Front_L {type=double, default=0}
 Leg_Front_R {type=double, default=0}
 Leg_Side_L {type=double, default=0}
 Leg_Side_R {type=double, default=0}
 swingPickaxe {type=bool, default=False}
]
