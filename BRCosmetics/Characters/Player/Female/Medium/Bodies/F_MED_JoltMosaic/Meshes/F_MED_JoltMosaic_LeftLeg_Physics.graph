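# Physics asset for the left-leg mesh of the female medium JoltMosaic cosmetic body, defining capsule bodies and joints for the left side of its skirt.
# Front and side chains of four dynamic bones each hang from tiny base capsules, whose swing and twist are free relative to the pelvis.
# Chain joints allow narrow, asymmetric swing and almost no twist.
# Larger capsules on pelvis and left thigh, joined by a limited constraint, supply the body volume.
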PhysicsAsset /BRCosmetics/Characters/Player/Female/Medium/Bodies/F_MED_JoltMosaic/Meshes/F_MED_JoltMosaic_LeftLeg_Physics
bodies {
 base_skirt_fr_l [{shapeType=Capsule, length=0.505, radius=0.505, shapeName=base_skirt_fr_l_capsule, rotation=<rot 0 -90 0>}]
 base_skirt_sd_l [{shapeType=Capsule, length=0.505, radius=0.505, shapeName=base_skirt_sd_l_capsule, rotation=<rot 0 -90 0>}]
 dyn_skirt_fr_01_l [{shapeType=Capsule, length=8, radius=3, shapeName=dyn_skirt_fr_01_l_capsule, center=<2 -0.0133 1.1177>, rotation=<rot 0.0001 0 97.2079>}]
 dyn_skirt_fr_02_l [{shapeType=Capsule, length=13, radius=4.5, shapeName=dyn_skirt_fr_02_l_capsule, center=<2 0 2>, rotation=<rot 0 0 100>}]
 dyn_skirt_fr_03_l [{shapeType=Capsule, length=13, radius=4, shapeName=dyn_skirt_fr_03_l_capsule, center=<2 0.3156 1.9858>, rotation=<rot 0.0002 0.0002 120.3768>}]
 dyn_skirt_fr_04_l [{
   shapeType=Capsule
   length=15
   radius=4.5
   shapeName=dyn_skirt_fr_04_l_capsule
   center=<3 0.0873 -0.1319>
   rotation=<rot 0.0003 0.0002 120.3606>
  }]
 dyn_skirt_sd_01_l [{shapeType=Capsule, length=8, radius=3, shapeName=dyn_skirt_sd_01_l_capsule, center=<2.3655 0 2.973>, rotation=<rot 0 0 90>}]
 dyn_skirt_sd_02_l [{shapeType=Capsule, length=15, radius=3.5, shapeName=dyn_skirt_sd_02_l_capsule, center=<3.5 0 4>, rotation=<rot 0 0 90>}]
 dyn_skirt_sd_03_l [{shapeType=Capsule, length=15, radius=3.5, shapeName=dyn_skirt_sd_03_l_capsule, center=<3.0587 0.0552 3.6311>, rotation=<rot 0 0 95>}]
 dyn_skirt_sd_04_l [{shapeType=Capsule, length=12, radius=3.5, shapeName=dyn_skirt_sd_04_l_capsule, center=<3 0 2>, rotation=<rot 0 0 100>}]
 pelvis [{shapeType=Capsule, length=11.4233, radius=7.5319, shapeName=pelvis_capsule, center=<0 -1.618 0>}]
 thigh_l [{shapeType=Capsule, length=25, radius=8, shapeName=thigh_l_capsule, center=<-20 0 0>, rotation=<rot 90 0 0>}]
}
constraints [
 {bone1Name=thigh_l, bone2Name=pelvis, swing1LimitDegrees=45, swing2LimitDegrees=45, twistLimitDegrees=45}
 {
  bone1Name=base_skirt_fr_l
  bone2Name=pelvis
  swing1LimitDegrees=45
  swing1Motion=Free
  swing2LimitDegrees=45
  swing2Motion=Free
  twistLimitDegrees=45
  twistMotion=Free
 }
 {bone1Name=dyn_skirt_fr_01_l, bone2Name=base_skirt_fr_l, swing1LimitDegrees=6, swing2LimitDegrees=15, twistLimitDegrees=2}
 {bone1Name=dyn_skirt_fr_02_l, bone2Name=dyn_skirt_fr_01_l, swing1LimitDegrees=2, swing2LimitDegrees=10, twistLimitDegrees=0}
 {bone1Name=dyn_skirt_fr_03_l, bone2Name=dyn_skirt_fr_02_l, swing1LimitDegrees=3, swing2LimitDegrees=12, twistLimitDegrees=0}
 {
  bone1Name=base_skirt_sd_l
  bone2Name=pelvis
  swing1LimitDegrees=45
  swing1Motion=Free
  swing2LimitDegrees=45
  swing2Motion=Free
  twistLimitDegrees=45
  twistMotion=Free
 }
 {bone1Name=dyn_skirt_sd_01_l, bone2Name=base_skirt_sd_l, swing1LimitDegrees=5, swing2LimitDegrees=12, twistLimitDegrees=2}
 {bone1Name=dyn_skirt_sd_02_l, bone2Name=dyn_skirt_sd_01_l, swing1LimitDegrees=5, swing2LimitDegrees=12, twistLimitDegrees=0}
 {bone1Name=dyn_skirt_sd_03_l, bone2Name=dyn_skirt_sd_02_l, swing1LimitDegrees=4, swing2LimitDegrees=12, twistLimitDegrees=0}
 {bone1Name=dyn_skirt_fr_04_l, bone2Name=dyn_skirt_fr_03_l, swing1LimitDegrees=2.5, swing2LimitDegrees=10, twistLimitDegrees=0}
 {bone1Name=dyn_skirt_sd_04_l, bone2Name=dyn_skirt_sd_03_l, swing1LimitDegrees=5, swing2LimitDegrees=12, twistLimitDegrees=0}
]
