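# Physics asset for the right-leg mesh of the female medium JoltMosaic body cosmetic, defining physics bodies for the right front and side skirt panels.
# Each panel is a four-link capsule chain from a tiny base body, which joins the pelvis with free swing and twist.
# Links have narrow swing limits and near-zero twist, keeping the cloth-like panels fairly stiff.
# Larger pelvis and right thigh capsules supply body volumes alongside the chains.
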PhysicsAsset /BRCosmetics/Characters/Player/Female/Medium/Bodies/F_MED_JoltMosaic/Meshes/F_MED_JoltMosaic_RightLeg_Physics
bodies {
 base_skirt_fr_r [{shapeType=Capsule, length=0.505, radius=0.505, shapeName=base_skirt_fr_r_capsule, rotation=<rot 0 -90 0>}]
 base_skirt_sd_r [{shapeType=Capsule, length=0.505, radius=0.505, shapeName=base_skirt_sd_r_capsule, rotation=<rot 0 -90 0>}]
 dyn_skirt_fr_01_r [{shapeType=Capsule, length=8, radius=3, shapeName=dyn_skirt_fr_01_r_capsule, center=<2 0 1.2>, rotation=<rot 0.0001 -0.0001 76.1692>}]
 dyn_skirt_fr_02_r [{shapeType=Capsule, length=13, radius=2.5, shapeName=dyn_skirt_fr_02_r_capsule, center=<1 0 3>, rotation=<rot 0 0 70>}]
 dyn_skirt_fr_03_r [
  {shapeType=Capsule, length=13, radius=5, shapeName=dyn_skirt_fr_03_r_capsule, center=<1.6341 1.2139 2.4418>, rotation=<rot 0 0 70>}
  {shapeType=Capsule, length=13, radius=4, shapeName=dyn_skirt_fr_03_r_capsule, center=<1.5 9.7549 4.8216>, rotation=<rot 0 0 110>}
 ]
 dyn_skirt_fr_04_r [{
   shapeType=Capsule
   length=17.5
   radius=4.5
   shapeName=dyn_skirt_fr_04_r_capsule
   center=<3.596 4.9006 1.7323>
   rotation=<rot 0.0463 -0.0394 76.6729>
  }]
 dyn_skirt_sd_01_r [{shapeType=Capsule, length=8, radius=3, shapeName=dyn_skirt_sd_01_r_capsule, center=<2 0 3.0849>, rotation=<rot 0 0 90>}]
 dyn_skirt_sd_02_r [{shapeType=Capsule, length=13, radius=4.5, shapeName=dyn_skirt_sd_02_r_capsule, center=<2 0 4.5>, rotation=<rot 0 0 90>}]
 dyn_skirt_sd_03_r [{shapeType=Capsule, length=12, radius=4.5, shapeName=dyn_skirt_sd_03_r_capsule, center=<2 0 4.5>, rotation=<rot 0 0 90>}]
 dyn_skirt_sd_04_r [{shapeType=Capsule, length=13.5, radius=4.5, shapeName=dyn_skirt_sd_04_r_capsule, center=<3.1432 0 4.5>, rotation=<rot 0 -0.0001 87.7207>}]
 pelvis [{shapeType=Capsule, length=11.4233, radius=7.5319, shapeName=pelvis_capsule, center=<0 -1.618 0>}]
 thigh_r [{shapeType=Capsule, length=25, radius=8.3, shapeName=thigh_l_capsule, center=<20 0.8293 0>, rotation=<rot 89.372 177.8091 179.9997>}]
}
constraints [
 {bone1Name=thigh_r, bone2Name=pelvis, swing1LimitDegrees=45, swing2LimitDegrees=45, twistLimitDegrees=45}
 {
  bone1Name=base_skirt_fr_r
  bone2Name=pelvis
  swing1LimitDegrees=45
  swing1Motion=Free
  swing2LimitDegrees=45
  swing2Motion=Free
  twistLimitDegrees=45
  twistMotion=Free
 }
 {bone1Name=dyn_skirt_fr_01_r, bone2Name=base_skirt_fr_r, swing1LimitDegrees=8, swing2LimitDegrees=20, twistLimitDegrees=2}
 {bone1Name=dyn_skirt_fr_02_r, bone2Name=dyn_skirt_fr_01_r, swing1LimitDegrees=5, swing2LimitDegrees=10, twistLimitDegrees=0}
 {bone1Name=dyn_skirt_fr_03_r, bone2Name=dyn_skirt_fr_02_r, swing1LimitDegrees=3, swing2LimitDegrees=10, twistLimitDegrees=0}
 {bone1Name=dyn_skirt_fr_04_r, bone2Name=dyn_skirt_fr_03_r, swing1LimitDegrees=4, swing2LimitDegrees=5, twistLimitDegrees=0}
 {
  bone1Name=base_skirt_sd_r
  bone2Name=pelvis
  swing1LimitDegrees=45
  swing1Motion=Free
  swing2LimitDegrees=45
  swing2Motion=Free
  twistLimitDegrees=45
  twistMotion=Free
 }
 {bone1Name=dyn_skirt_sd_01_r, bone2Name=base_skirt_sd_r, swing1LimitDegrees=8, swing2LimitDegrees=20, twistLimitDegrees=2}
 {bone1Name=dyn_skirt_sd_02_r, bone2Name=dyn_skirt_sd_01_r, swing1LimitDegrees=5, swing2LimitDegrees=15, twistLimitDegrees=0}
 {bone1Name=dyn_skirt_sd_03_r, bone2Name=dyn_skirt_sd_02_r, swing1LimitDegrees=5, swing2LimitDegrees=10, twistLimitDegrees=0}
 {bone1Name=dyn_skirt_sd_04_r, bone2Name=dyn_skirt_sd_03_r, swing1LimitDegrees=10, swing2LimitDegrees=10, twistLimitDegrees=0}
]
