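# Cosmetic AnimBlueprint for the F_MED_JoltMosaic female body, adding secondary motion to skirt, bows and shoelaces.
# The copied parent pose runs through a state-driven Control Rig and rear-skirt AnimDynamics into a cache.
# With rigid bodies enabled, standing and crouch branches add per-leg RigidBody and center-bone averaging constraints, blend, then receive offsets and bow/lace dynamics; otherwise the cache outputs directly.
# EventGraph derives state alphas and gravity overrides.
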AnimBlueprint /BRCosmetics/Characters/Player/Female/Medium/Bodies/F_MED_JoltMosaic/Meshes/F_MED_JoltMosaic_AnimBP
target_skeleton=&/F_MED_JoltMosaic_Skeleton
graphs [
 AnimGraph {nodes [
   Copy Pose From Mesh() {bUseAttachedParent=true}
   "Transform (Modify) Bone - Bone: attach_cape"(Alpha=1, Translation="0,-5,2", Rotation="0,0,0", Scale="1,1,1") {
    rotationMode=BMM_Additive
    rotationSpace=BCS_BoneSpace
    scaleMode=BMM_Additive
    scaleSpace=BCS_BoneSpace
    translationMode=BMM_Additive
    translationSpace=BCS_BoneSpace
    alphaBoolBlend {blendInTime=0.2, blendOutTime=0.2}
    alphaScaleBiasClamp {bInterpResult=true}
    boneToModify {boneName=attach_cape}
    translation {y=-5, z=2}
   }
   n1 = Use cached pose 'CrouchCache'()
   n2 = Use cached pose 'StandingCache'()
   n3 = Get bIsCrouching()
   n4 = "Transform (Modify) Bone - Bone: dyn_skirt_fr_04_c"!(
    ComponentPose=Local To Component(LocalPose=Blend Poses by bool(BlendPose_0=@n1, BlendPose_1=@n2, BlendTime_0=0.5, bActiveValue=@n3))
    bAlphaBoolEnabled:bool=True
    Translation="8,0,4.155"
    Rotation="0.014,-0.011,-0.815"
    Scale="1,1,1"
   ) -> (Pose) {
    alphaInputType=Bool
    rotationMode=BMM_Additive
    rotationSpace=BCS_BoneSpace
    scaleMode=BMM_Additive
    scaleSpace=BCS_BoneSpace
    translationMode=BMM_Additive
    translationSpace=BCS_BoneSpace
    alphaBoolBlend {blendInTime=0.1, blendOutTime=0.2}
    alphaScaleBiasClamp {bInterpResult=true}
    boneToModify {boneName=dyn_skirt_fr_04_c}
    rotation=<rot 0.014 -0.011 -0.815>
    translation {x=8, z=4.155}
   }
   n5 = "Transform (Modify) Bone - Bone: dyn_skirt_bk_03_c"!(ComponentPose=@n4, bAlphaBoolEnabled:bool=True, Translation="0,0,0", Rotation="-15.355,-0.016,0.008", Scale="1,1,1") -> (Pose) {
    alphaInputType=Bool
    rotationMode=BMM_Additive
    rotationSpace=BCS_BoneSpace
    scaleMode=BMM_Additive
    scaleSpace=BCS_BoneSpace
    translationMode=BMM_Additive
    translationSpace=BCS_BoneSpace
    alphaBoolBlend {blendInTime=0.1, blendOutTime=0.2}
    alphaScaleBiasClamp {bInterpResult=true}
    boneToModify {boneName=dyn_skirt_bk_03_c}
    rotation=<rot -15.355 -0.0161 0.008>
   }
   n6 = "Transform (Modify) Bone - Bone: dyn_skirt_bk_03_l"!(ComponentPose=@n5, bAlphaBoolEnabled:bool=True, Translation="0,0,0", Rotation="-21.012,9.816,-3.550", Scale="1,1,1") -> (Pose) {
    alphaInputType=Bool
    rotationMode=BMM_Additive
    rotationSpace=BCS_BoneSpace
    scaleMode=BMM_Additive
    scaleSpace=BCS_BoneSpace
    translationMode=BMM_Additive
    translationSpace=BCS_BoneSpace
    alphaBoolBlend {blendInTime=0.1, blendOutTime=0.2}
    alphaScaleBiasClamp {bInterpResult=true}
    boneToModify {boneName=dyn_skirt_bk_03_l}
    rotation=<rot -21.012 9.816 -3.55>
   }
   n7 = "Transform (Modify) Bone - Bone: dyn_skirt_bk_03_r"!(ComponentPose=@n6, bAlphaBoolEnabled:bool=True, Translation="0,0,0", Rotation="-14.356,-5.023,1.246", Scale="1,1,1") -> (Pose) {
    alphaInputType=Bool
    rotationMode=BMM_Additive
    rotationSpace=BCS_BoneSpace
    scaleMode=BMM_Additive
    scaleSpace=BCS_BoneSpace
    translationMode=BMM_Additive
    translationSpace=BCS_BoneSpace
    alphaBoolBlend {blendInTime=0.1, blendOutTime=0.2}
    alphaScaleBiasClamp {bInterpResult=true}
    boneToModify {boneName=dyn_skirt_bk_03_r}
    rotation=<rot -14.356 -5.023 1.246>
   }
   n8 = "Transform (Modify) Bone - Bone: base_skirt_bk_r"!(ComponentPose=@n7, bAlphaBoolEnabled:bool=True, Translation="-0.055,0.010,0.562", Rotation="5.638,-10.912,0.000", Scale="1,1,1") -> (Pose) {
    alphaInputType=Bool
    rotationMode=BMM_Additive
    rotationSpace=BCS_BoneSpace
    scaleMode=BMM_Additive
    scaleSpace=BCS_BoneSpace
    translationMode=BMM_Additive
    translationSpace=BCS_BoneSpace
    alphaBoolBlend {blendInTime=0.1, blendOutTime=0.2}
    alphaScaleBiasClamp {bInterpResult=true}
    boneToModify {boneName=base_skirt_bk_r}
    rotation {pitch=5.638, yaw=-10.912}
    translation=<-0.055 0.01 0.562>
   }
   n9 = "Transform (Modify) Bone - Bone: base_skirt_bk_c"!(ComponentPose=@n8, bAlphaBoolEnabled:bool=True, Translation="-0.000,-0.205,1.233", Rotation="0.618,-0.022,0.011", Scale="1,1,1") -> (Pose) {
    alphaInputType=Bool
    rotationMode=BMM_Additive
    rotationSpace=BCS_BoneSpace
    scaleMode=BMM_Additive
    scaleSpace=BCS_BoneSpace
    translationMode=BMM_Additive
    translationSpace=BCS_BoneSpace
    alphaBoolBlend {blendInTime=0.1, blendOutTime=0.2}
    alphaScaleBiasClamp {bInterpResult=true}
    boneToModify {boneName=base_skirt_bk_c}
    rotation=<rot 0.618 -0.022 0.011>
    translation {y=-0.205, z=1.233}
   }
   n10 = "Transform (Modify) Bone - Bone: base_skirt_fr_c"(ComponentPose=@n9, Alpha=1, Translation="0,0,0", Rotation="-20,0,0", Scale="1,1,1") {
    rotationMode=BMM_Additive
    rotationSpace=BCS_BoneSpace
    scaleMode=BMM_Additive
    scaleSpace=BCS_BoneSpace
    translationMode=BMM_Additive
    translationSpace=BCS_BoneSpace
    alphaBoolBlend {blendInTime=0.1, blendOutTime=0.2}
    alphaScaleBiasClamp {bInterpResult=true}
    boneToModify {boneName=base_skirt_fr_c}
    rotation {pitch=-20}
   }
   n11 = Anim Dynamics - Bone: Bow_C_base(ComponentPose=@n10, Alpha=1, ExternalForce="0,0,0") {
    angularDampingOverride=1
    angularSpringConstant=20000
    bAngularSpring=true
    bOverrideAngularDamping=true
    bUseGravityOverride=true
    gravityScale=0.35
    simulationSpace=BoneRelative
    boundBone {boneName=Bow_C_base}
    componentLinearAccScale=<1 1 1>
    componentLinearVelScale=<1 1 1>
    gravityOverride {z=-980}
    physicsBodyDefinitions [{
      boundBone {boneName=Bow_C_base}
      boxExtents=<20 20 20>
      constraintSetup {angularLimitsMax=<0 -2 0>, angularLimitsMin=<0 -2 0>, angularTarget=<1 0 0>}
      localJointOffset=<4 0 0>
     }]
    relativeSpaceBone {boneName=spine_04}
   }
   n12 = Anim Dynamics - Bone: Bow_l_dyn(ComponentPose=@n11, Alpha=1, ExternalForce="0,0,0") {
    angularDampingOverride=1
    angularSpringConstant=7000
    bAngularSpring=true
    bOverrideAngularDamping=true
    bUseGravityOverride=true
    gravityScale=0.35
    simulationSpace=BoneRelative
    boundBone {boneName=Bow_l_dyn}
    componentLinearAccScale=<1 1 1>
    componentLinearVelScale=<1 1 1>
    gravityOverride {z=-980}
    physicsBodyDefinitions [{
      boundBone {boneName=Bow_l_dyn}
      boxExtents=<10 15 15>
      constraintSetup {angularLimitsMax=<0 20 25>, angularLimitsMin=<0 0 -25>, angularTarget=<1 0 0>}
      localJointOffset=<4 0 0>
     }]
    relativeSpaceBone {boneName=spine_04}
   }
   n13 = Anim Dynamics - Bone: Bow_r_dyn(ComponentPose=@n12, Alpha=1, ExternalForce="0,0,0") {
    angularDampingOverride=1
    angularSpringConstant=7000
    bAngularSpring=true
    bOverrideAngularDamping=true
    bUseGravityOverride=true
    gravityScale=0.35
    simulationSpace=BoneRelative
    boundBone {boneName=Bow_r_dyn}
    componentLinearAccScale=<1 1 1>
    componentLinearVelScale=<1 1 1>
    gravityOverride {z=-980}
    physicsBodyDefinitions [{
      boundBone {boneName=Bow_r_dyn}
      boxExtents=<10 15 15>
      constraintSetup {angularLimitsMax=<0 20 20>, angularLimitsMin=<0 0 -25>, angularTarget=<1 0 0>}
      localJointOffset=<4 0 0>
     }]
    relativeSpaceBone {boneName=spine_04}
   }
   n14 = Anim Dynamics - Bone: Bow_side_l_dyn(ComponentPose=@n13, Alpha=1, ExternalForce="0,0,0") {
    angularDampingOverride=1
    angularSpringConstant=7000
    bAngularSpring=true
    bOverrideAngularDamping=true
    bUseGravityOverride=true
    gravityScale=0.35
    simulationSpace=BoneRelative
    boundBone {boneName=Bow_side_l_dyn}
    componentLinearAccScale=<1 1 1>
    componentLinearVelScale=<1 1 1>
    gravityOverride {z=-980}
    physicsBodyDefinitions [{
      boundBone {boneName=Bow_side_l_dyn}
      boxExtents=<10 15 15>
      constraintSetup {angularLimitsMax=<0 40 10>, angularLimitsMin=<0 0 -10>, angularTarget=<1 0 0>}
      localJointOffset=<4 0 0>
     }]
    relativeSpaceBone {boneName=spine_04}
   }
   n15 = Anim Dynamics - Bone: Bow_side_r_dyn(ComponentPose=@n14, Alpha=1, ExternalForce="0,0,0") {
    angularDampingOverride=1
    angularSpringConstant=7000
    bAngularSpring=true
    bOverrideAngularDamping=true
    bUseGravityOverride=true
    gravityScale=0.35
    simulationSpace=BoneRelative
    boundBone {boneName=Bow_side_r_dyn}
    componentLinearAccScale=<1 1 1>
    componentLinearVelScale=<1 1 1>
    gravityOverride {z=-980}
    physicsBodyDefinitions [{
      boundBone {boneName=Bow_side_r_dyn}
      boxExtents=<10 15 15>
      constraintSetup {angularLimitsMax=<0 30 15>, angularLimitsMin=<0 0 -15>, angularTarget=<1 0 0>}
      localJointOffset=<4 0 0>
     }]
    relativeSpaceBone {boneName=spine_04}
   }
   n16 = Anim Dynamics - Bone: Shoelaces_l_dyn(ComponentPose=@n15, Alpha=1, ExternalForce="0,0,0") {
    angularDampingOverride=1
    angularSpringConstant=10000
    bAngularSpring=true
    bOverrideAngularDamping=true
    bUseGravityOverride=true
    gravityScale=0.35
    simulationSpace=BoneRelative
    boundBone {boneName=Shoelaces_l_dyn}
    componentAppliedLinearAccClamp=<10000 10000 10000>
    componentLinearAccScale=<1 1 1>
    componentLinearVelScale=<1 1 1>
    gravityOverride {z=-980}
    physicsBodyDefinitions [{
      boundBone {boneName=Shoelaces_l_dyn}
      boxExtents=<20 30 30>
      constraintSetup {angularLimitsMax=<0 0 1>, angularLimitsMin=<0 -2 -1>, angularTarget=<1 0 0>}
      localJointOffset=<4 0 0>
     }]
    relativeSpaceBone {boneName=calf_l}
    simSpaceSettings {simSpaceAngularAlpha=1}
   }
   n17 = Anim Dynamics - Bone: Showlace_left_r_dyn(ComponentPose=@n16, Alpha=1, ExternalForce="0,0,0") {
    angularDampingOverride=1
    angularSpringConstant=5000
    bAngularSpring=true
    bOverrideAngularDamping=true
    bUseGravityOverride=true
    gravityScale=0.35
    simulationSpace=BoneRelative
    boundBone {boneName=Showlace_left_r_dyn}
    componentAppliedLinearAccClamp=<10000 10000 10000>
    componentLinearAccScale=<1 1 1>
    componentLinearVelScale=<1 1 1>
    gravityOverride {z=-980}
    physicsBodyDefinitions [{
      boundBone {boneName=Showlace_left_r_dyn}
      boxExtents=<10 15 15>
      constraintSetup {angularLimitsMax=<5 5 20>, angularLimitsMin=<-5 -5 0>, angularTarget=<1 0 0>}
      localJointOffset=<0 4 0>
     }]
    relativeSpaceBone {boneName=calf_l}
    simSpaceSettings {simSpaceAngularAlpha=1}
   }
   n18 = Anim Dynamics - Bone: Showlace_left_l_dyn(ComponentPose=@n17, Alpha=1, ExternalForce="0,0,0") {
    angularDampingOverride=1
    angularSpringConstant=5000
    bAngularSpring=true
    bOverrideAngularDamping=true
    bUseGravityOverride=true
    gravityScale=0.35
    simulationSpace=BoneRelative
    boundBone {boneName=Showlace_left_l_dyn}
    componentAppliedLinearAccClamp=<10000 10000 10000>
    componentLinearAccScale=<1 1 1>
    componentLinearVelScale=<1 1 1>
    gravityOverride {z=-980}
    physicsBodyDefinitions [{
      boundBone {boneName=Showlace_left_l_dyn}
      boxExtents=<10 15 15>
      constraintSetup {angularLimitsMax=<5 5 20>, angularLimitsMin=<-5 -5 -5>, angularTarget=<1 0 0>}
      localJointOffset=<0 4 0>
     }]
    relativeSpaceBone {boneName=calf_l}
    simSpaceSettings {simSpaceAngularAlpha=1}
   }
   n19 = Anim Dynamics - Bone: Left_knot_r_dyn(ComponentPose=@n18, Alpha=1, ExternalForce="0,0,0") {
    angularDampingOverride=1
    angularSpringConstant=5000
    bAngularSpring=true
    bOverrideAngularDamping=true
    bUseGravityOverride=true
    gravityScale=0.35
    simulationSpace=BoneRelative
    boundBone {boneName=Left_knot_r_dyn}
    componentAppliedLinearAccClamp=<10000 10000 10000>
    componentLinearAccScale=<1 1 1>
    componentLinearVelScale=<1 1 1>
    gravityOverride {z=-980}
    physicsBodyDefinitions [{
      boundBone {boneName=Left_knot_r_dyn}
      boxExtents=<10 15 15>
      constraintSetup {angularLimitsMax=<3 10 15>, angularLimitsMin=<-3 -10 -15>, angularTarget=<1 0 0>}
      localJointOffset=<0 4 0>
     }]
    relativeSpaceBone {boneName=calf_l}
    simSpaceSettings {simSpaceAngularAlpha=1}
   }
   n20 = Anim Dynamics - Bone: Left_knot_l_dyn(ComponentPose=@n19, Alpha=1, ExternalForce="0,0,0") {
    angularDampingOverride=1
    angularSpringConstant=5000
    bAngularSpring=true
    bOverrideAngularDamping=true
    bUseGravityOverride=true
    gravityScale=0.35
    simulationSpace=BoneRelative
    boundBone {boneName=Left_knot_l_dyn}
    componentAppliedLinearAccClamp=<1000 1000 1000>
    componentLinearAccScale=<0.3 0.3 1>
    componentLinearVelScale=<0.3 0.3 1>
    gravityOverride {z=-980}
    physicsBodyDefinitions [{
      boundBone {boneName=Left_knot_l_dyn}
      boxExtents=<10 15 15>
      constraintSetup {angularLimitsMax=<3 10 15>, angularLimitsMin=<-3 -10 -15>, angularTarget=<1 0 0>}
      localJointOffset=<0 4 0>
     }]
    relativeSpaceBone {boneName=calf_l}
   }
   n21 = Anim Dynamics - Bone: Shoelaces_r_dyn(ComponentPose=@n20, Alpha=1, ExternalForce="0,0,0") {
    angularDampingOverride=1
    angularSpringConstant=5000
    bAngularSpring=true
    bOverrideAngularDamping=true
    bUseGravityOverride=true
    gravityScale=0.35
    simulationSpace=BoneRelative
    boundBone {boneName=Shoelaces_r_dyn}
    componentLinearAccScale=<1 1 1>
    componentLinearVelScale=<1 1 1>
    gravityOverride {z=-980}
    physicsBodyDefinitions [{
      boundBone {boneName=Shoelaces_r_dyn}
      boxExtents=<10 15 15>
      constraintSetup {angularLimitsMax=<0 0 1>, angularLimitsMin=<0 -5 -1>, angularTarget=<1 0 0>}
      localJointOffset=<0 0 4>
     }]
    relativeSpaceBone {boneName=calf_r}
   }
   n22 = Anim Dynamics - Bone: Showlace_right_l_dyn(ComponentPose=@n21, Alpha=1, ExternalForce="0,0,0") {
    angularDampingOverride=1
    angularSpringConstant=5000
    bAngularSpring=true
    bOverrideAngularDamping=true
    bUseGravityOverride=true
    gravityScale=0.35
    simulationSpace=BoneRelative
    boundBone {boneName=Showlace_right_l_dyn}
    componentLinearAccScale=<1 1 1>
    componentLinearVelScale=<1 1 1>
    gravityOverride {z=-980}
    physicsBodyDefinitions [{
      boundBone {boneName=Showlace_right_l_dyn}
      boxExtents=<10 15 15>
      constraintSetup {angularLimitsMax=<0 5 -10>, angularLimitsMin=<0 -5 30>, angularTarget=<1 0 0>}
      localJointOffset=<-4 0 0>
     }]
    relativeSpaceBone {boneName=calf_r}
   }
   n23 = Anim Dynamics - Bone: Showlace_right_r_dyn(ComponentPose=@n22, Alpha=1, ExternalForce="0,0,0") {
    angularDampingOverride=1
    angularSpringConstant=5000
    bAngularSpring=true
    bOverrideAngularDamping=true
    bUseGravityOverride=true
    gravityScale=0.35
    simulationSpace=BoneRelative
    boundBone {boneName=Showlace_right_r_dyn}
    componentLinearAccScale=<1 1 1>
    componentLinearVelScale=<1 1 1>
    gravityOverride {z=-980}
    physicsBodyDefinitions [{
      boundBone {boneName=Showlace_right_r_dyn}
      boxExtents=<10 15 15>
      constraintSetup {angularLimitsMax=<0 10 -10>, angularLimitsMin=<0 -10 20>, angularTarget=<1 0 0>}
      localJointOffset=<-4 0 0>
     }]
    relativeSpaceBone {boneName=calf_r}
   }
   n24 = Anim Dynamics - Bone: Right_knot_r_dyn(ComponentPose=@n23, Alpha=1, ExternalForce="0,0,0") {
    angularDampingOverride=1
    angularSpringConstant=5000
    bAngularSpring=true
    bOverrideAngularDamping=true
    bUseGravityOverride=true
    gravityScale=0.35
    simulationSpace=BoneRelative
    boundBone {boneName=Right_knot_r_dyn}
    componentLinearAccScale=<1 1 1>
    componentLinearVelScale=<1 1 1>
    gravityOverride {z=-980}
    physicsBodyDefinitions [{
      boundBone {boneName=Right_knot_r_dyn}
      boxExtents=<10 15 15>
      constraintSetup {angularLimitsMax=<0 5 15>, angularLimitsMin=<0 -5 -15>, angularTarget=<1 0 0>}
      localJointOffset=<-3 -4 0>
     }]
    relativeSpaceBone {boneName=calf_r}
   }
   n25 = Anim Dynamics - Bone: Right_knot_r_dyn1(ComponentPose=@n24, Alpha=1, ExternalForce="0,0,0") {
    angularDampingOverride=1
    angularSpringConstant=5000
    bAngularSpring=true
    bOverrideAngularDamping=true
    bUseGravityOverride=true
    gravityScale=0.35
    simulationSpace=BoneRelative
    boundBone {boneName=Right_knot_r_dyn1}
    componentLinearAccScale=<1 1 1>
    componentLinearVelScale=<1 1 1>
    gravityOverride {z=-980}
    physicsBodyDefinitions [{
      boundBone {boneName=Right_knot_r_dyn1}
      boxExtents=<10 15 15>
      constraintSetup {angularLimitsMax=<15 15 15>, angularLimitsMin=<-15 -15 -15>, angularTarget=<1 0 0>}
      localJointOffset=<0 0 4>
     }]
    relativeSpaceBone {boneName=calf_r}
   }
   Output Pose(Result=Blend Poses by bool(BlendPose_0=Component To Local(ComponentPose=@n25), BlendPose_1=Use cached pose 'ControlRigCache'(), bActiveValue=Get bRigidBodyEnabled())) {layerGroup=None}
   n26 = Control Rig!(
    Source=Copy Pose From Mesh() {bUseAttachedParent=true}
    Alpha
    CrouchingBool:bool=Get bIsCrouching()
    DBNOBool:bool=Get bIsDBNO()
    DBNOIdle:double=Get DBNOIdle() -> (DBNOIdle:double)
    GrindRail:bool=Get Is On Grind Rail() -> (Is On Grind Rail:bool)
    DBNOCarried:bool=Get bIsDBNOCarrying() -> (bIsDBNOCarrying:bool)
    Idle:double=Get Idle() -> (Idle:double)
    Jumping:bool=Get bIsJumping()
    HangGlider:double=Get HangGlidingAlpha() -> (HangGlidingAlpha:double)
    FallingOnGround:double=Get Alpha_Falling() -> (Alpha_Falling:double)
    swingPickaxe:bool=Get bIsPlayingMeleeAnim()
   ) -> (Pose) {controlRigAssetReference {blueprintRigClass=&/F_MED_JoltMosaic_CtrlRig.F_MED_JoltMosaic_CtrlRig_C}}
   n27 = Anim Dynamics - Bone: dyn_skirt_bk_02_r!(ComponentPose=Local To Component(LocalPose=@n26), Alpha=1, GravityOverride:Vector="0,0,-980", ExternalForce="0,0,0") -> (Pose) {
    angularDampingOverride=0.9999
    angularSpringConstant=1000
    bAngularSpring=true
    bOverrideAngularDamping=true
    bUseGravityOverride=true
    gravityScale=0.35
    boundBone {boneName=dyn_skirt_bk_02_r}
    componentAppliedLinearAccClamp=<3000 3000 3000>
    componentLinearAccScale=<0.4 0.4 0.4>
    componentLinearVelScale=<0.4 0.4 0.4>
    gravityOverride {z=-980}
    physicsBodyDefinitions [{
      boundBone {boneName=dyn_skirt_bk_02_r}
      boxExtents=<15 20 20>
      constraintSetup {angularLimitsMax=<2 0 4>, angularLimitsMin=<-2 -10 -4>, angularTarget=<1 0 0>}
      localJointOffset=<4 0 0>
     }]
   }
   n28 = Anim Dynamics - Bone: dyn_skirt_bk_03_r!(ComponentPose=@n27, Alpha=1, GravityOverride:Vector="0,0,-980", ExternalForce="0,0,0") -> (Pose) {
    angularDampingOverride=0.9999
    angularSpringConstant=1000
    bAngularSpring=true
    bOverrideAngularDamping=true
    bUseGravityOverride=true
    gravityScale=0.35
    boundBone {boneName=dyn_skirt_bk_03_r}
    componentAppliedLinearAccClamp=<3000 3000 3000>
    componentLinearAccScale=<0.4 0.4 0.4>
    componentLinearVelScale=<0.4 0.4 0.4>
    gravityOverride {z=-980}
    physicsBodyDefinitions [{
      boundBone {boneName=dyn_skirt_bk_03_r}
      boxExtents=<15 20 20>
      constraintSetup {angularLimitsMax=<2 0 4>, angularLimitsMin=<-2 -10 -4>, angularTarget=<1 0 0>}
      localJointOffset=<4 0 0>
     }]
   }
   n29 = Anim Dynamics - Bone: dyn_skirt_bk_04_r!(ComponentPose=@n28, Alpha=1, GravityOverride:Vector="0,0,-980", ExternalForce="0,0,0") -> (Pose) {
    angularDampingOverride=0.9999
    angularSpringConstant=1000
    bAngularSpring=true
    bOverrideAngularDamping=true
    bUseGravityOverride=true
    gravityScale=0.35
    boundBone {boneName=dyn_skirt_bk_04_r}
    componentAppliedLinearAccClamp=<3000 3000 3000>
    componentLinearAccScale=<0.4 0.4 0.4>
    componentLinearVelScale=<0.4 0.4 0.4>
    gravityOverride {z=-980}
    physicsBodyDefinitions [{
      boundBone {boneName=dyn_skirt_bk_04_r}
      boxExtents=<15 20 20>
      constraintSetup {angularLimitsMax=<2 0 4>, angularLimitsMin=<-2 -10 -4>, angularTarget=<1 0 0>}
      localJointOffset=<4 0 0>
     }]
   }
   n30 = Anim Dynamics - Bone: dyn_skirt_bk_02_l!(ComponentPose=@n29, Alpha=1, GravityOverride:Vector="0,0,-980", ExternalForce="0,0,0") -> (Pose) {
    angularDampingOverride=0.9999
    angularSpringConstant=1000
    bAngularSpring=true
    bOverrideAngularDamping=true
    bUseGravityOverride=true
    gravityScale=0.35
    boundBone {boneName=dyn_skirt_bk_02_l}
    componentAppliedLinearAccClamp=<3000 3000 3000>
    componentLinearAccScale=<0.4 0.4 0.4>
    componentLinearVelScale=<0.4 0.4 0.4>
    gravityOverride {z=-980}
    physicsBodyDefinitions [{
      boundBone {boneName=dyn_skirt_bk_02_l}
      boxExtents=<15 20 20>
      constraintSetup {angularLimitsMax=<2 0 4>, angularLimitsMin=<-2 -10 -4>, angularTarget=<1 0 0>}
      localJointOffset=<4 0 0>
     }]
   }
   n31 = Anim Dynamics - Bone: dyn_skirt_bk_03_l!(ComponentPose=@n30, Alpha=1, GravityOverride:Vector="0,0,-980", ExternalForce="0,0,0") -> (Pose) {
    angularDampingOverride=0.9999
    angularSpringConstant=1000
    bAngularSpring=true
    bOverrideAngularDamping=true
    bUseGravityOverride=true
    gravityScale=0.35
    boundBone {boneName=dyn_skirt_bk_03_l}
    componentAppliedLinearAccClamp=<3000 3000 3000>
    componentLinearAccScale=<0.4 0.4 0.4>
    componentLinearVelScale=<0.4 0.4 0.4>
    gravityOverride {z=-980}
    physicsBodyDefinitions [{
      boundBone {boneName=dyn_skirt_bk_03_l}
      boxExtents=<15 20 20>
      constraintSetup {angularLimitsMax=<2 0 4>, angularLimitsMin=<-2 -10 -4>, angularTarget=<1 0 0>}
      localJointOffset=<4 0 0>
     }]
   }
   n32 = Anim Dynamics - Bone: dyn_skirt_bk_04_l!(ComponentPose=@n31, Alpha=1, GravityOverride:Vector="0,0,-980", ExternalForce="0,0,0") -> (Pose) {
    angularDampingOverride=0.9999
    angularSpringConstant=1000
    bAngularSpring=true
    bOverrideAngularDamping=true
    bUseGravityOverride=true
    gravityScale=0.35
    boundBone {boneName=dyn_skirt_bk_04_l}
    componentAppliedLinearAccClamp=<3000 3000 3000>
    componentLinearAccScale=<0.4 0.4 0.4>
    componentLinearVelScale=<0.4 0.4 0.4>
    gravityOverride {z=-980}
    physicsBodyDefinitions [{
      boundBone {boneName=dyn_skirt_bk_04_l}
      boxExtents=<15 20 20>
      constraintSetup {angularLimitsMax=<2 0 4>, angularLimitsMin=<-2 -10 -4>, angularTarget=<1 0 0>}
      localJointOffset=<4 0 0>
     }]
   }
   Save cached pose 'ControlRigCache'(Pose=Component To Local(ComponentPose=@n32)) {cacheName=ControlRigCache}
   n33 = RigidBody!(
    ComponentPose=Local To Component(LocalPose=Use cached pose 'ControlRigCache'())
    Alpha
    OverrideWorldGravity:Vector="0,0,0"
    ExternalForce
    SimSpaceSettings:Sim Space Settings Structure=(WorldAlpha=0,VelocityScaleZ=1,DampingAlpha=1,MaxLinearVelocity=10000,MaxAngularVelocity=10000,MaxLinearAcceleration=10000,MaxAngularAcceleration=10000,ExternalLinearDragV=(X=0,Y=0,Z=0),ExternalLinearVelocity=(X=0,Y=0,Z=0),ExternalAngularVelocity=(X=0,Y=0,Z=0))
   ) -> (Pose) {
    bOverrideWorldGravity=true
    overridePhysicsAsset=&/F_MED_JoltMosaic_LeftLeg_Physics
    simulationSpace=BaseBoneSpace
    baseBoneRef {boneName=thigh_l}
    componentAppliedLinearAccClamp=<3000 3000 3000>
   }
   n34 = RigidBody(
    ComponentPose=@n33
    OverrideWorldGravity="0,0,0"
    SimSpaceSettings=(WorldAlpha=0,VelocityScaleZ=1,DampingAlpha=1,MaxLinearVelocity=10000,MaxAngularVelocity=10000,MaxLinearAcceleration=10000,MaxAngularAcceleration=10000,ExternalLinearDragV=(X=0,Y=0,Z=0),ExternalLinearVelocity=(X=0,Y=0,Z=0),ExternalAngularVelocity=(X=0,Y=0,Z=0))
   ) {
    bOverrideWorldGravity=true
    overridePhysicsAsset=&/F_MED_JoltMosaic_RightLeg_Physics
    simulationSpace=BaseBoneSpace
    baseBoneRef {boneName=thigh_r}
    componentAppliedLinearAccClamp=<3000 3000 3000>
   }
   n35 = "Constraint - dyn_skirt_bk_02_c (RR)"!(ComponentPose=@n34, bAlphaBoolEnabled:bool=True, ConstraintWeights_0=0.5, ConstraintWeights_1:float=0.5) -> (Pose) {
    alphaInputType=Bool
    boneToModify {boneName=dyn_skirt_bk_02_c}
    constraintSetup [{transformType=Rotation, targetBone {boneName=dyn_skirt_bk_02_l}}, {transformType=Rotation, targetBone {boneName=dyn_skirt_bk_02_r}}]
    constraintWeights [1,1]
   }
   n36 = "Constraint - dyn_skirt_bk_03_c (RR)"!(ComponentPose=@n35, bAlphaBoolEnabled:bool=True, ConstraintWeights_0=0.5, ConstraintWeights_1:float=0.5) -> (Pose) {
    alphaInputType=Bool
    boneToModify {boneName=dyn_skirt_bk_03_c}
    constraintSetup [{transformType=Rotation, targetBone {boneName=dyn_skirt_bk_03_l}}, {transformType=Rotation, targetBone {boneName=dyn_skirt_bk_03_r}}]
    constraintWeights [1,1]
   }
   n37 = "Constraint - dyn_skirt_bk_04_c (RR)"!(ComponentPose=@n36, bAlphaBoolEnabled:bool=True, ConstraintWeights_0=0.5, ConstraintWeights_1:float=0.5) -> (Pose) {
    alphaInputType=Bool
    boneToModify {boneName=dyn_skirt_bk_04_c}
    constraintSetup [{transformType=Rotation, targetBone {boneName=dyn_skirt_bk_04_l}}, {transformType=Rotation, targetBone {boneName=dyn_skirt_bk_04_r}}]
    constraintWeights [1,1]
   }
   n38 = "Constraint - dyn_skirt_fr_02_c (RR)"!(ComponentPose=@n37, Alpha=1, ConstraintWeights_0=0.5, ConstraintWeights_1:float=0.5) -> (Pose) {
    alphaScaleBiasClamp {bInterpResult=true, interpSpeedDecreasing=5, interpSpeedIncreasing=5}
    boneToModify {boneName=dyn_skirt_fr_02_c}
    constraintSetup [{transformType=Rotation, targetBone {boneName=dyn_skirt_fr_02_l}}, {transformType=Rotation, targetBone {boneName=dyn_skirt_fr_02_r}}]
    constraintWeights [1,1]
   }
   n39 = "Constraint - dyn_skirt_fr_03_c (RR)"!(ComponentPose=@n38, Alpha=1, ConstraintWeights_0=0.5, ConstraintWeights_1:float=0.5) -> (Pose) {
    alphaScaleBiasClamp {bInterpResult=true, interpSpeedDecreasing=5, interpSpeedIncreasing=5}
    boneToModify {boneName=dyn_skirt_fr_03_c}
    constraintSetup [{transformType=Rotation, targetBone {boneName=dyn_skirt_fr_03_l}}, {transformType=Rotation, targetBone {boneName=dyn_skirt_fr_03_r}}]
    constraintWeights [1,1]
   }
   n40 = "Constraint - dyn_skirt_fr_04_c (RR)"!(ComponentPose=@n39, Alpha=1, ConstraintWeights_0=0.5, ConstraintWeights_1:float=0.5) -> (Pose) {
    alphaScaleBiasClamp {bInterpResult=true, interpSpeedDecreasing=5, interpSpeedIncreasing=5}
    boneToModify {boneName=dyn_skirt_fr_04_c}
    constraintSetup [{transformType=Rotation, targetBone {boneName=dyn_skirt_fr_04_l}}, {transformType=Rotation, targetBone {boneName=dyn_skirt_fr_04_r}}]
    constraintWeights [1,1]
   }
   Save cached pose 'StandingCache'(Pose=Component To Local(ComponentPose=@n40)) {cacheName=StandingCache}
   n41 = RigidBody(
    ComponentPose=Local To Component(LocalPose=Use cached pose 'ControlRigCache'())
    OverrideWorldGravity="0,0,0"
    SimSpaceSettings=(WorldAlpha=0,VelocityScaleZ=1,DampingAlpha=1,MaxLinearVelocity=10000,MaxAngularVelocity=10000,MaxLinearAcceleration=10000,MaxAngularAcceleration=10000,ExternalLinearDragV=(X=0,Y=0,Z=0),ExternalLinearVelocity=(X=0,Y=0,Z=0),ExternalAngularVelocity=(X=0,Y=0,Z=0))
   ) {
    bOverrideWorldGravity=true
    overridePhysicsAsset=&F_MED_JoltMosaic_LeftLeg_Physics
    simulationSpace=BaseBoneSpace
    baseBoneRef {boneName=thigh_twist_01_l}
    componentAppliedLinearAccClamp=<3000 3000 3000>
   }
   n42 = RigidBody(
    ComponentPose=@n41
    OverrideWorldGravity="0,0,0"
    SimSpaceSettings=(WorldAlpha=0,VelocityScaleZ=1,DampingAlpha=1,MaxLinearVelocity=10000,MaxAngularVelocity=10000,MaxLinearAcceleration=10000,MaxAngularAcceleration=10000,ExternalLinearDragV=(X=0,Y=0,Z=0),ExternalLinearVelocity=(X=0,Y=0,Z=0),ExternalAngularVelocity=(X=0,Y=0,Z=0))
   ) {
    bOverrideWorldGravity=true
    overridePhysicsAsset=&F_MED_JoltMosaic_RightLeg_Physics
    simulationSpace=BaseBoneSpace
    baseBoneRef {boneName=thigh_r}
    componentAppliedLinearAccClamp=<3000 3000 3000>
   }
   n43 = "Transform (Modify) Bone - Bone: base_skirt_fr_r"!(ComponentPose=@n42, bAlphaBoolEnabled:bool=True, Translation="0,0,0", Rotation="1.417,4.898,0.234", Scale="1,1,1") -> (Pose) {
    alphaInputType=Bool
    rotationMode=BMM_Additive
    rotationSpace=BCS_BoneSpace
    scaleMode=BMM_Additive
    scaleSpace=BCS_BoneSpace
    translationMode=BMM_Additive
    translationSpace=BCS_BoneSpace
    alphaBoolBlend {blendInTime=0.1, blendOutTime=0.2}
    boneToModify {boneName=base_skirt_fr_r}
    rotation=<rot 1.4171 4.898 0.234>
   }
   n44 = "Constraint - dyn_skirt_bk_02_c (RR)"(ComponentPose=@n43, bAlphaBoolEnabled=True, ConstraintWeights_0=0.5, ConstraintWeights_1=0.5) {
    alphaInputType=Bool
    boneToModify {boneName=dyn_skirt_bk_02_c}
    constraintSetup [{transformType=Rotation, targetBone {boneName=dyn_skirt_bk_02_l}}, {transformType=Rotation, targetBone {boneName=dyn_skirt_bk_02_r}}]
    constraintWeights [1,1]
   }
   n45 = "Constraint - dyn_skirt_bk_03_c (RR)"(ComponentPose=@n44, bAlphaBoolEnabled=True, ConstraintWeights_0=0.5, ConstraintWeights_1=0.5) {
    alphaInputType=Bool
    boneToModify {boneName=dyn_skirt_bk_03_c}
    constraintSetup [{transformType=Rotation, targetBone {boneName=dyn_skirt_bk_03_l}}, {transformType=Rotation, targetBone {boneName=dyn_skirt_bk_03_r}}]
    constraintWeights [1,1]
   }
   n46 = "Constraint - dyn_skirt_bk_04_c (RR)"(ComponentPose=@n45, bAlphaBoolEnabled=True, ConstraintWeights_0=0.5, ConstraintWeights_1=0.5) {
    alphaInputType=Bool
    boneToModify {boneName=dyn_skirt_bk_04_c}
    constraintSetup [{transformType=Rotation, targetBone {boneName=dyn_skirt_bk_04_l}}, {transformType=Rotation, targetBone {boneName=dyn_skirt_bk_04_r}}]
    constraintWeights [1,1]
   }
   n47 = "Transform (Modify) Bone - Bone: dyn_skirt_sd_02_l"(
    ComponentPose="Transform (Modify) Bone - Bone: base_skirt_fr_l"(ComponentPose=@n46, Alpha=1, Translation="0,0,0", Rotation="0,-10,10", Scale="1,1,1") {rotationMode=BMM_Additive, boneToModify {boneName=base_skirt_fr_l}, rotation {roll=10, yaw=-10}}
    Alpha=1
    Translation="0,0,1"
    Rotation="0,5,5"
    Scale="1,1,1"
   ) {
    rotationMode=BMM_Additive
    translationMode=BMM_Additive
    boneToModify {boneName=dyn_skirt_sd_02_l}
    rotation {roll=5, yaw=5}
    translation {z=1}
   }
   n48 = "Transform (Modify) Bone - Bone: dyn_skirt_sd_02_r"(ComponentPose=@n47, Alpha=1, Translation="0,0,0", Rotation="-0.960,3.349,-5.457", Scale="1,1,1") {rotationMode=BMM_Additive, translationMode=BMM_Additive, boneToModify {boneName=dyn_skirt_sd_02_r}, rotation=<rot -0.96 3.349 -5.457>}
   n49 = "Transform (Modify) Bone - Bone: dyn_skirt_fr_02_r"(ComponentPose=@n48, Alpha=1, Translation="0,0,0", Rotation="0,0,11.5", Scale="1,1,1") {rotationMode=BMM_Additive, translationMode=BMM_Additive, boneToModify {boneName=dyn_skirt_fr_02_r}, rotation {roll=11.5}}
   n50 = "Transform (Modify) Bone - Bone: base_skirt_bk_r"!(ComponentPose=@n49, Alpha:float=1, Translation="0,0,0", Rotation="0,0,20", Scale="1,1,1") -> (Pose) {rotationMode=BMM_Additive, translationMode=BMM_Additive, boneToModify {boneName=base_skirt_bk_r}, rotation {roll=20}}
   Save cached pose 'CrouchCache'(Pose=Component To Local(ComponentPose=@n50)) {cacheName=CrouchCache}
  ]}
 EventGraph {nodes [
   n1 = Event BlueprintUpdateAnimation()
   > n2 = Sequence!(execute) -> (then_0, then_1, then_2:Exec, then_3:Exec, then_4:Exec, then_5:Exec)
   > n3 = "Gravity Override State Function (Parameter Asset) (Nativized)"(
    CharacterPartAnimInstance=Self-Reference()
    GameTimeInSeconds=GetGameTimeInSeconds()
    DebugJointName=spine_05
    Parameters=&/F_MED_Jolt_Mosaic_Parameters
   )
   > Set Gravity Override(Gravity Override=@n3)
   > Set RBAN Settings Out(RBAN Settings Out=@n3.OutRigidBodyAnimNodeParameters)
   n4 = Branch(execute=@n2.then_2, Condition=Get bIsDBNO())
   > Set DBNO(DBNO:double=1) -> (Output_Get:double)
   > Set NoDBNO(NoDBNO:double=0) -> (Output_Get:double)
   Set DBNO(execute=@n4.else, DBNO=0)
   > Set NoDBNO(NoDBNO=1)
   n5 = Branch(execute=@n2.then_4, Condition=OR Boolean(A=Get bIsCrouching(), B=Get bIsCrouchSprinting()))
   > Set Crou_Or_CroSpring(Crou_Or_CroSpring:double=1) -> (Output_Get:double)
   Set Crou_Or_CroSpring(execute=@n5.else, Crou_Or_CroSpring=0)
   n6 = Get bIsSliding()
   n7 = OR Boolean(A=AND Boolean(A=Get bIsCrouching(), B=NOT Boolean(A=@n6)), B=Get bIsDBNO())
   n8 = Get bRigidBodyEnabled()
   Set CrouchNotSliding(execute=@n2.then_1, CrouchNotSliding:double="To Float (Boolean)"(InBool=AND Boolean(A=@n7, B=@n8))) -> (Output_Get:double)
   n9 = Get bIsSkydiveFloating()
   n10 = Get bIsSkydiveDiveMode()
   > Set FloatingAndSkydiveMode(FloatingAndSkydiveMode:double="To Float (Boolean)"(InBool=OR Boolean(A=@n9, B=@n10))) -> (Output_Get:double)
   n11 = NOT Boolean(A=Get bIsDBNOCarried())
   n12 = NOT Boolean(A=Get bIsMoving2D())
   n13 = NOT Boolean(A=Get bIsDBNOCarrying())
   n14 = Get bIsDBNO()
   > Set DBNOIdle(DBNOIdle:double="To Float (Boolean)"(InBool=AND Boolean!(A=@n12, B=@n14, C:bool=@n13, D:bool=@n11) -> (ReturnValue))) -> (Output_Get:double)
   n15 = NOT Boolean(A=Get bIsDBNO())
   n16 = NOT Boolean(A=Get bIsJumping())
   n17 = NOT Boolean(A=Get bIsPlayingEmote())
   n18 = NOT Boolean(A=Get bIsMoving2D())
   n19 = NOT Boolean(A=Get bIsSkyDiving())
   > Set Idle(Idle:double="To Float (Boolean)"(InBool=AND Boolean!(A=@n18, B=@n19, C=@n17, D=@n15, E:bool=@n16) -> (ReturnValue))) -> (Output_Get:double)
   n20 = NOT Boolean(A=Get bIsSkyDiving())
   n21 = Get bRigidBodyEnabled()
   n22 = Get bIsFalling()
   > Set Alpha_Falling(Alpha_Falling:double="To Float (Boolean)"(InBool=AND Boolean!(A=@n22, B=@n20, C=@n21) -> (ReturnValue))) -> (Output_Get:double)
   n23 = Get bIsDBNOCarried()
   n24 = OR Boolean!(
    A=Get bIsOnGrindRail()
    B=AND Boolean!(A=Get bIsPlayingMeleeAnim(), B=Get bIsCrouching()) -> (ReturnValue)
    C:bool=AND Boolean(A=NOT Boolean(A=@n23), B=Get bIsDBNO())
   ) -> (ReturnValue)
   > n25 = Branch(Condition=@n24)
   > Set RB Alpha(RB Alpha:double=FInterpTo(Current=Get RB Alpha() -> (RB Alpha:double), Target="To Float (Boolean)"(InBool=@n24), DeltaTime=@n1.DeltaTimeX, InterpSpeed=0.5)) -> (Output_Get:double)
   Set RB Alpha(execute=@n25.else, RB Alpha=1)
   n26 = GetOwnerSkeletalMeshComponent()
   n27 = BreakRotator(InRot=BreakTransform(InTransform=GetDeltaTransformFromRefPose(self=@n26, BoneName=spine_05, BaseName=pelvis)).Rotation)
   n28 = Branch(
    execute=@n2.then_3
    Condition=AND Boolean!(A="float > float"(A=@n27, B=12), B=Get bIsPlayingMeleeAnim(), C:bool=Get bIsMoving2D()) -> (ReturnValue)
   )
   > Set Alpha_ArmLBackwardMelee(Alpha_ArmLBackwardMelee:double=1) -> (Output_Get:double)
   Set Alpha_ArmLBackwardMelee(execute=@n28.else, Alpha_ArmLBackwardMelee=0)
   Event OnParachuteClosed()
   > Set HangGlidingAlpha(HangGlidingAlpha:double=0) -> (Output_Get:double)
   Event OnParachuteOpened()
   > Switch on EGliderType(Selection=Get PlayerGliderType() -> (PlayerGliderType=HangGlider))
   > Set HangGlidingAlpha(HangGlidingAlpha=1)
   Event OnBeginKneeSliding() {is=title K2Node_Event}
   > Set Is knee sliding(Is knee sliding:double=1) -> (Output_Get:double)
   Event OnEndKneeSliding() {is=title K2Node_Event}
   > Set Is knee sliding(Is knee sliding=0)
  ]}
]
variables [
 Alpha_ArmLBackwardMelee {type=number}
 Alpha_Falling {type=number}
 Crou_Or_CroSpring {type=number}
 CrouchNotSliding {type=number}
 DBNO {type=number}
 DBNOIdle {type=number}
 FloatingAndSkydiveMode {type=number}
 Gravity Override {type=Vector}
 HangGlidingAlpha {type=number}
 Idle {type=number}
 Is knee sliding {type=number}
 Is On Grind Rail {type=boolean}
 Iscelebration {type=boolean}
 IsCelebrationEmote {type=number}
 NoDBNO {type=number}
 RB Alpha {type=number}
 RBAN Settings Out {type=FortRigidBodyAnimNodeParameters}
]
