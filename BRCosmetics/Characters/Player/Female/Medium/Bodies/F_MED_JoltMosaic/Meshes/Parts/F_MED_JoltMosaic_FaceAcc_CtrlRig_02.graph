ControlRigBlueprint /BRCosmetics/Characters/Player/Female/Medium/Bodies/F_MED_JoltMosaic/Meshes/Parts/F_MED_JoltMosaic_FaceAcc_CtrlRig_02
graphs [
 CollapseNode_ContainedGraph {nodes [
   n1 = RigVMFunctionEntryNode()
   constraints = RigVMCollapseNode(ExecutePin=@n1.ExecutePin)
   lerp bools = RigVMCollapseNode(ExecuteContext=@constraints.ExecutePin)
   > RigVMFunction_Sequence = RigVMAggregateNode()
   RigVMFunction_MathBoolToFloat = &RigVMFunction_MathBoolToFloat(Value=RigVMVariableNode() {variable=pub isWearingBackBling}.Value)
   n2 = RigVMVariableNode() {variable=isFrontEnd}
   &RigUnit_ModifyTransforms(
    ItemToModify=((Item=(Type=Bone,Name="hair_side_r_01_base"),Transform=(Rotation=(X=-0.2162,Y=-0.0426,Z=-0.0094,W=0.9754),Translation=(X=0,Y=0.5,Z=1),Scale3D=(X=1,Y=1,Z=1))))
    WeightMinimum=0
    Weight=&RigVMFunction_MathDoubleMul(A=@RigVMFunction_MathBoolToFloat.Result, B=&RigVMFunction_MathBoolToFloat(Value=@n2.Value).Result).Result
    ExecutePin=@RigVMFunction_Sequence.E
   )
   RigVMFunctionReturnNode(ExecutePin=@RigVMFunction_Sequence.F)
   joints_twist_c = RigVMDispatchNode(Value=((Type=Bone,Name="dyn_hair_2_c"),(Type=Bone,Name="dyn_hair_3_c"),(Type=Bone,Name="dyn_hair_4_c"),(Type=Bone,Name="dyn_hair_5_c"),(Type=Bone,Name="dyn_hair_6_c"),(Type=Bone,Name="dyn_hair_7_c"))) {template="DISPATCH_RigVMDispatch_Constant(io Value)"}
   TwistBones = RigVMFunctionReferenceNode(
    Axis to Twist=(bX=True,bY=true,bZ=true)
    Debug=false
    End=(Type=Bone,Name="neck_01")
    Start=(Type=Bone,Name="spine_02")
    Bones to Twist=@joints_twist_c.Value
    ExecuteContext=@RigVMFunction_Sequence.B
   )
   joints_twist_l = RigVMDispatchNode(Value=((Type=Bone,Name="dyn_hair_2_l"),(Type=Bone,Name="dyn_hair_3_l"),(Type=Bone,Name="dyn_hair_4_l"),(Type=Bone,Name="dyn_hair_5_l"),(Type=Bone,Name="dyn_hair_6_l"),(Type=Bone,Name="dyn_hair_7_l"))) {template="DISPATCH_RigVMDispatch_Constant(io Value)"}
   > TwistBones_1 = RigVMFunctionReferenceNode(
    Axis to Twist=(bX=True,bY=true,bZ=true)
    Debug=false
    End=(Type=Bone,Name="neck_01")
    Start=(Type=Bone,Name="spine_02")
    Bones to Twist=@joints_twist_l.Value
   )
   joints_twist_r = RigVMDispatchNode(Value=((Type=Bone,Name="dyn_hair_2_r"),(Type=Bone,Name="dyn_hair_3_r"),(Type=Bone,Name="dyn_hair_4_r"),(Type=Bone,Name="dyn_hair_5_r"),(Type=Bone,Name="dyn_hair_6_r"),(Type=Bone,Name="dyn_hair_7_r"))) {template="DISPATCH_RigVMDispatch_Constant(io Value)"}
   > TwistBones_2 = RigVMFunctionReferenceNode(
    Axis to Twist=(bX=True,bY=true,bZ=true)
    Debug=false
    End=(Type=Bone,Name="neck_01")
    Start=(Type=Bone,Name="spine_02")
    Bones to Twist=@joints_twist_r.Value
   )
   RigVMFunction_MathBoolNand2 = /Script/RigVM.RigVMFunction_MathBoolNand2(
    B=&RigVMFunction_MathBoolNot(Value=RigVMVariableNode() {variable=pub isWearingBackBling}.Value).Result
    A=RigVMVariableNode() {variable=pub isRigidBodies Enabled}.Value
   ) {is=struct RigVMUnitNode, template="Nand::Execute(in A,in B,out Result)"}
   RigVMFunction_ControlFlowBranch = &RigVMFunction_ControlFlowBranch(Condition=@RigVMFunction_MathBoolNand2.Result, ExecuteContext=@TwistBones_2.ExecuteContext)
   controls_ik_c_x4 = RigVMDispatchNode(Value=((Type=Control,Name="dyn_hair_1_c_ctrl"),(Type=Control,Name="dyn_hair_2_c_ctrl"),(Type=Control,Name="dyn_hair_3_c_ctrl"),(Type=Control,Name="dyn_hair_4_c_ctrl"))) {template="DISPATCH_RigVMDispatch_Constant(io Value)"}
   joints_ik_c_x4 = RigVMDispatchNode(Value=((Type=Bone,Name="dyn_hair_1_c"),(Type=Bone,Name="dyn_hair_2_c"),(Type=Bone,Name="dyn_hair_3_c"),(Type=Bone,Name="dyn_hair_4_c"))) {template="DISPATCH_RigVMDispatch_Constant(io Value)"}
   SplineIK = RigVMFunctionReferenceNode(
    Primary Axis=(X=1,Y=0,Z=0)
    Samples Per Segment=1
    Secondary Spline Direction=(X=0,Y=0,Z=0)
    Spline Mode=BSpline
    Stretch=False
    Up Axis=(X=0,Y=0,Z=0)
    debug=false
    ExecuteContext=@RigVMFunction_ControlFlowBranch.False
    Controls=@controls_ik_c_x4.Value
    Bones=@joints_ik_c_x4.Value
   )
   controls_ik_l_x4 = RigVMDispatchNode(Value="((Type=Control,Name=\"dyn_hair_1_l_ctrl\"),(Type=Control,Name=\"dyn_hair_2_l_ctrl\"),(Type=Control,Name=\"dyn_hair_3_l_ctrl\"),(Type=Control,Name=\"dyn_hair_4_l_ctrl\"),(Type=Control,Name=\"dyn_hair_5_l_ctrl\"),(Type=Control,Name=\"dyn_hair_6_l_ctrl\"),(Type=Control,N ...(truncated)") {template="DISPATCH_RigVMDispatch_Constant(io Value)"}
   joints_ik_l_x4 = RigVMDispatchNode(Value=((Type=Bone,Name="dyn_hair_1_l"),(Type=Bone,Name="dyn_hair_2_l"),(Type=Bone,Name="dyn_hair_3_l"),(Type=Bone,Name="dyn_hair_4_l"))) {template="DISPATCH_RigVMDispatch_Constant(io Value)"}
   > SplineIK_1 = RigVMFunctionReferenceNode(
    Primary Axis=(X=1,Y=0,Z=0)
    Samples Per Segment=1
    Secondary Spline Direction=(X=0,Y=0,Z=0)
    Spline Mode=BSpline
    Stretch=False
    Up Axis=(X=0,Y=0,Z=0)
    debug=false
    Controls=@controls_ik_l_x4.Value
    Bones=@joints_ik_l_x4.Value
   )
   controls_ik_r_x4 = RigVMDispatchNode(Value=((Type=Control,Name="dyn_hair_1_r_ctrl"),(Type=Control,Name="dyn_hair_2_r_ctrl"),(Type=Control,Name="dyn_hair_3_r_ctrl"),(Type=Control,Name="dyn_hair_4_r_ctrl"))) {template="DISPATCH_RigVMDispatch_Constant(io Value)"}
   joints_ik_r_x4 = RigVMDispatchNode(Value=((Type=Bone,Name="dyn_hair_1_r"),(Type=Bone,Name="dyn_hair_2_r"),(Type=Bone,Name="dyn_hair_3_r"),(Type=Bone,Name="dyn_hair_4_r"))) {template="DISPATCH_RigVMDispatch_Constant(io Value)"}
   > SplineIK_2 = RigVMFunctionReferenceNode(
    Primary Axis=(X=1,Y=0,Z=0)
    Samples Per Segment=1
    Secondary Spline Direction=(X=0,Y=0,Z=0)
    Spline Mode=BSpline
    Stretch=False
    Up Axis=(X=0,Y=0,Z=0)
    debug=false
    Controls=@controls_ik_r_x4.Value
    Bones=@joints_ik_r_x4.Value
   )
   controls_ik_c_x7 = RigVMDispatchNode(Value="((Type=Control,Name=\"dyn_hair_1_c_ctrl\"),(Type=Control,Name=\"dyn_hair_2_c_ctrl\"),(Type=Control,Name=\"dyn_hair_3_c_ctrl\"),(Type=Control,Name=\"dyn_hair_4_c_ctrl\"),(Type=Control,Name=\"dyn_hair_5_c_ctrl\"),(Type=Control,Name=\"dyn_hair_6_c_ctrl\"),(Type=Control,N ...(truncated)") {template="DISPATCH_RigVMDispatch_Constant(io Value)"}
   joints_ik_c_f_x7 = RigVMDispatchNode(Value=((Type=Bone,Name="dyn_hair_1_c"),(Type=Bone,Name="dyn_hair_2_c"),(Type=Bone,Name="dyn_hair_3_c"),(Type=Bone,Name="dyn_hair_4_c"),(Type=Bone,Name="dyn_hair_5_c"),(Type=Bone,Name="dyn_hair_6_c"),(Type=Bone,Name="dyn_hair_7_c"))) {template="DISPATCH_RigVMDispatch_Constant(io Value)"}
   SplineIK_6 = RigVMFunctionReferenceNode(
    Primary Axis=(X=1,Y=0,Z=0)
    Samples Per Segment=1
    Secondary Spline Direction=(X=0,Y=0,Z=0)
    Spline Mode=BSpline
    Stretch=False
    Up Axis=(X=0,Y=0,Z=0)
    debug=false
    ExecuteContext=@RigVMFunction_ControlFlowBranch.True
    Controls=@controls_ik_c_x7.Value
    Bones=@joints_ik_c_f_x7.Value
   )
   controls_ik_l_f_x7 = RigVMDispatchNode(Value="((Type=Control,Name=\"dyn_hair_1_l_ctrl\"),(Type=Control,Name=\"dyn_hair_2_l_ctrl\"),(Type=Control,Name=\"dyn_hair_3_l_ctrl\"),(Type=Control,Name=\"dyn_hair_4_l_ctrl\"),(Type=Control,Name=\"dyn_hair_5_l_ctrl\"),(Type=Control,Name=\"dyn_hair_6_l_ctrl\"),(Type=Control,N ...(truncated)") {template="DISPATCH_RigVMDispatch_Constant(io Value)"}
   joints_ik_l_f_x7 = RigVMDispatchNode(Value=((Type=Bone,Name="dyn_hair_1_l"),(Type=Bone,Name="dyn_hair_2_l"),(Type=Bone,Name="dyn_hair_3_l"),(Type=Bone,Name="dyn_hair_4_l"),(Type=Bone,Name="dyn_hair_5_l"),(Type=Bone,Name="dyn_hair_6_l"),(Type=Bone,Name="dyn_hair_7_l"))) {template="DISPATCH_RigVMDispatch_Constant(io Value)"}
   > SplineIK_7 = RigVMFunctionReferenceNode(
    Primary Axis=(X=1,Y=0,Z=0)
    Samples Per Segment=1
    Secondary Spline Direction=(X=0,Y=0,Z=0)
    Spline Mode=BSpline
    Stretch=False
    Up Axis=(X=0,Y=0,Z=0)
    debug=false
    Controls=@controls_ik_l_f_x7.Value
    Bones=@joints_ik_l_f_x7.Value
   )
   controls_ik_r_x7 = RigVMDispatchNode(Value="((Type=Control,Name=\"dyn_hair_1_r_ctrl\"),(Type=Control,Name=\"dyn_hair_2_r_ctrl\"),(Type=Control,Name=\"dyn_hair_3_r_ctrl\"),(Type=Control,Name=\"dyn_hair_4_r_ctrl\"),(Type=Control,Name=\"dyn_hair_5_r_ctrl\"),(Type=Control,Name=\"dyn_hair_6_r_ctrl\"),(Type=Control,N ...(truncated)") {template="DISPATCH_RigVMDispatch_Constant(io Value)"}
   joints_ik_r_f_x7 = RigVMDispatchNode(Value=((Type=Bone,Name="dyn_hair_1_r"),(Type=Bone,Name="dyn_hair_2_r"),(Type=Bone,Name="dyn_hair_3_r"),(Type=Bone,Name="dyn_hair_4_r"),(Type=Bone,Name="dyn_hair_5_r"),(Type=Bone,Name="dyn_hair_6_r"),(Type=Bone,Name="dyn_hair_7_r"))) {template="DISPATCH_RigVMDispatch_Constant(io Value)"}
   > SplineIK_8 = RigVMFunctionReferenceNode(
    Primary Axis=(X=1,Y=0,Z=0)
    Samples Per Segment=1
    Secondary Spline Direction=(X=0,Y=0,Z=0)
    Spline Mode=BSpline
    Stretch=False
    Up Axis=(X=0,Y=0,Z=0)
    debug=false
    Controls=@controls_ik_r_x7.Value
    Bones=@joints_ik_r_f_x7.Value
   )
   RigVMFunction_Sequence_1 = RigVMAggregateNode(ExecuteContext=@RigVMFunction_Sequence.A)
   n3 = &RigUnit_SphericalPoseReader(
    ActiveRegionScaleFactors=(PositiveWidth=1,NegativeWidth=1,PositiveHeight=1,NegativeHeight=1)
    ActiveRegionSize=0.05
    Debug=(bDrawDebug=false,bDraw2D=False,bDrawLocalAxes=False,DebugScale=25,DebugSegments=20,DebugThickness=0.25)
    DriverAxis=(X=1,Y=0,Z=0)
    DriverItem=(Type=Bone,Name="head")
    FalloffRegionScaleFactors=(PositiveWidth=1,NegativeWidth=1,PositiveHeight=1,NegativeHeight=1)
    FalloffSize=0.2
    FlipHeightScaling=False
    FlipWidthScaling=False
    OptionalParentItem=(Type=Bone,Name="spine_05")
    RotationOffset=(X=0,Y=-62,Z=-90)
    ExecutePin=@RigVMFunction_Sequence_1.A
   )
   n4 = &RigVMFunction_MathDoubleMul(A=@n3.OutputParam, B=RigVMVariableNode() {variable=isTargeting lerp}.Value)
   > RigVMVariableNode(Value=@n4.Result) {variable=isPistolASDUp Extreme}
   > &RigUnit_ModifyTransforms(
    ItemToModify="((Item=(Type=Null,Name=\"dyn_hair_1_r_ctrl_null\"),Transform=(Rotation=(X=0,Y=0,Z=0,W=1),Translation=(X=0,Y=0,Z=0),Scale3D=(X=1,Y=1,Z=1))),(Item=(Type=Null,Name=\"dyn_hair_3_r_ctrl_null\"),Transform=(Rotation=(X=0,Y=0,Z=0,W=1),Translation=(X=-2,Y=0,Z=2.5),Scal ...(truncated)"
    WeightMinimum=0
    Weight=&RigVMFunction_MathDoubleMul(A=@n4.Result, B=RigVMVariableNode() {variable=isPistol Equipped lerp}.Value).Result
   )
   DeltaRotation v4 = RigVMFunctionReferenceNode(
    Rotation Order=ZXY
    child=(Type=Bone,Name="head")
    debugRotation=false
    outputMinMax=(X=0,Y=1)
    parent=(Type=Null,Name="head_null")
    premult=1
    rotationMinMax=(X=0,Y=-60)
    triggerAxis=Z
    twistAxis=(X=1,Y=0,Z=0)
    ExecuteContext=@RigVMFunction_Sequence_1.B
   )
   Remap = &RigVMFunction_MathDoubleRemap(SourceMaximum=1, SourceMinimum=0, TargetMaximum=0, TargetMinimum=1, Value=RigVMVariableNode() {variable=isPistolASDUp Extreme}.Value)
   n5 = &RigVMFunction_MathDoubleMul(B=@Remap.Result, A=@"DeltaRotation v4".outputWeight)
   Multiply = &RigVMFunction_MathDoubleMul(B=0.5, A=@n5.Result)
   > &RigUnit_ModifyTransforms(
    ItemToModify="((Item=(Type=Null,Name=\"dyn_hair_1_l_ctrl_null\"),Transform=(Rotation=(X=0,Y=0,Z=0,W=1),Translation=(X=0,Y=0,Z=5),Scale3D=(X=1,Y=1,Z=1))),(Item=(Type=Null,Name=\"dyn_hair_1_r_ctrl_null\"),Transform=(Rotation=(X=0,Y=0,Z=0,W=1),Translation=(X=0,Y=0,Z=5),Scale3D ...(truncated)"
    WeightMinimum=0
    Weight=@Multiply.Result
   )
   > &RigUnit_ModifyTransforms(
    ItemToModify="((Item=(Type=Null,Name=\"dyn_hair_3_l_ctrl_null\"),Transform=(Rotation=(X=0,Y=0,Z=0,W=1),Translation=(X=1,Y=0,Z=7),Scale3D=(X=1,Y=1,Z=1))),(Item=(Type=Null,Name=\"dyn_hair_3_r_ctrl_null\"),Transform=(Rotation=(X=0,Y=0,Z=0,W=1),Translation=(X=1,Y=0,Z=7),Scale3D ...(truncated)"
    WeightMinimum=0
    Weight=&RigVMFunction_MathDoubleMul(B=1, A=@n5.Result).Result
   )
   > DeltaRotation v4_3 = RigVMFunctionReferenceNode(
    Rotation Order=ZYX
    child=(Type=Bone,Name="head")
    debugRotation=false
    onlyTwist=true
    outputMinMax=(X=0,Y=1)
    parent=(Type=Null,Name="head_null")
    premult=1
    rotationMinMax=(X=0,Y=-60)
    twistAxis=(X=1,Y=0,Z=0)
   )
   > &RigUnit_ModifyTransforms(
    ItemToModify="((Item=(Type=Null,Name=\"dyn_hair_2_r_ctrl_null\"),Transform=(Rotation=(X=0,Y=0,Z=0,W=-1),Translation=(X=-1.3116,Y=0,Z=-3.1247),Scale3D=(X=1,Y=1,Z=1))),(Item=(Type=Null,Name=\"dyn_hair_3_r_ctrl_null\"),Transform=(Rotation=(X=0,Y=0,Z=0,W=-1),Translation=(X=0,Y= ...(truncated)"
    WeightMinimum=0
    Weight=@"DeltaRotation v4_3".outputWeight
   )
   > DeltaRotation v4_5 = RigVMFunctionReferenceNode(
    Rotation Order=ZYX
    child=(Type=Bone,Name="head")
    debugRotation=false
    onlyTwist=true
    outputMinMax=(X=0,Y=1)
    parent=(Type=Null,Name="head_null")
    premult=1
    rotationMinMax=(X=0,Y=60)
    twistAxis=(X=1,Y=0,Z=0)
   )
   > &RigUnit_ModifyTransforms(
    ItemToModify="((Item=(Type=Null,Name=\"dyn_hair_1_r_ctrl_null\"),Transform=(Rotation=(X=0,Y=0,Z=0,W=1),Translation=(X=-0.2341,Y=-3.6072,Z=-0.9853),Scale3D=(X=1,Y=1,Z=1))),(Item=(Type=Null,Name=\"dyn_hair_2_r_ctrl_null\"),Transform=(Rotation=(X=0,Y=0,Z=0,W=-1),Translation=(X ...(truncated)"
    WeightMinimum=0
    Weight=@"DeltaRotation v4_5".outputWeight
   )
   > DeltaRotation v4_7 = RigVMFunctionReferenceNode(
    Rotation Order=ZYX
    child=(Type=Bone,Name="head")
    debugRotation=false
    outputMinMax=(X=0,Y=1)
    parent=(Type=Null,Name="head_null")
    premult=1
    rotationMinMax=(X=-60,Y=-80)
    triggerAxis=Z
    twistAxis=(X=1,Y=0,Z=0)
   )
   > RigVMVariableNode(Value=@"DeltaRotation v4_7".outputWeight) {variable=head rz n60n80 alpha}
   > &RigUnit_ModifyTransforms(
    ItemToModify="((Item=(Type=Null,Name=\"dyn_hair_3_r_ctrl_null\"),Transform=(Rotation=(X=0,Y=0,Z=0,W=-1),Translation=(X=2.2209,Y=-3.1709,Z=-5.7956),Scale3D=(X=1,Y=1,Z=1))),(Item=(Type=Null,Name=\"dyn_hair_2_l_ctrl_null\"),Transform=(Rotation=(X=0,Y=0,Z=0,W=-1),Translation=(X ...(truncated)"
    WeightMinimum=0
    Weight=@"DeltaRotation v4_7".outputWeight
   )
   n6 = &RigVMFunction_AnimEvalRichCurve(
    Curve=(EditorCurveData=(Keys=((InterpMode=RCIM_Cubic),(InterpMode=RCIM_Cubic,Time=0.5,Value=1),(Time=0.8))))
    SourceMaximum=2
    SourceMinimum=0
    TargetMinimum=0
    Value=&RigVMFunction_MathDoubleRemap(
     SourceMaximum=75
     SourceMinimum=45
     TargetMaximum=1
     TargetMinimum=0
     bClamp=false
     Value=&RigUnit_GetRelativeTransformForItem(Child=(Type=Bone,Name="pelvis"), Parent=(Type=Bone,Name="root"), bChildInitial=False, bParentInitial=False).RelativeTransform.Translation.Z
    ).Result
   )
   > &RigUnit_ModifyTransforms(
    ItemToModify="((Item=(Type=Null,Name=\"dyn_hair_1_r_ctrl_null\"),Transform=(Rotation=(X=0,Y=0,Z=0,W=1),Translation=(X=-2,Y=1.5531,Z=-1.7685),Scale3D=(X=1,Y=1,Z=1))),(Item=(Type=Null,Name=\"dyn_hair_1_c_ctrl_null\"),Transform=(Rotation=(X=0,Y=0,Z=0,W=-1),Translation=(X=-5.56 ...(truncated)"
    WeightMinimum=0
    Weight=&RigVMFunction_MathDoubleMul(B=@n6.Result, A=RigVMVariableNode() {variable=isPistolASDUp Extreme}.Value).Result
   )
   DeltaRotation v4_1 = RigVMFunctionReferenceNode(
    Rotation Order=ZYX
    child=(Type=Bone,Name="spine_05")
    debugRotation=false
    onlyTwist=true
    outputMinMax=(X=0,Y=1)
    parent=(Type=Bone,Name="spine_03")
    premult=1
    rotationMinMax=(X=0,Y=35)
    twistAxis=(X=1,Y=0,Z=0)
    ExecuteContext=@RigVMFunction_Sequence_1.C
   )
   > &RigUnit_ModifyTransforms(
    ItemToModify="((Item=(Type=Null,Name=\"dyn_hair_4_r_ctrl_null\"),Transform=(Rotation=(X=0,Y=0,Z=0,W=1),Translation=(X=0,Y=-2.112,Z=3.5606),Scale3D=(X=1,Y=1,Z=1))),(Item=(Type=Null,Name=\"dyn_hair_5_r_ctrl_null\"),Transform=(Rotation=(X=0,Y=0,Z=0,W=-1),Translation=(X=0,Y=0,Z ...(truncated)"
    WeightMinimum=0
    Weight=@"DeltaRotation v4_1".outputWeight
   )
   > &RigUnit_ModifyTransforms(
    ItemToModify="((Item=(Type=Null,Name=\"dyn_hair_3_r_ctrl_null\"),Transform=(Rotation=(X=0,Y=0,Z=0,W=1),Translation=(X=0,Y=0,Z=3),Scale3D=(X=1,Y=1,Z=1))),(Item=(Type=Null,Name=\"dyn_hair_2_r_ctrl_null\"),Transform=(Rotation=(X=0,Y=0,Z=0,W=1),Translation=(X=0,Y=0,Z=2.0512),Sc ...(truncated)"
    WeightMinimum=0
    Weight=@"DeltaRotation v4_1".outputWeight
   )
   > DeltaRotation v4_2 = RigVMFunctionReferenceNode(
    Rotation Order=ZYX
    child=(Type=Bone,Name="spine_05")
    debugRotation=false
    onlyTwist=true
    outputMinMax=(X=0,Y=1)
    parent=(Type=Bone,Name="spine_03")
    premult=0
    rotationMinMax=(X=0,Y=-35)
    twistAxis=(X=1,Y=0,Z=0)
   )
   > &RigUnit_ModifyTransforms(
    ItemToModify="((Item=(Type=Null,Name=\"dyn_hair_4_l_ctrl_null\"),Transform=(Rotation=(X=0,Y=0,Z=0,W=1),Translation=(X=0,Y=-2.112,Z=3.5606),Scale3D=(X=1,Y=1,Z=1))),(Item=(Type=Null,Name=\"dyn_hair_5_l_ctrl_null\"),Transform=(Rotation=(X=0,Y=0,Z=0,W=-1),Translation=(X=0,Y=0,Z ...(truncated)"
    WeightMinimum=0
    Weight=@"DeltaRotation v4_2".outputWeight
   )
   > n7 = &RigUnit_SphericalPoseReader(
    ActiveRegionScaleFactors=(PositiveWidth=0.5,NegativeWidth=0.5,PositiveHeight=0.5,NegativeHeight=0.5)
    ActiveRegionSize=0.12
    Debug=(bDrawDebug=false,bDraw2D=False,bDrawLocalAxes=False,DebugScale=25,DebugSegments=20,DebugThickness=0.25)
    DriverAxis=(X=1,Y=0,Z=0)
    DriverItem=(Type=Bone,Name="clavicle_l")
    FalloffRegionScaleFactors=(PositiveWidth=0.5,NegativeWidth=0.5,PositiveHeight=0.5,NegativeHeight=0.5)
    FalloffSize=0.3
    FlipHeightScaling=False
    FlipWidthScaling=False
    OptionalParentItem=(Type=Bone,Name="None")
    RotationOffset=(X=0,Y=90,Z=140)
   )
   > &RigUnit_ModifyTransforms(
    ItemToModify="((Item=(Type=Null,Name=\"dyn_hair_3_l_ctrl_null\"),Transform=(Rotation=(X=0,Y=0,Z=0,W=-1),Translation=(X=0,Y=3,Z=2.1),Scale3D=(X=1,Y=1,Z=1))),(Item=(Type=Null,Name=\"dyn_hair_4_l_ctrl_null\"),Transform=(Rotation=(X=0,Y=0,Z=0,W=-1),Translation=(X=0,Y=3.8,Z=3.2) ...(truncated)"
    WeightMinimum=0
    Weight=@n7.OutputParam
   )
   > n8 = &RigUnit_SphericalPoseReader(
    ActiveRegionScaleFactors=(PositiveWidth=0.5,NegativeWidth=0.5,PositiveHeight=0.5,NegativeHeight=0.5)
    ActiveRegionSize=0.12
    Debug=(bDrawDebug=false,bDraw2D=False,bDrawLocalAxes=False,DebugScale=25,DebugSegments=20,DebugThickness=0.25)
    DriverAxis=(X=-1,Y=0,Z=0)
    DriverItem=(Type=Bone,Name="clavicle_r")
    FalloffRegionScaleFactors=(PositiveWidth=0.5,NegativeWidth=0.5,PositiveHeight=0.5,NegativeHeight=0.5)
    FalloffSize=0.3
    FlipHeightScaling=False
    FlipWidthScaling=False
    OptionalParentItem=(Type=Bone,Name="None")
    RotationOffset=(X=0,Y=-90,Z=140)
   )
   > &RigUnit_ModifyTransforms(
    ItemToModify=((Item=(Type=Null,Name="dyn_hair_3_r_ctrl_null"),Transform=(Translation=(X=0,Y=-3,Z=2.1),Scale3D=(X=1,Y=1,Z=1))),(Item=(Type=Null,Name="dyn_hair_4_r_ctrl_null"),Transform=(Rotation=(X=0,Y=0,Z=0,W=1),Translation=(X=0,Y=-3.8,Z=3.2),Scale3D=(X=1,Y=1,Z=1))))
    WeightMinimum=0
    Weight=@n8.OutputParam
   )
   > n9 = &RigUnit_SphericalPoseReader(
    ActiveRegionScaleFactors=(PositiveWidth=1,NegativeWidth=1,PositiveHeight=1,NegativeHeight=1)
    ActiveRegionSize=0.05
    Debug=(bDrawDebug=false,bDraw2D=False,bDrawLocalAxes=False,DebugScale=25,DebugSegments=20,DebugThickness=0.25)
    DriverAxis=(X=1,Y=0,Z=0)
    DriverItem=(Type=Bone,Name="head")
    FalloffRegionScaleFactors=(PositiveWidth=1,NegativeWidth=1,PositiveHeight=1,NegativeHeight=1)
    FalloffSize=0.2
    FlipHeightScaling=False
    FlipWidthScaling=False
    OptionalParentItem=(Type=Bone,Name="spine_05")
    RotationOffset=(X=0,Y=-50,Z=-15)
   )
   n10 = &RigVMFunction_MathDoubleMul(
    A=&RigVMFunction_MathDoubleMul(A=@n9.OutputParam, B=RigVMVariableNode() {variable=isTargeting lerp}.Value).Result
    B=RigVMVariableNode() {variable=isAR Equipped lerp}.Value
   )
   > &RigUnit_ModifyTransforms(
    ItemToModify="((Item=(Type=Null,Name=\"dyn_hair_2_r_ctrl_null\"),Transform=(Rotation=(X=0,Y=0,Z=0,W=1),Translation=(X=-5,Y=-7.478,Z=-4),Scale3D=(X=1,Y=1,Z=1))),(Item=(Type=Null,Name=\"dyn_hair_3_r_ctrl_null\"),Transform=(Rotation=(X=0,Y=0,Z=0,W=1),Translation=(X=-1.6185,Y=- ...(truncated)"
    WeightMinimum=0
    Weight=@n10.Result
   )
   > &RigUnit_ModifyTransforms(
    ItemToModify="((Item=(Type=Bone,Name=\"base_hair_1_r\"),Transform=(Rotation=(X=0,Y=-0.766,Z=0,W=0.6428),Translation=(X=0,Y=0,Z=0),Scale3D=(X=1,Y=1,Z=1))),(Item=(Type=Bone,Name=\"base_hair_1_l\"),Transform=(Rotation=(X=0,Y=-0.766,Z=0,W=0.6428),Translation=(X=0,Y=0,Z=0),Scale ...(truncated)"
    WeightMinimum=0
    Weight=@n10.Result
   )
   > n11 = &RigUnit_SphericalPoseReader(
    ActiveRegionScaleFactors=(PositiveWidth=1,NegativeWidth=1,PositiveHeight=1,NegativeHeight=1)
    ActiveRegionSize=0.05
    Debug=(bDrawDebug=false,bDraw2D=False,bDrawLocalAxes=False,DebugScale=25,DebugSegments=20,DebugThickness=0.25)
    DriverAxis=(X=1,Y=0,Z=0)
    DriverItem=(Type=Bone,Name="head")
    FalloffRegionScaleFactors=(PositiveWidth=1,NegativeWidth=1,PositiveHeight=1,NegativeHeight=1)
    FalloffSize=0.4
    FlipHeightScaling=False
    FlipWidthScaling=False
    OptionalParentItem=(Type=Bone,Name="spine_05")
    RotationOffset=(X=0,Y=-65,Z=-52)
   )
   > &RigUnit_ModifyTransforms(
    ItemToModify="((Item=(Type=Null,Name=\"dyn_hair_1_r_ctrl_null\"),Transform=(Rotation=(X=0,Y=0,Z=0,W=-1),Translation=(X=-3.6768,Y=0.046,Z=0.0676),Scale3D=(X=1,Y=1,Z=1))),(Item=(Type=Null,Name=\"dyn_hair_2_r_ctrl_null\"),Transform=(Rotation=(X=0,Y=0,Z=0,W=1),Translation=(X=0, ...(truncated)"
    WeightMinimum=0
    Weight=@n11.OutputParam
   )
   n12 = RigVMVariableNode() {variable=isTargeting lerp}
   > &RigUnit_ModifyTransforms(
    ItemToModify="((Item=(Type=Null,Name=\"dyn_hair_2_l_ctrl_null\"),Transform=(Rotation=(X=0,Y=0,Z=0,W=1),Translation=(X=1,Y=0,Z=3),Scale3D=(X=1,Y=1,Z=1))),(Item=(Type=Null,Name=\"dyn_hair_1_r_ctrl_null\"),Transform=(Rotation=(X=0,Y=0,Z=0,W=1),Translation=(X=0,Y=-0.5,Z=1),Scal ...(truncated)"
    WeightMinimum=0
    Weight=&RigVMFunction_MathDoubleMul(
     A=&RigVMFunction_MathDoubleMul(A=@n11.OutputParam, B=@n12.Value).Result
     B=RigVMVariableNode() {variable=isPistol Equipped lerp}.Value
    ).Result
   )
   > n13 = &RigUnit_SphericalPoseReader(
    ActiveRegionScaleFactors=(PositiveWidth=0.5,NegativeWidth=1,PositiveHeight=1,NegativeHeight=1)
    ActiveRegionSize=0.1
    Debug=(bDrawDebug=false,bDraw2D=False,bDrawLocalAxes=False,DebugScale=25,DebugSegments=20,DebugThickness=0.25)
    DriverAxis=(X=1,Y=0,Z=0)
    DriverItem=(Type=Bone,Name="clavicle_l")
    FalloffRegionScaleFactors=(PositiveWidth=0.5,NegativeWidth=1,PositiveHeight=1,NegativeHeight=1)
    FalloffSize=0.2
    FlipHeightScaling=False
    FlipWidthScaling=False
    OptionalParentItem=(Type=Bone,Name="None")
    RotationOffset=(X=-20,Y=-50,Z=0)
   )
   > &RigUnit_ModifyTransforms(
    ItemToModify="((Item=(Type=Null,Name=\"dyn_hair_2_l_ctrl_null\"),Transform=(Rotation=(X=0,Y=0,Z=0,W=1),Translation=(X=0,Y=-6.5129,Z=17.4268),Scale3D=(X=1,Y=1,Z=1))),(Item=(Type=Null,Name=\"dyn_hair_3_l_ctrl_null\"),Transform=(Rotation=(X=0,Y=0,Z=0,W=1),Translation=(X=-2.002 ...(truncated)"
    WeightMinimum=0
    Weight=&RigVMFunction_MathDoubleMul(A=@n13.OutputParam, B=RigVMVariableNode() {variable=isPASwing lerp}.Value).Result
   )
   n14 = &RigVMFunction_Sequence(ExecuteContext=@RigVMFunction_Sequence.D)
   n15 = &RigVMFunction_ControlFlowBranch(ExecuteContext=@n14.A, Condition=RigVMVariableNode() {variable=pub isRigidBodies Enabled}.Value)
   DeltaRotation v4_14 = RigVMFunctionReferenceNode(
    Rotation Order=ZYX
    child=(Type=Bone,Name="head")
    debugRotation=false
    outputMinMax=(X=0,Y=1)
    parent=(Type=Null,Name="head_null")
    premult=1
    rotationMinMax=(X=20,Y=-5)
    triggerAxis=Z
    twistAxis=(X=1,Y=0,Z=0)
    ExecuteContext=@n15.False
   )
   > &RigUnit_ModifyTransforms(
    ItemToModify=((Item=(Type=Bone,Name="hair_side_r_01_base"),Transform=(Rotation=(X=0,Y=0.1736,Z=0,W=0.9848),Translation=(X=0,Y=0,Z=0),Scale3D=(X=1,Y=1,Z=1))))
    WeightMinimum=0
    Weight=@"DeltaRotation v4_14".outputWeight
   )
   > &RigUnit_ParentConstraint(
    AdvancedSettings=(InterpolationType=Average,RotationOrderForFilter=XZY)
    Child=(Type=Bone,Name="hair_side_r_02_dyn")
    Parents=((Item=(Type=Bone,Name="spine_05"),Weight=0.9),(Item=(Type=Bone,Name="head"),Weight=0.1))
    Weight=1
   )
   > &RigUnit_ModifyTransforms(
    ItemToModify=((Item=(Type=Bone,Name="hair_side_l_01_base"),Transform=(Rotation=(X=0,Y=0.2164,Z=0,W=0.9763),Translation=(X=0,Y=0,Z=0),Scale3D=(X=1,Y=1,Z=1))))
    WeightMinimum=0
    Weight=@"DeltaRotation v4_14".outputWeight
   )
   > &RigUnit_ParentConstraint(
    AdvancedSettings=(InterpolationType=Average,RotationOrderForFilter=XZY)
    Child=(Type=Bone,Name="hair_side_l_02_dyn")
    Parents=((Item=(Type=Bone,Name="clavicle_l"),Weight=0.9),(Item=(Type=Bone,Name="head"),Weight=0.1))
    Weight=1
   )
   > &RigUnit_ModifyTransforms(
    ItemToModify="((Item=(Type=Bone,Name=\"None\"),Transform=(Rotation=(X=0,Y=0,Z=0,W=1),Translation=(X=0,Y=0,Z=0),Scale3D=(X=1,Y=1,Z=1))),(Item=(Type=Bone,Name=\"hair_side_l_01_base\"),Transform=(Rotation=(X=0,Y=0.1045,Z=0,W=0.9945),Translation=(X=0,Y=0,Z=0),Scale3D=(X=1,Y=1,Z ...(truncated)"
    WeightMinimum=0
    Weight=RigVMVariableNode() {variable=looking_left}.Value
   )
   DeltaRotation v4_6 = RigVMFunctionReferenceNode(
    Rotation Order=ZYX
    child=(Type=Bone,Name="head")
    debugRotation=false
    outputMinMax=(X=0,Y=1)
    parent=(Type=Null,Name="head_null")
    premult=1
    rotationMinMax=(X=20,Y=-5)
    triggerAxis=Z
    twistAxis=(X=1,Y=0,Z=0)
    ExecuteContext=@n15.True
   )
   > DeltaRotation v4_9 = RigVMFunctionReferenceNode(
    Rotation Order=ZYX
    child=(Type=Bone,Name="head")
    debugRotation=false
    outputMinMax=(X=0,Y=1)
    parent=(Type=Null,Name="head_null")
    premult=1
    rotationMinMax=(X=-15,Y=40)
    triggerAxis=Z
    twistAxis=(X=1,Y=0,Z=0)
   )
   > &RigUnit_ModifyTransforms(
    ItemToModify="((Item=(Type=Bone,Name=\"None\"),Transform=(Rotation=(X=0,Y=0,Z=0,W=1),Translation=(X=0,Y=0,Z=0),Scale3D=(X=1,Y=1,Z=1))),(Item=(Type=Bone,Name=\"hair_side_l_01_base\"),Transform=(Rotation=(X=0.0145,Y=0,Z=0,W=0.9999),Translation=(X=-0.2328,Y=1.925,Z=-0.2327),Sc ...(truncated)"
    WeightMinimum=0
    Weight=@"DeltaRotation v4_9".outputWeight
   )
   n16 = &RigVMFunction_MathDoubleMul(A=@"DeltaRotation v4_6".outputWeight, B=RigVMVariableNode() {variable=isTargeting lerp}.Value)
   n17 = RigVMVariableNode() {variable=isPistol Equipped lerp}
   > &RigUnit_ModifyTransforms(
    ItemToModify="((Item=(Type=Bone,Name=\"hair_side_l_01_base\"),Transform=(Rotation=(X=0.0151,Y=0.173,Z=-0.0858,W=0.9811),Translation=(X=0,Y=0,Z=0),Scale3D=(X=1,Y=1,Z=1))),(Item=(Type=Bone,Name=\"hair_side_r_01_base\"),Transform=(Rotation=(X=-0.0076,Y=0.1735,Z=0.043,W=0.9839) ...(truncated)"
    WeightMinimum=0
    Weight=&RigVMFunction_MathDoubleRemap(
     SourceMaximum=1
     SourceMinimum=0
     TargetMaximum=0
     TargetMinimum=1
     Value=&RigVMFunction_MathDoubleMul(A=@n16.Result, B=@n17.Value).Result
    ).Result
   )
   &RigUnit_ModifyTransforms(
    ItemToModify="((Item=(Type=Bone,Name=\"dyn_hair_1_l\"),Transform=(Rotation=(X=0,Y=0,Z=0,W=1),Translation=(X=-3,Y=0,Z=2),Scale3D=(X=1,Y=1,Z=1))),(Item=(Type=Bone,Name=\"dyn_hair_2_l\"),Transform=(Rotation=(X=0,Y=0,Z=0,W=1),Translation=(X=-5,Y=0,Z=0),Scale3D=(X=1,Y=1,Z=1))),( ...(truncated)"
    WeightMinimum=0
    ExecutePin=@n14.B
    Weight=RigVMVariableNode() {variable=is Parachuting}.Value
   )
   n18 = &RigVMFunction_AlphaInterp(
    InterpSpeedDecreasing=4
    InterpSpeedIncreasing=4
    Scale=1
    bClampResult=False
    Value=&RigVMFunction_MathDoubleMul(A=RigVMVariableNode() {variable=is Sky Diving Floating}.Value, B=RigVMVariableNode() {variable=Is Moving Right}.Value).Result
   )
   > &RigUnit_ModifyTransforms(
    ItemToModify="((Item=(Type=Bone,Name=\"None\"),Transform=(Rotation=(X=0,Y=0,Z=0,W=1),Translation=(X=0,Y=0,Z=0),Scale3D=(X=1,Y=1,Z=1))),(Item=(Type=Bone,Name=\"base_hair_1_l\"),Transform=(Rotation=(X=0,Y=0,Z=0.2588,W=0.9659),Translation=(X=-2,Y=1,Z=0),Scale3D=(X=1,Y=1,Z=1))) ...(truncated)"
    WeightMinimum=0
    Weight=@n18.Result
   )
   AlterTransform = RigVMFunctionReferenceNode(
    inheritsTransform=false
    transforms="((Item=(Type=Bone,Name=\"dyn_hair_1_r\"),Transform=(Rotation=(X=0,Y=0,Z=0,W=1),Translation=(X=0,Y=0,Z=0),Scale3D=(X=1,Y=1,Z=1))),(Item=(Type=Bone,Name=\"dyn_hair_1_l\"),Transform=(Rotation=(X=0,Y=0,Z=0,W=1),Translation=(X=2,Y=0,Z=8),Scale3D=(X=1,Y=1,Z=1))),(It ...(truncated)"
    weights=RigVMVariableNode() {variable=head rz n60n80 alpha}.Value
    ExecuteContext=@RigVMFunction_Sequence.C
   )
   > DeltaRotation v4_4 = RigVMFunctionReferenceNode(
    Rotation Order=ZYX
    child=(Type=Bone,Name="head")
    debugRotation=false
    outputMinMax=(X=0,Y=1)
    parent=(Type=Null,Name="head_null")
    premult=1
    rotationMinMax=(X=-30,Y=-45)
    triggerAxis=Z
    twistAxis=(X=1,Y=0,Z=0)
   )
   > AlterTransform_1 = RigVMFunctionReferenceNode(
    inheritsTransform=false
    transforms="((Item=(Type=Bone,Name=\"dyn_hair_1_r\"),Transform=(Rotation=(X=0,Y=0,Z=0,W=1),Translation=(X=1,Y=0,Z=2),Scale3D=(X=1,Y=1,Z=1))),(Item=(Type=Bone,Name=\"dyn_hair_1_l\"),Transform=(Rotation=(X=0,Y=0,Z=0,W=1),Translation=(X=1,Y=0,Z=2),Scale3D=(X=1,Y=1,Z=1))),(It ...(truncated)"
    weights=@"DeltaRotation v4_4".outputWeight
   )
   > n19 = &RigUnit_SphericalPoseReader(
    ActiveRegionScaleFactors=(PositiveWidth=1,NegativeWidth=1,PositiveHeight=1,NegativeHeight=1)
    ActiveRegionSize=0.05
    Debug=(bDrawDebug=false,bDraw2D=False,bDrawLocalAxes=False,DebugScale=25,DebugSegments=20,DebugThickness=0.25)
    DriverAxis=(X=1,Y=0,Z=0)
    DriverItem=(Type=Bone,Name="head")
    FalloffRegionScaleFactors=(PositiveWidth=1,NegativeWidth=1,PositiveHeight=1,NegativeHeight=1)
    FalloffSize=0.2
    FlipHeightScaling=False
    FlipWidthScaling=False
    OptionalParentItem=(Type=Bone,Name="spine_05")
    RotationOffset=(X=0,Y=-62,Z=-89)
   )
   > AlterTransform_2 = RigVMFunctionReferenceNode(
    inheritsTransform=false
    transforms="((Item=(Type=Bone,Name=\"dyn_hair_1_l\"),Transform=(Rotation=(X=0,Y=0,Z=0,W=1),Translation=(X=8,Y=0,Z=0),Scale3D=(X=1,Y=1,Z=1))),(Item=(Type=Bone,Name=\"dyn_hair_1_c\"),Transform=(Rotation=(X=0,Y=0,Z=0,W=1),Translation=(X=8,Y=0,Z=0),Scale3D=(X=1,Y=1,Z=1))),(It ...(truncated)"
    weights=&RigVMFunction_MathDoubleMul(A=@n19.OutputParam, B=RigVMVariableNode() {variable=isTargeting lerp}.Value).Result
   )
   &RigUnit_ModifyTransforms(
    ExecutePin=()
    ItemToModify="((Item=(Type=Null,Name=\"dyn_hair_2_l_ctrl_null\"),Transform=(Rotation=(X=0,Y=0,Z=0,W=1),Translation=(X=2.22,Y=0,Z=8.25),Scale3D=(X=1,Y=1,Z=1))),(Item=(Type=Null,Name=\"dyn_hair_2_r_ctrl_null\"),Transform=(Rotation=(X=0,Y=0,Z=0,W=1),Translation=(X=2.22,Y=0,Z=8 ...(truncated)"
    Weight=1
    WeightMinimum=0
   )
   &RigUnit_ModifyTransforms(
    ExecutePin=()
    ItemToModify="((Item=(Type=Null,Name=\"dyn_hair_2_r_ctrl_null\"),Transform=(Rotation=(X=0,Y=0,Z=0,W=1),Translation=(X=-4.9,Y=-7.4781,Z=1.8842),Scale3D=(X=1,Y=1,Z=1))),(Item=(Type=Null,Name=\"dyn_hair_3_r_ctrl_null\"),Transform=(Rotation=(X=0,Y=0,Z=0,W=1),Translation=(X=-1.6 ...(truncated)"
    Weight=0
    WeightMinimum=0
   )
   n20 = &RigVMFunction_AlphaInterp(InterpSpeedDecreasing=5, InterpSpeedIncreasing=5, Scale=1, bClampResult=False, Value=RigVMVariableNode() {variable=isCelbration Emote}.Value)
   &RigUnit_ModifyTransforms(
    ExecutePin=()
    ItemToModify="((Item=(Type=Bone,Name=\"Back_rear_dyn\"),Transform=(Rotation=(X=0,Y=0,Z=0,W=1),Translation=(X=0,Y=0,Z=0),Scale3D=(X=1,Y=1,Z=1))),(Item=(Type=Bone,Name=\"dyn_hair_1_r\"),Transform=(Rotation=(X=0,Y=-0.2164,Z=0,W=0.9763),Translation=(X=-6.5,Y=0,Z=10),Scale3D=(X= ...(truncated)"
    WeightMinimum=0
    Weight=@n20.Result
   )
   &RigUnit_ModifyTransforms(
    ExecutePin=()
    ItemToModify="((Item=(Type=Null,Name=\"dyn_hair_1_r_ctrl_null\"),Transform=(Rotation=(X=0,Y=0,Z=0,W=1),Translation=(X=-4,Y=1.5531,Z=-1.7685),Scale3D=(X=1,Y=1,Z=1))),(Item=(Type=Null,Name=\"dyn_hair_1_c_ctrl_null\"),Transform=(Rotation=(X=0,Y=0,Z=0,W=-1),Translation=(X=-5.56 ...(truncated)"
    Weight=1
    WeightMinimum=0
   )
   &RigUnit_ModifyTransforms(
    ExecutePin=()
    ItemToModify="((Item=(Type=Null,Name=\"dyn_hair_1_r_ctrl_null\"),Transform=(Rotation=(X=0,Y=0,Z=0,W=1),Translation=(X=0,Y=0,Z=0),Scale3D=(X=1,Y=1,Z=1))),(Item=(Type=Null,Name=\"dyn_hair_3_r_ctrl_null\"),Transform=(Rotation=(X=0,Y=0,Z=0,W=1),Translation=(X=-2,Y=0,Z=0),Scale3 ...(truncated)"
    Weight=1
    WeightMinimum=0
   )
   &RigUnit_ModifyTransforms(
    ExecutePin=()
    ItemToModify="((Item=(Type=Bone,Name=\"Back_rear_dyn\"),Transform=(Rotation=(X=-0.0307,Y=-0.2843,Z=0.2958,W=0.9115),Translation=(X=5,Y=0,Z=7),Scale3D=(X=1,Y=1,Z=1))),(Item=(Type=Bone,Name=\"dyn_hair_1_r\"),Transform=(Rotation=(X=0,Y=-0.2164,Z=0,W=0.9763),Translation=(X=-6.5 ...(truncated)"
    Weight=1
    WeightMinimum=0
   )
  ]}
 CollapseNode_ContainedGraph {nodes [
   n1 = RigVMFunctionEntryNode()
   &RigUnit_ParentConstraint(
    AdvancedSettings=(RotationOrderForFilter=XZY)
    Child=(Type=Bone,Name="Back_side_r_base")
    Parents="((Item=(Type=Bone,Name=\"dyn_hair_1_l\"),Weight=0.0003),(Item=(Type=Bone,Name=\"dyn_hair_2_c\"),Weight=0.0004),(Item=(Type=Bone,Name=\"dyn_hair_4_r\"),Weight=0.0005),(Item=(Type=Bone,Name=\"dyn_hair_3_r\"),Weight=0.0034),(Item=(Type=Bone,Name=\"dyn_hair_2_r\"),Weigh ...(truncated)"
    Weight=1
    ExecutePin=@n1.ExecutePin
   )
   > &RigUnit_ParentConstraint(
    AdvancedSettings=(RotationOrderForFilter=XZY)
    Child=(Type=Bone,Name="Back_side_l_base")
    Parents="((Item=(Type=Bone,Name=\"dyn_hair_4_l\"),Weight=0.0027),(Item=(Type=Bone,Name=\"Back_side_r_dyn\"),Weight=0.0034),(Item=(Type=Bone,Name=\"dyn_hair_1_c\"),Weight=0.0086),(Item=(Type=Bone,Name=\"dyn_hair_2_c\"),Weight=0.0091),(Item=(Type=Bone,Name=\"dyn_hair_3_l\"),We ...(truncated)"
    Weight=1
   )
   > &RigUnit_ParentConstraint(
    AdvancedSettings=(RotationOrderForFilter=XZY)
    Child=(Type=Null,Name="dyn_hair_1_c_ctrl_null")
    Filter=(TranslationFilter=(bX=True,bY=True,bZ=True),RotationFilter=(bX=True,bY=True,bZ=True),ScaleFilter=(bX=False,bY=False,bZ=False))
    Parents=((Item=(Type=Bone,Name="neck_02"),Weight=0.0434),(Item=(Type=Bone,Name="head"),Weight=0.9566))
    Weight=1
   )
   > &RigUnit_ParentConstraint(
    AdvancedSettings=(RotationOrderForFilter=XZY)
    Child=(Type=Null,Name="dyn_hair_2_c_ctrl_null")
    Filter=(TranslationFilter=(bX=True,bY=True,bZ=True),RotationFilter=(bX=True,bY=True,bZ=True),ScaleFilter=(bX=False,bY=False,bZ=False))
    Parents="((Item=(Type=Bone,Name=\"clavicle_r\"),Weight=0.015),(Item=(Type=Bone,Name=\"clavicle_l\"),Weight=0.015),(Item=(Type=Bone,Name=\"spine_04\"),Weight=0.0681),(Item=(Type=Bone,Name=\"head\"),Weight=0.0921),(Item=(Type=Bone,Name=\"neck_02\"),Weight=0.1321),(Item=(Type=B ...(truncated)"
    Weight=1
   )
   > &RigUnit_ParentConstraint(
    AdvancedSettings=(RotationOrderForFilter=XZY)
    Child=(Type=Null,Name="dyn_hair_3_c_ctrl_null")
    Filter=(TranslationFilter=(bX=True,bY=True,bZ=True),RotationFilter=(bX=True,bY=True,bZ=True),ScaleFilter=(bX=False,bY=False,bZ=False))
    Parents=((Item=(Type=Bone,Name="clavicle_l"),Weight=0.0369),(Item=(Type=Bone,Name="clavicle_r"),Weight=0.0369),(Item=(Type=Bone,Name="spine_05"),Weight=0.9261))
    Weight=1
   )
   > &RigUnit_ParentConstraint(
    AdvancedSettings=(RotationOrderForFilter=XZY)
    Child=(Type=Null,Name="dyn_hair_4_c_ctrl_null")
    Filter=(TranslationFilter=(bX=True,bY=True,bZ=True),RotationFilter=(bX=True,bY=True,bZ=True),ScaleFilter=(bX=False,bY=False,bZ=False))
    Parents=((Item=(Type=Bone,Name="spine_04"),Weight=0.0415),(Item=(Type=Bone,Name="clavicle_l"),Weight=0.0758),(Item=(Type=Bone,Name="clavicle_r"),Weight=0.0758),(Item=(Type=Bone,Name="spine_05"),Weight=0.807))
    Weight=1
   )
   > &RigUnit_ParentConstraint(
    AdvancedSettings=(RotationOrderForFilter=XZY)
    Child=(Type=Null,Name="dyn_hair_5_c_ctrl_null")
    Filter=(TranslationFilter=(bX=True,bY=True,bZ=True),RotationFilter=(bX=True,bY=True,bZ=True),ScaleFilter=(bX=False,bY=False,bZ=False))
    Parents=((Item=(Type=Bone,Name="spine_03"),Weight=0.0792),(Item=(Type=Bone,Name="clavicle_r"),Weight=0.163),(Item=(Type=Bone,Name="spine_05"),Weight=0.2371),(Item=(Type=Bone,Name="spine_04"),Weight=0.5206))
    Weight=1
   )
   > &RigUnit_ParentConstraint(
    AdvancedSettings=(RotationOrderForFilter=XZY)
    Child=(Type=Null,Name="dyn_hair_6_c_ctrl_null")
    Filter=(TranslationFilter=(bX=True,bY=True,bZ=True),RotationFilter=(bX=True,bY=True,bZ=True),ScaleFilter=(bX=False,bY=False,bZ=False))
    Parents="((Item=(Type=Bone,Name=\"spine_01\"),Weight=0.0165),(Item=(Type=Bone,Name=\"clavicle_l\"),Weight=0.0199),(Item=(Type=Bone,Name=\"clavicle_r\"),Weight=0.0199),(Item=(Type=Bone,Name=\"spine_02\"),Weight=0.2113),(Item=(Type=Bone,Name=\"spine_04\"),Weight=0.2299),(Item= ...(truncated)"
    Weight=1
   )
   > &RigUnit_ParentConstraint(
    AdvancedSettings=(RotationOrderForFilter=XZY)
    Child=(Type=Null,Name="dyn_hair_7_c_ctrl_null")
    Filter=(TranslationFilter=(bX=True,bY=True,bZ=True),RotationFilter=(bX=True,bY=True,bZ=True),ScaleFilter=(bX=False,bY=False,bZ=False))
    Parents=((Item=(Type=Bone,Name="spine_03"),Weight=0.0635),(Item=(Type=Bone,Name="pelvis"),Weight=0.0679),(Item=(Type=Bone,Name="spine_02"),Weight=0.4127),(Item=(Type=Bone,Name="spine_01"),Weight=0.4559))
    Weight=1
   )
   > &RigUnit_ParentConstraint(
    AdvancedSettings=(RotationOrderForFilter=XZY)
    Child=(Type=Null,Name="dyn_hair_1_l_ctrl_null")
    Filter=(TranslationFilter=(bX=True,bY=True,bZ=True),RotationFilter=(bX=True,bY=True,bZ=True),ScaleFilter=(bX=False,bY=False,bZ=False))
    Parents=((Item=(Type=Bone,Name="head"),Weight=1.0))
    Weight=1
   )
   > &RigUnit_ParentConstraint(
    AdvancedSettings=(RotationOrderForFilter=XZY)
    Child=(Type=Null,Name="dyn_hair_2_l_ctrl_null")
    Filter=(TranslationFilter=(bX=True,bY=True,bZ=True),RotationFilter=(bX=True,bY=True,bZ=True),ScaleFilter=(bX=False,bY=False,bZ=False))
    Parents="((Item=(Type=Bone,Name=\"clavicle_l\"),Weight=0.0503),(Item=(Type=Bone,Name=\"spine_04\"),Weight=0.0613),(Item=(Type=Bone,Name=\"head\"),Weight=0.1054),(Item=(Type=Bone,Name=\"neck_02\"),Weight=0.1567),(Item=(Type=Bone,Name=\"spine_05\"),Weight=0.2205),(Item=(Type=B ...(truncated)"
    Weight=1
   )
   > &RigUnit_ParentConstraint(
    AdvancedSettings=(RotationOrderForFilter=XZY)
    Child=(Type=Null,Name="dyn_hair_3_l_ctrl_null")
    Filter=(TranslationFilter=(bX=True,bY=True,bZ=True),RotationFilter=(bX=True,bY=True,bZ=True),ScaleFilter=(bX=False,bY=False,bZ=False))
    Parents=((Item=(Type=Bone,Name="clavicle_l"),Weight=0.4559),(Item=(Type=Bone,Name="spine_05"),Weight=0.5441))
    Weight=1
   )
   > &RigUnit_ParentConstraint(
    AdvancedSettings=(RotationOrderForFilter=XZY)
    Child=(Type=Null,Name="dyn_hair_4_l_ctrl_null")
    Filter=(TranslationFilter=(bX=True,bY=True,bZ=True),RotationFilter=(bX=True,bY=True,bZ=True),ScaleFilter=(bX=False,bY=False,bZ=False))
    Parents=((Item=(Type=Bone,Name="upperarm_l"),Weight=0.0282),(Item=(Type=Bone,Name="spine_04"),Weight=0.0451),(Item=(Type=Bone,Name="spine_05"),Weight=0.377),(Item=(Type=Bone,Name="clavicle_l"),Weight=0.5498))
    Weight=1
   )
   > &RigUnit_ParentConstraint(
    AdvancedSettings=(RotationOrderForFilter=XZY)
    Child=(Type=Null,Name="dyn_hair_5_l_ctrl_null")
    Filter=(TranslationFilter=(bX=True,bY=True,bZ=True),RotationFilter=(bX=True,bY=True,bZ=True),ScaleFilter=(bX=False,bY=False,bZ=False))
    Parents=((Item=(Type=Bone,Name="upperarm_l"),Weight=0.0176),(Item=(Type=Bone,Name="spine_03"),Weight=0.0228),(Item=(Type=Bone,Name="spine_05"),Weight=0.285),(Item=(Type=Bone,Name="clavicle_l"),Weight=0.3331),(Item=(Type=Bone,Name="spine_04"),Weight=0.3415))
    Weight=1
   )
   > &RigUnit_ParentConstraint(
    AdvancedSettings=(RotationOrderForFilter=XZY)
    Child=(Type=Null,Name="dyn_hair_6_l_ctrl_null")
    Filter=(TranslationFilter=(bX=True,bY=True,bZ=True),RotationFilter=(bX=True,bY=True,bZ=True),ScaleFilter=(bX=False,bY=False,bZ=False))
    Parents=((Item=(Type=Bone,Name="spine_05"),Weight=0.0187),(Item=(Type=Bone,Name="spine_02"),Weight=0.0811),(Item=(Type=Bone,Name="clavicle_l"),Weight=0.1267),(Item=(Type=Bone,Name="spine_04"),Weight=0.3575),(Item=(Type=Bone,Name="spine_03"),Weight=0.416))
    Weight=1
   )
   > &RigUnit_ParentConstraint(
    AdvancedSettings=(RotationOrderForFilter=XZY)
    Child=(Type=Null,Name="dyn_hair_7_l_ctrl_null")
    Filter=(TranslationFilter=(bX=True,bY=True,bZ=True),RotationFilter=(bX=True,bY=True,bZ=True),ScaleFilter=(bX=False,bY=False,bZ=False))
    Parents=((Item=(Type=Bone,Name="spine_03"),Weight=0.055),(Item=(Type=Bone,Name="pelvis"),Weight=0.0574),(Item=(Type=Bone,Name="spine_02"),Weight=0.4194),(Item=(Type=Bone,Name="spine_01"),Weight=0.4683))
    Weight=1
   )
   > &RigUnit_ParentConstraint(
    AdvancedSettings=(RotationOrderForFilter=XZY)
    Child=(Type=Null,Name="dyn_hair_1_r_ctrl_null")
    Filter=(TranslationFilter=(bX=True,bY=True,bZ=True),RotationFilter=(bX=True,bY=True,bZ=True),ScaleFilter=(bX=False,bY=False,bZ=False))
    Parents=((Item=(Type=Bone,Name="head"),Weight=1.0))
    Weight=1
   )
   > &RigUnit_ParentConstraint(
    AdvancedSettings=(RotationOrderForFilter=XZY)
    Child=(Type=Null,Name="dyn_hair_2_r_ctrl_null")
    Filter=(TranslationFilter=(bX=True,bY=True,bZ=True),RotationFilter=(bX=True,bY=True,bZ=True),ScaleFilter=(bX=False,bY=False,bZ=False))
    Parents="((Item=(Type=Bone,Name=\"clavicle_r\"),Weight=0.0503),(Item=(Type=Bone,Name=\"spine_04\"),Weight=0.0613),(Item=(Type=Bone,Name=\"head\"),Weight=0.1054),(Item=(Type=Bone,Name=\"neck_02\"),Weight=0.1567),(Item=(Type=Bone,Name=\"spine_05\"),Weight=0.2204),(Item=(Type=B ...(truncated)"
    Weight=1
   )
   > &RigUnit_ParentConstraint(
    AdvancedSettings=(RotationOrderForFilter=XZY)
    Child=(Type=Null,Name="dyn_hair_3_r_ctrl_null")
    Filter=(TranslationFilter=(bX=True,bY=True,bZ=True),RotationFilter=(bX=True,bY=True,bZ=True),ScaleFilter=(bX=False,bY=False,bZ=False))
    Parents=((Item=(Type=Bone,Name="clavicle_r"),Weight=0.4559),(Item=(Type=Bone,Name="spine_05"),Weight=0.5441))
    Weight=1
   )
   > &RigUnit_ParentConstraint(
    AdvancedSettings=(RotationOrderForFilter=XZY)
    Child=(Type=Null,Name="dyn_hair_4_r_ctrl_null")
    Filter=(TranslationFilter=(bX=True,bY=True,bZ=True),RotationFilter=(bX=True,bY=True,bZ=True),ScaleFilter=(bX=False,bY=False,bZ=False))
    Parents=((Item=(Type=Bone,Name="upperarm_r"),Weight=0.0282),(Item=(Type=Bone,Name="spine_04"),Weight=0.0451),(Item=(Type=Bone,Name="spine_05"),Weight=0.377),(Item=(Type=Bone,Name="clavicle_r"),Weight=0.5498))
    Weight=1
   )
   > &RigUnit_ParentConstraint(
    AdvancedSettings=(RotationOrderForFilter=XZY)
    Child=(Type=Null,Name="dyn_hair_5_r_ctrl_null")
    Filter=(TranslationFilter=(bX=True,bY=True,bZ=True),RotationFilter=(bX=True,bY=True,bZ=True),ScaleFilter=(bX=False,bY=False,bZ=False))
    Parents=((Item=(Type=Bone,Name="upperarm_r"),Weight=0.0176),(Item=(Type=Bone,Name="spine_03"),Weight=0.0228),(Item=(Type=Bone,Name="spine_05"),Weight=0.285),(Item=(Type=Bone,Name="clavicle_r"),Weight=0.3331),(Item=(Type=Bone,Name="spine_04"),Weight=0.3415))
    Weight=1
   )
   > &RigUnit_ParentConstraint(
    AdvancedSettings=(RotationOrderForFilter=XZY)
    Child=(Type=Null,Name="dyn_hair_6_r_ctrl_null")
    Filter=(TranslationFilter=(bX=True,bY=True,bZ=True),RotationFilter=(bX=True,bY=True,bZ=True),ScaleFilter=(bX=False,bY=False,bZ=False))
    Parents=((Item=(Type=Bone,Name="spine_05"),Weight=0.0187),(Item=(Type=Bone,Name="spine_02"),Weight=0.0811),(Item=(Type=Bone,Name="clavicle_r"),Weight=0.1267),(Item=(Type=Bone,Name="spine_04"),Weight=0.3575),(Item=(Type=Bone,Name="spine_03"),Weight=0.416))
    Weight=1
   )
   > n2 = &RigUnit_ParentConstraint(
    AdvancedSettings=(RotationOrderForFilter=XZY)
    Child=(Type=Null,Name="dyn_hair_7_r_ctrl_null")
    Filter=(TranslationFilter=(bX=True,bY=True,bZ=True),RotationFilter=(bX=True,bY=True,bZ=True),ScaleFilter=(bX=False,bY=False,bZ=False))
    Parents=((Item=(Type=Bone,Name="spine_03"),Weight=0.055),(Item=(Type=Bone,Name="pelvis"),Weight=0.0574),(Item=(Type=Bone,Name="spine_02"),Weight=0.4194),(Item=(Type=Bone,Name="spine_01"),Weight=0.4683))
    Weight=1
   )
   RigVMFunctionReturnNode(ExecutePin=@n2.ExecutePin)
  ]}
 CollapseNode_ContainedGraph {nodes [
   RigVMFunctionEntryNode()
   n1 = &RigVMFunction_AlphaInterp(
    InterpSpeedDecreasing=15
    InterpSpeedIncreasing=15
    Scale=1
    bClampResult=False
    Value=&RigVMFunction_MathBoolToFloat(Value=RigVMVariableNode() {variable=pub isTargeting}.Value).Result
   )
   > RigVMVariableNode(Value=@n1.Result) {variable=isTargeting lerp}
   RigVMFunction_MathBoolOr = RigVMAggregateNode(
    C=RigVMVariableNode() {variable=SMG_P90}.Value
    A=RigVMVariableNode() {variable=pub isAR Equipped}.Value
    B=RigVMVariableNode() {variable=is_OutLaw_SG}.Value
   )
   n2 = &RigVMFunction_AlphaInterp(
    InterpSpeedDecreasing=15
    InterpSpeedIncreasing=15
    Scale=1
    bClampResult=False
    Value=&RigVMFunction_MathBoolToFloat(Value=@RigVMFunction_MathBoolOr.Result).Result
   )
   > RigVMVariableNode(Value=@n2.Result) {variable=isAR Equipped lerp}
   n3 = &RigVMFunction_AlphaInterp(
    InterpSpeedDecreasing=15
    InterpSpeedIncreasing=15
    Scale=1
    bClampResult=False
    Value=&RigVMFunction_MathBoolToFloat(Value=RigVMVariableNode() {variable=pub isPistol Equipped}.Value).Result
   )
   > RigVMVariableNode(Value=@n3.Result) {variable=isPistol Equipped lerp}
   n4 = &RigVMFunction_AlphaInterp(
    InterpSpeedDecreasing=15
    InterpSpeedIncreasing=15
    Scale=1
    bClampResult=False
    Value=&RigVMFunction_MathBoolToFloat(Value=RigVMVariableNode() {variable=pub PA Swing}.Value).Result
   )
   > RigVMVariableNode(Value=@n4.Result) {variable=isPASwing lerp}
   > RigVMFunctionReturnNode()
  ]}
 New Function_ContainedGraph {nodes [
   n1 = RigVMFunctionEntryNode()
   RigVMFunction_MathBoolToInteger = &RigVMFunction_MathBoolToInteger(Value=@n1.onlyTwist)
   RigVMFunction_MathQuaternionSwingTwist = &RigVMFunction_MathQuaternionSwingTwist(
    twistAxis=@n1.twistAxis
    Input=&RigUnit_GetRelativeTransformForItem(bChildInitial=False, bParentInitial=False, child=@n1.child, parent=@n1.parent).RelativeTransform.Rotation
   )
   n2 = RigVMDispatchNode(
    Values=((),())
    Index=@RigVMFunction_MathBoolToInteger.Result
    Values.0=@RigVMFunction_MathQuaternionSwingTwist.Swing
    Values.1=@RigVMFunction_MathQuaternionSwingTwist.Twist
   ) {template="DISPATCH_RigVMDispatch_SelectInt32(in Index,in Values,out Result)"}
   RigVMFunction_MathQuaternionToEuler = &RigVMFunction_MathQuaternionToEuler(Value=@n2.Result, RotationOrder=@n1.Rotation Order)
   > DISPATCH_RigVMDispatch_Print = RigVMDispatchNode(ScreenColor=(R=1,G=1,B=1,A=1), ScreenDuration=1.5, Enabled=@n1.debugRotation, Value=@RigVMFunction_MathQuaternionToEuler.Result) {template="DISPATCH_RigVMDispatch_Print(in Prefix,in Value,in Enabled,in ScreenDuration,in ScreenColor)"}
   Cast = RigVMDispatchNode(Value=@n1.triggerAxis) {template="DISPATCH_RigVMDispatch_CastEnumToInt(in Value,out Result)"}
   DISPATCH_RigVMDispatch_SelectInt32 = RigVMDispatchNode(
    Values=(0,0,0)
    Index=@Cast.Result
    Values.0=@RigVMFunction_MathQuaternionToEuler.Result.X
    Values.1=@RigVMFunction_MathQuaternionToEuler.Result.Y
    Values.2=@RigVMFunction_MathQuaternionToEuler.Result.Z
   ) {template="DISPATCH_RigVMDispatch_SelectInt32(in Index,in Values,out Result)"}
   Remap = &RigVMFunction_MathDoubleRemap(
    bClamp=False
    Value=@DISPATCH_RigVMDispatch_SelectInt32.Result
    TargetMinimum=@n1.outputMinMax.X
    TargetMaximum=@n1.outputMinMax.Y
    SourceMinimum=@n1.rotationMinMax.X
    SourceMaximum=@n1.rotationMinMax.Y
   )
   Absolute = &RigVMFunction_MathDoubleAbs(Value=@Remap.Result)
   Multiply = &RigVMFunction_MathDoubleMul(
    A=&RigVMFunction_MathDoubleClamp(Value=RigVMDispatchNode(True=@Absolute.Result, Condition=@n1.absoluteValue, False=@Remap.Result) {template="DISPATCH_RigVMDispatch_If(in Condition,in True,in False,out Result)"}.Result).Result
    B=@n1.premult
   )
   > RigVMFunctionReturnNode(outputQuat=@n2.Result, outputWeight=@Multiply.Result)
  ]}
 New Function_ContainedGraph {nodes [
   n1 = RigVMFunctionEntryNode()
   > For_Each = RigVMDispatchNode(Array=@n1.transforms) {template="DISPATCH_RigVMDispatch_ArrayIterator(in Array,out Element,out Index,out Count,out Ratio)"}
   > &RigUnit_OffsetTransformForItem(Item=@For_Each.Element.Item, OffsetTransform=@For_Each.Element.Transform, Weight=@n1.weights, bPropagateToChildren=@n1.inheritsTransform)
   RigVMFunctionReturnNode(ExecuteContext=@For_Each.Completed)
  ]}
 RigVMFunctionLibrary {nodes [
   AlterTransform_1 = RigVMCollapseNode(ExecuteContext=(), inheritsTransform=false, transforms=(), weights=1)
   DeltaRotation v4_1 = RigVMCollapseNode(
    ExecuteContext=()
    Rotation Order=ZYX
    child=(Type=Bone,Name="None")
    outputMinMax=(X=0,Y=1)
    parent=(Type=Bone,Name="None")
    premult=1
    rotationMinMax=(X=0,Y=0)
    twistAxis=(X=1,Y=0,Z=0)
   )
  ]}
 RigVMFunction_MathBoolOr_1_ContainedGraph {nodes [
   n1 = RigVMFunctionEntryNode()
   RigVMFunction_MathBoolOr = &RigVMFunction_MathBoolOr(A=@n1.A, B=@n1.B)
   RigVMFunctionReturnNode(Result=&RigVMFunction_MathBoolOr(B=@n1.C, A=@RigVMFunction_MathBoolOr.Result).Result)
  ]}
 RigVMFunction_Sequence_1_ContainedGraph {nodes [
   n1 = RigVMFunctionEntryNode()
   RigVMFunction_Sequence = &RigVMFunction_Sequence(ExecuteContext=@n1.ExecuteContext)
   n2 = &RigVMFunction_Sequence(ExecuteContext=@RigVMFunction_Sequence.B)
   n3 = &RigVMFunction_Sequence(ExecuteContext=@n2.B)
   n4 = &RigVMFunction_Sequence(ExecuteContext=@n3.B)
   n5 = &RigVMFunction_Sequence(ExecuteContext=@n4.B)
   RigVMFunctionReturnNode(A=@RigVMFunction_Sequence.A, B=@n2.A, C=@n3.A, D=@n4.A, E=@n5.A, F=@n5.B)
  ]}
 RigVMFunction_Sequence_2_ContainedGraph {nodes [
   n1 = RigVMFunctionEntryNode()
   n2 = &RigVMFunction_Sequence(ExecuteContext=@n1.ExecuteContext)
   n3 = &RigVMFunction_Sequence(ExecuteContext=@n2.B)
   RigVMFunctionReturnNode(A=@n2.A, B=@n3.A, C=@n3.B)
  ]}
 RigVMModel {nodes [
   RigUnit_BeginExecution = &RigUnit_BeginExecution()
   setFKCtrl = RigVMFunctionReferenceNode(setFK_Ctrl=false, ExecuteContext=@RigUnit_BeginExecution.ExecutePin)
   hair = RigVMCollapseNode(ExecutePin=@setFKCtrl.ExecuteContext)
  ]}
 RigVMModel Rig {nodes [
   n1 = &RigUnit_PrepareForExecution()
   ItemArray = RigVMDispatchNode(Value=((Type=Bone,Name="head"))) {template="DISPATCH_RigVMDispatch_Constant(io Value)"}
   addNullObject = RigVMFunctionReferenceNode(
    Options_Null="(pelvis_1_#1=False,spine_3_#2=False,clavicle_5_#3=False,arm_7_#4=False,leg_9_#5=False,head_11_46DC15EF465C ...(truncated)"
    Custom_Joints=@ItemArray.Value
    ExecuteContext=@n1.ExecutePin
   )
   > &RigUnit_HierarchyAddNull(
    Name=head_sp_null
    Parent=(Type=Bone,Name="spine_05")
    Space=LocalSpace
    Transform=&RigUnit_GetRelativeTransformForItem(Child=(Type=Bone,Name="head"), Parent=(Type=Bone,Name="spine_05"), bChildInitial=true, bParentInitial=true).RelativeTransform
   )
   addFKCtrl = RigVMFunctionReferenceNode(
    Custom_FKCtrl=()
    ExecuteContext=()
    Options_FKCtrl="(pelvis_1_#1=true,spine_3_#2=true,clavicle_5_#3=true,arm_7_#4=false,leg_9_#5=False,head_11_46DC15EF465CD29 ...(truncated)"
    Shape_Settings="(pelvis_2_#6=(bVisible=True,Name=\"Hexagon_Thin\",Color=(R=0.0117,G=0.8281,B=1,A=1),Transform=(Rotation=(X=0,Y=-0.7071,Z=0,W=0.7071),Translation=(X=2,Y=0,Z=0),Scale3D=(X=5,Y=5,Z=7))),spine_4_#7=(bVi ...(truncated)"
   )
  ]}
]
hierarchy [
 root {loc=<0 0 0>, rot=<0 0 0 1>}
 pelvis {parent=root, loc=<0 -0.5301 98.0571>, rot=<0.0127 0.707 -0.0127 -0.707>}
 spine_01 {parent=pelvis, loc=<6.1208 -0.9328 0>, rot=<0 0 0.034 -0.9994>}
 spine_02 {parent=spine_01, loc=<6.5997 -0.2095 0>, rot=<0 0 0.0238 -0.9997>}
 spine_03 {parent=spine_02, loc=<6.4914 -0.1237 0>, rot=<0 0 0.0149 -0.9999>}
 spine_04 {parent=spine_03, loc=<6.9598 0.0794 0>, rot=<0 0 -0.0132 -0.9999>}
 spine_05 {parent=spine_04, loc=<6.9987 0 0>, rot=<0 0 -0.0015 -1>}
 neck_01 {parent=spine_05, loc=<11.3138 0.1529 0>, rot=<0 0 -0.2003 -0.9797>}
 neck_02 {parent=neck_01, loc=<4.5627 0 0>, rot=<0 0 0.0504 -0.9987>}
 head {parent=neck_02, loc=<5.3905 0 0.0001>, rot=<0 0 0.115 -0.9934>}
 base_hair_1_l {parent=head, loc=<1.2178 -5.8046 -5.6>, rot=<0.1334 -0.6944 0.6944 -0.1334>}
 dyn_hair_1_l {parent=base_hair_1_l, loc=<0 0 0>, rot=<0 -0.0463 0 -0.9989>}
 dyn_hair_2_l {parent=dyn_hair_1_l, loc=<5.8179 0 0.2787>, rot=<0 0 0 1>}
 dyn_hair_3_l {parent=dyn_hair_2_l, loc=<5.8259 0 0.373>, rot=<0 -0.0833 0 -0.9965>}
 dyn_hair_4_l {parent=dyn_hair_3_l, loc=<5.7582 0 0.19>, rot=<0 -0.068 0 -0.9977>}
 dyn_hair_5_l {parent=dyn_hair_4_l, loc=<5.7522 0 -0.011>, rot=<0 -0.0079 0 -1>}
 dyn_hair_6_l {parent=dyn_hair_5_l, loc=<5.7806 0 -0.0625>, rot=<0 0.026 0 -0.9997>}
 Back_side_l_base {parent=dyn_hair_1_l, loc=<5.3341 3.0195 -6.5325>, rot=<0.4629 0.2913 -0.5965 -0.5874>}
 Back_side_l_dyn {parent=Back_side_l_base, loc=<0 0 0>, rot=<0 0 0 1>}
 base_hair_1_r {parent=head, loc=<1.2178 -5.8046 5.6>, rot=<0.1334 -0.6944 0.6944 -0.1334>}
 dyn_hair_1_r {parent=base_hair_1_r, loc=<0 0 0>, rot=<0 -0.0463 0 -0.9989>}
 dyn_hair_2_r {parent=dyn_hair_1_r, loc=<5.8181 0 0.2786>, rot=<0 0.0027 0 -1>}
 dyn_hair_3_r {parent=dyn_hair_2_r, loc=<5.8254 0 0.3732>, rot=<0 -0.0833 0 -0.9965>}
 dyn_hair_4_r {parent=dyn_hair_3_r, loc=<5.7585 0 0.1899>, rot=<0 -0.068 0 -0.9977>}
 dyn_hair_5_r {parent=dyn_hair_4_r, loc=<5.7523 0 -0.011>, rot=<0 -0.0079 0 -1>}
 dyn_hair_6_r {parent=dyn_hair_5_r, loc=<5.7803 0 -0.0626>, rot=<0 0.026 0 -0.9997>}
 Back_side_r_base {parent=dyn_hair_1_r, loc=<5.3341 -3.3678 -6.5325>, rot=<0.4629 0.2913 -0.5965 -0.5874>}
 Back_side_r_dyn {parent=Back_side_r_base, loc=<0 0 0>, rot=<0 0 0 1>}
 base_hair_1_c {parent=head, loc=<1.2187 -5.9142 0>, rot=<0.1334 -0.6944 0.6944 -0.1334>}
 dyn_hair_1_c {parent=base_hair_1_c, loc=<0 0 0>, rot=<0 -0.054 0 -0.9985>}
 dyn_hair_2_c {parent=dyn_hair_1_c, loc=<5.8138 0 0.254>, rot=<0 0.0118 0 -0.9999>}
 dyn_hair_3_c {parent=dyn_hair_2_c, loc=<5.8271 0 0.3784>, rot=<0 -0.0861 0 -0.9963>}
 dyn_hair_4_c {parent=dyn_hair_3_c, loc=<5.7577 0 0.1855>, rot=<0 -0.0666 0 -0.9978>}
 dyn_hair_5_c {parent=dyn_hair_4_c, loc=<5.7522 0 -0.0121>, rot=<0 -0.0071 0 -1>}
 dyn_hair_6_c {parent=dyn_hair_5_c, loc=<5.7806 0 -0.0609>, rot=<0 0.026 0 -0.9997>}
 hair_side_r_01_base {parent=head, loc=<2.0222 2.3085 7.3401>, rot=<-0.0309 -0.712 -0.7014 -0.0064>}
 hair_side_r_01_dyn {parent=hair_side_r_01_base, loc=<0 0 0>, rot=<0 0 0 1>}
 hair_side_r_02_dyn {parent=hair_side_r_01_dyn, loc=<8.0329 0 0>, rot=<0.0804 0.0981 0.0047 -0.9919>}
 hair_side_r_03_dyn {parent=hair_side_r_02_dyn, loc=<7.1361 0 0>, rot=<-0.0141 0.0944 0.0052 -0.9954>}
 hair_side_r_04_dyn {parent=hair_side_r_03_dyn, loc=<6.7078 0 0>, rot=<0 0 0 1>}
 hair_side_l_01_base {parent=head, loc=<2.0222 2.3085 -7.3401>, rot=<0.0064 0.7014 0.712 0.0309>}
 hair_side_l_01_dyn {parent=hair_side_l_01_base, loc=<0 0 0>, rot=<0 0 0 1>}
 hair_side_l_02_dyn {parent=hair_side_l_01_dyn, loc=<8.0329 0 0>, rot=<-0.0804 0.0981 -0.0047 -0.9919>}
 hair_side_l_03_dyn {parent=hair_side_l_02_dyn, loc=<7.1361 0 0>, rot=<0.0141 0.0944 -0.0052 -0.9954>}
 hair_side_l_04_dyn {parent=hair_side_l_03_dyn, loc=<6.7078 0 0>, rot=<0 0 0 1>}
 front_hair_dyn {parent=head, loc=<8.5333 9.574 -0.4558>, rot=<0 0 0 1>}
 Back_rear_base {parent=head, loc=<-1.3089 -9.2165 -0.2788>, rot=<0.5347 -0.4627 0.4627 -0.5347>}
 Back_rear_dyn {parent=Back_rear_base, loc=<0 0 0>, rot=<0 0 0 1>}
 clavicle_l {parent=spine_05, loc=<5.4831 -0.1053 -3.2946>, rot=<0.0465 -0.7289 -0.009 -0.6829>}
 upperarm_l {parent=clavicle_l, loc=<9.8658 -1.2136 -0.4105>, rot=<0.0164 -0.3245 -0.0269 -0.9454>}
 upperarm_twist_01_l {parent=upperarm_l, loc=<7.2135 -0.1134 0>, rot=<0 0 0 1>}
 deform_upperarm_fr_base_l {parent=upperarm_twist_01_l, loc=<-5.6312 5.092 0.3354>, rot=<0 0 0 -1>}
 group3 {parent=deform_upperarm_fr_base_l, loc=<0 0 0>, rot=<0 0 0 1>}
 deform_upperarm_fr_l {parent=group3, loc=<0 0 0>, rot=<0 0 0 1>}
 clavicle_r {parent=spine_05, loc=<5.4831 -0.1053 3.2946>, rot=<0.7289 0.0465 -0.6829 0.009>}
 upperarm_r {parent=clavicle_r, loc=<-9.8658 1.2136 0.4105>, rot=<0.0164 -0.3245 -0.0269 -0.9454>}
 upperarm_twist_01_r {parent=upperarm_r, loc=<-7.2135 0.1134 0>, rot=<0 0 0 -1>}
 deform_upperarm_fr_base_r {parent=upperarm_twist_01_r, loc=<5.6312 -5.092 -0.3354>, rot=<0 0 0 -1>}
 deform_upperarm_fr_r {parent=deform_upperarm_fr_base_r, loc=<0 0 0>, rot=<0 0 0 1>}
 Bow_C_base {parent=spine_05, loc=<6.6805 6.6646 0.001>, rot=<-0.5877 -0.3932 -0.3932 -0.5877>}
 Bow_C_dyn {parent=Bow_C_base, loc=<0 0 0>, rot=<0 0 0 1>}
 Bow_l_dyn {parent=Bow_C_dyn, loc=<0.5921 -3.2569 -0.4827>, rot=<-0.5642 0.4135 0.6545 -0.2869>}
 Bow_r_dyn {parent=Bow_C_dyn, loc=<0.3163 2.7804 -0.4463>, rot=<-0.572 -0.3902 0.6178 0.3726>}
 Bow_side_l_dyn {parent=Bow_C_dyn, loc=<1.3178 -1.3924 -3.5735>, rot=<-0.6949 0.0897 0.7113 0.0557>}
 Bow_side_r_dyn {parent=Bow_C_dyn, loc=<1.0336 1.2109 -2.995>, rot=<0.7137 0.0392 -0.6987 -0.0308>}
 deform_pec_aim_base_l {parent=spine_04, loc=<3.2456 9.3309 -1.2237>, rot=<0.0603 -0.4033 0.3318 -0.8507>}
 deform_pec_aim_l {parent=deform_pec_aim_base_l, loc=<0 0 0>, rot=<0 0 0 1>}
 deform_pec_base_l {parent=deform_pec_aim_l, loc=<9.3665 0 0>, rot=<0 0 0 1>}
 deform_pec_l {parent=deform_pec_base_l, loc=<0 0 0>, rot=<0 0 0 1>}
 deform_pec_aim_base_r {parent=spine_04, loc=<3.2452 9.3309 1.2237>, rot=<-0.4033 -0.0603 0.8507 0.3318>}
 deform_pec_aim_r {parent=deform_pec_aim_base_r, loc=<0 0 0>, rot=<0 0 0 1>}
 deform_pec_base_r {parent=deform_pec_aim_r, loc=<-9.3667 -0.0003 -0.0004>, rot=<0 0 0 1>}
 deform_pec_r {parent=deform_pec_base_r, loc=<0 0 0>, rot=<0 0 0 1>}
 dyn_hair_1_l_ctrl_null {type=Null, loc=<5.6 -7.5018 153.204>, rot=<-0.4257 -0.4257 0.5646 -0.5646>}
 dyn_hair_1_l_ctrl {controlType=EULER_TRANSFORM, parent=dyn_hair_1_l_ctrl_null, type=Control, loc=<145.2309 -5.6 -49.353>, rot=<0.4257 0.4257 -0.5646 -0.5646>}
 dyn_hair_2_l_ctrl_null {type=Null, loc=<5.6 -9.37 147.6872>, rot=<-0.4242 -0.4242 0.5657 -0.5657>}
 dyn_hair_2_l_ctrl {controlType=EULER_TRANSFORM, parent=dyn_hair_2_l_ctrl_null, type=Control, loc=<-5.8193 0 -0.2471>, rot=<0 -0.0027 0 -1>}
 dyn_hair_3_l_ctrl_null {type=Null, loc=<5.6 -11.3609 142.1993>, rot=<-0.4699 -0.4699 0.5284 -0.5284>}
 dyn_hair_3_l_ctrl {controlType=EULER_TRANSFORM, parent=dyn_hair_3_l_ctrl_null, type=Control, loc=<-5.683 0 -1.3356>, rot=<0 0.0833 0 -0.9965>}
 dyn_hair_4_l_ctrl_null {type=Null, loc=<5.6 -12.223 136.5029>, rot=<-0.5047 -0.5047 0.4953 -0.4953>}
 dyn_hair_4_l_ctrl {controlType=EULER_TRANSFORM, parent=dyn_hair_4_l_ctrl_null, type=Control, loc=<-5.6793 0 -0.9691>, rot=<0 0.068 0 -0.9977>}
 dyn_hair_5_l_ctrl_null {type=Null, loc=<5.6 -12.1036 130.7519>, rot=<-0.5086 -0.5086 0.4913 -0.4913>}
 dyn_hair_5_l_ctrl {controlType=EULER_TRANSFORM, parent=dyn_hair_5_l_ctrl_null, type=Control, loc=<-5.7516 0 -0.0802>, rot=<0 0.0079 0 -1>}
 dyn_hair_6_l_ctrl_null {type=Null, loc=<5.6 -11.8407 124.9769>, rot=<-0.4956 -0.4956 0.5043 -0.5043>}
 dyn_hair_6_l_ctrl {controlType=EULER_TRANSFORM, parent=dyn_hair_6_l_ctrl_null, type=Control, loc=<-5.7696 0 0.3631>, rot=<0 -0.026 0 -0.9997>}
 dyn_hair_7_l_ctrl_null {type=Null, loc=<5.6 -11.9586 119.1988>, rot=<-0.4956 -0.4956 0.5043 -0.5043>}
 dyn_hair_7_l_ctrl {controlType=EULER_TRANSFORM, parent=dyn_hair_7_l_ctrl_null, type=Control, loc=<-5.7793 0 -0.0177>, rot=<0 0 0 1>}
 dyn_hair_1_r_ctrl_null {type=Null, loc=<-5.6 -7.5018 153.204>, rot=<-0.4257 -0.4257 0.5646 -0.5646>}
 dyn_hair_1_r_ctrl {controlType=EULER_TRANSFORM, parent=dyn_hair_1_r_ctrl_null, type=Control, loc=<145.2309 5.6 -49.353>, rot=<0.4257 0.4257 -0.5646 -0.5646>}
 dyn_hair_2_r_ctrl_null {type=Null, loc=<-5.6 -9.37 147.687>, rot=<-0.4242 -0.4242 0.5657 -0.5657>}
 dyn_hair_2_r_ctrl {controlType=EULER_TRANSFORM, parent=dyn_hair_2_r_ctrl_null, type=Control, loc=<-5.8195 0 -0.247>, rot=<0 -0.0027 0 -1>}
 dyn_hair_3_r_ctrl_null {type=Null, loc=<-5.6 -11.361 142.1997>, rot=<-0.4699 -0.4699 0.5284 -0.5284>}
 dyn_hair_3_r_ctrl {controlType=EULER_TRANSFORM, parent=dyn_hair_3_r_ctrl_null, type=Control, loc=<-5.6824 0 -1.3357>, rot=<0 0.0833 0 -0.9965>}
 dyn_hair_4_r_ctrl_null {type=Null, loc=<-5.6 -12.2229 136.5029>, rot=<-0.5047 -0.5047 0.4953 -0.4953>}
 dyn_hair_4_r_ctrl {controlType=EULER_TRANSFORM, parent=dyn_hair_4_r_ctrl_null, type=Control, loc=<-5.6796 0 -0.969>, rot=<0 0.068 0 -0.9977>}
 dyn_hair_5_r_ctrl_null {type=Null, loc=<-5.6 -12.1037 130.7518>, rot=<-0.5086 -0.5086 0.4913 -0.4913>}
 dyn_hair_5_r_ctrl {controlType=EULER_TRANSFORM, parent=dyn_hair_5_r_ctrl_null, type=Control, loc=<-5.7518 0 -0.0803>, rot=<0 0.0079 0 -1>}
 dyn_hair_6_r_ctrl_null {type=Null, loc=<-5.6 -11.8406 124.9772>, rot=<-0.4956 -0.4956 0.5043 -0.5043>}
 dyn_hair_6_r_ctrl {controlType=EULER_TRANSFORM, parent=dyn_hair_6_r_ctrl_null, type=Control, loc=<-5.7692 0 0.3631>, rot=<0 -0.026 0 -0.9997>}
 dyn_hair_7_r_ctrl_null {type=Null, loc=<-5.6 -11.9585 119.1992>, rot=<-0.4956 -0.4956 0.5043 -0.5043>}
 dyn_hair_7_r_ctrl {controlType=EULER_TRANSFORM, parent=dyn_hair_7_r_ctrl_null, type=Control, loc=<-5.7791 0 -0.0177>, rot=<0 0 0 1>}
 dyn_hair_1_c_ctrl_null {type=Null, loc=<0 -7.6114 153.204>, rot=<-0.4301 -0.4301 0.5613 -0.5613>}
 dyn_hair_1_c_ctrl {controlType=EULER_TRANSFORM, parent=dyn_hair_1_c_ctrl_null, type=Control, loc=<145.9508 0 -47.1992>, rot=<0.4301 0.4301 -0.5613 -0.5613>}
 dyn_hair_2_c_ctrl_null {type=Null, loc=<0 -9.3689 147.6564>, rot=<-0.4235 -0.4235 0.5663 -0.5663>}
 dyn_hair_2_c_ctrl {controlType=EULER_TRANSFORM, parent=dyn_hair_2_c_ctrl_null, type=Control, loc=<-5.8182 0 -0.1172>, rot=<0 -0.0118 0 -0.9999>}
 dyn_hair_3_c_ctrl_null {type=Null, loc=<0 -11.3794 142.1741>, rot=<-0.4707 -0.4707 0.5277 -0.5277>}
 dyn_hair_3_c_ctrl {controlType=EULER_TRANSFORM, parent=dyn_hair_3_c_ctrl_null, type=Control, loc=<-5.6757 0 -1.3727>, rot=<0 0.0861 0 -0.9963>}
 dyn_hair_4_c_ctrl_null {type=Null, loc=<0 -12.2198 136.475>, rot=<-0.5047 -0.5047 0.4952 -0.4952>}
 dyn_hair_4_c_ctrl {controlType=EULER_TRANSFORM, parent=dyn_hair_4_c_ctrl_null, type=Control, loc=<-5.682 0 -0.9486>, rot=<0 0.0666 0 -0.9978>}
 dyn_hair_5_c_ctrl_null {type=Null, loc=<0 -12.0983 130.7242>, rot=<-0.5082 -0.5082 0.4916 -0.4916>}
 dyn_hair_5_c_ctrl {controlType=EULER_TRANSFORM, parent=dyn_hair_5_c_ctrl_null, type=Control, loc=<-5.7518 0 -0.0692>, rot=<0 0.0071 0 -1>}
 dyn_hair_6_c_ctrl_null {type=Null, loc=<0 -11.8459 124.9488>, rot=<-0.4953 -0.4953 0.5047 -0.5047>}
 dyn_hair_6_c_ctrl {controlType=EULER_TRANSFORM, parent=dyn_hair_6_c_ctrl_null, type=Control, loc=<-5.7696 0 0.3615>, rot=<0 -0.026 0 -0.9997>}
 dyn_hair_7_c_ctrl_null {type=Null, loc=<0 -11.9744 119.1709>, rot=<-0.4953 -0.4953 0.5047 -0.5047>}
 dyn_hair_7_c_ctrl {controlType=EULER_TRANSFORM, parent=dyn_hair_7_c_ctrl_null, type=Control, loc=<-5.7793 0 -0.0193>, rot=<0 0 0 1>}
]
variables [
 Is Moving Left {type=double, default=0}
 Is Moving Right {type=double, default=0}
 IsJumping {type=double, default=0}
 Is_Crouching {type=bool, default=False}
 Not_Crouching {type=double, default=0}
 SMG_P90 {type=bool, default=False}
 head rz n60n80 alpha {type=double, default=0}
 is CelebratinEmote {type=double, default=0}
 is Parachuting {type=double, default=0}
 is Sky Diving Floating {type=double, default=0}
 is SkyDiving Mode {type=double, default=0}
 isAR Equipped lerp {type=double, default=0}
 isCelbration Emote {type=double, default=0}
 isFrontEnd {type=bool, default=False}
 isPASwing lerp {type=double, default=0}
 isPistol Equipped lerp {type=double, default=0}
 isPistolASDUp Extreme {type=double, default=0}
 isTargeting lerp {type=double, default=0}
 is_OutLaw_SG {type=bool, default=False}
 looking_left {type=double, default=0}
 pub PA Swing {type=bool, default=False}
 pub isAR Equipped {type=bool, default=False}
 pub isPistol Equipped {type=bool, default=False}
 pub isRigidBodies Enabled {type=bool, default=False}
 pub isTargeting {type=bool, default=False}
 pub isWearingBackBling {type=bool, default=False}
]
