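# Physics asset for the JoltMosaic female medium cosmetic's front face accessory, simulating three dangling chains (left, center, right) on dyn_hair bones.
# Each chain is five capsules hung from spine_05, with zero twist limit and wider swing on one axis that narrows toward the tip.
# Capsules on the spine, clavicles and upper arms form the body shapes around the chains, linked by free-motion constraints.
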PhysicsAsset /BRCosmetics/Characters/Player/Female/Medium/Bodies/F_MED_JoltMosaic/Meshes/Parts/F_MED_JoltMosaic_FaceAcc_Front_Physics
bodies {
 clavicle_l [{shapeType=Capsule, length=8.6429, radius=6, shapeName=clavicle_l_capsule, center=<2.9094 0.5 -2.5725>, rotation=<rot -0.0001 -83.1598 -90>}]
 clavicle_r [{
   shapeType=Capsule
   length=8.6429
   radius=6.6
   shapeName=clavicle_r_capsule
   center=<-2.9609 0.1051 2.4063>
   rotation=<rot -0.0002 -85.1346 90>
  }]
 dyn_hair_2_c [{shapeType=Capsule, length=2.2, radius=2.2, center=<1 0 0>, rotation=<rot 90 0 0>}]
 dyn_hair_2_l [{shapeType=Capsule, length=2.2, radius=2.2, center=<1 0 0>, rotation=<rot 90 0 0>}]
 dyn_hair_2_r [{shapeType=Capsule, length=2.2, radius=2.2, center=<1 0 0>, rotation=<rot 90 0 0>}]
 dyn_hair_3_c [{shapeType=Capsule, length=3.4995, radius=2.5, shapeName=dyn_C_hair_1_capsule, center=<5.4739 0 0.68>, rotation=<rot 85.5986 0 0>}]
 dyn_hair_3_l [{shapeType=Capsule, length=3.7831, radius=2.5, shapeName=dyn_C_hair_1_capsule, center=<4 0 0>, rotation=<rot 73.9917 89.9988 179.9988>}]
 dyn_hair_3_r [{shapeType=Capsule, length=8, radius=2.5, center=<4 -0.8462 -0.0847>, rotation=<rot 79.8933 -89.9981 -179.9981>}]
 dyn_hair_4_c [{shapeType=Capsule, length=3.6191, radius=2.6, shapeName=dyn_C_hair_1_capsule, center=<4.8324 0 0.15>, rotation=<rot 88.2826 0 0>}]
 dyn_hair_4_l [{shapeType=Capsule, length=3.4698, radius=2.6, shapeName=dyn_C_hair_1_capsule, center=<4.8048 1.8 -0.13>, rotation=<rot 88.2826 0 0>}]
 dyn_hair_4_r [{shapeType=Capsule, length=3.4698, radius=2.6, center=<4.8048 -1.8 0.13>, rotation=<rot 88.2826 0 -180>}]
 dyn_hair_5_c [{shapeType=Capsule, length=3.9255, radius=2.5, shapeName=dyn_C_hair_1_capsule, center=<5.1655 0 -0.15>, rotation=<rot 86.9145 -180 -180>}]
 dyn_hair_5_l [{shapeType=Capsule, length=3.9255, radius=2.5, shapeName=dyn_C_hair_1_capsule, center=<5.1655 0 -0.15>, rotation=<rot 86.9145 -180 -180>}]
 dyn_hair_5_r [{shapeType=Capsule, length=3.9255, radius=2.5, shapeName=dyn_C_hair_1_capsule, center=<5.1655 0 0>, rotation=<rot 86.9145 -180 -180>}]
 dyn_hair_6_c [{shapeType=Capsule, length=4.7118, radius=2.5, shapeName=dyn_C_hair_1_capsule, center=<4.1881 0 0.1>, rotation=<rot 89.2083 -180 -180>}]
 dyn_hair_6_l [{shapeType=Capsule, length=4.7118, radius=2.5, shapeName=dyn_C_hair_1_capsule, center=<4.1881 0 0.1>, rotation=<rot 89.2084 -180 -180>}]
 dyn_hair_6_r [{shapeType=Capsule, length=4.7118, radius=2.5, shapeName=dyn_C_hair_1_capsule, center=<4.1881 0 0>, rotation=<rot 89.2083 -180 -180>}]
 spine_02 [{shapeType=Capsule, length=4.887, radius=9.8602, shapeName=spine_02_capsule, center=<-0.2722 3.3234 -0.3025>, rotation=<rot 0 -90 0>}]
 spine_03 [{shapeType=Capsule, length=8.629, radius=9.6494, shapeName=spine_03_capsule, center=<-1.3718 2.4078 0>, rotation=<rot 0 -90 0>}]
 spine_04 [{shapeType=Capsule, length=3.4282, radius=12.5388, shapeName=spine_04_capsule, center=<2.2798 4.3829 0>, rotation=<rot 0 -90 0>}]
 spine_05 [{shapeType=Capsule, length=5.7751, radius=8.5, shapeName=spine_04_capsule, center=<0.0586 0.9277 0>, rotation=<rot 0 -90 0>}]
 upperarm_l [{shapeType=Capsule, length=10, radius=5, shapeName=upperarm_l_capsule, center=<5 0 0>, rotation=<rot 90 0 0>}]
 upperarm_r [{shapeType=Capsule, length=10, radius=5, shapeName=upperarm_r_capsule, center=<-5 0 0>, rotation=<rot 90 180 0>}]
}
constraints [
 {
  bone1Name=spine_03
  bone2Name=spine_02
  swing1LimitDegrees=45
  swing1Motion=Free
  swing2LimitDegrees=45
  swing2Motion=Free
  twistLimitDegrees=45
  twistMotion=Free
 }
 {
  bone1Name=spine_04
  bone2Name=spine_03
  swing1LimitDegrees=45
  swing1Motion=Free
  swing2LimitDegrees=45
  swing2Motion=Free
  twistLimitDegrees=45
  twistMotion=Free
 }
 {
  bone1Name=spine_05
  bone2Name=spine_04
  swing1LimitDegrees=45
  swing1Motion=Free
  swing2LimitDegrees=45
  swing2Motion=Free
  twistLimitDegrees=45
  twistMotion=Free
 }
 {bone1Name=dyn_hair_3_l, bone2Name=dyn_hair_2_l, swing1LimitDegrees=10, swing2LimitDegrees=35, twistLimitDegrees=0}
 {bone1Name=dyn_hair_4_l, bone2Name=dyn_hair_3_l, swing1LimitDegrees=10, swing2LimitDegrees=35, twistLimitDegrees=0}
 {bone1Name=dyn_hair_5_l, bone2Name=dyn_hair_4_l, swing1LimitDegrees=10, swing2LimitDegrees=28, twistLimitDegrees=0}
 {bone1Name=dyn_hair_6_l, bone2Name=dyn_hair_5_l, swing1LimitDegrees=10, swing2LimitDegrees=25, twistLimitDegrees=0}
 {bone1Name=dyn_hair_3_r, bone2Name=dyn_hair_2_r, swing1LimitDegrees=10, swing2LimitDegrees=35, twistLimitDegrees=0}
 {bone1Name=dyn_hair_4_r, bone2Name=dyn_hair_3_r, swing1LimitDegrees=10, swing2LimitDegrees=35, twistLimitDegrees=0}
 {bone1Name=dyn_hair_5_r, bone2Name=dyn_hair_4_r, swing1LimitDegrees=15, swing2LimitDegrees=28, twistLimitDegrees=0}
 {bone1Name=dyn_hair_6_r, bone2Name=dyn_hair_5_r, swing1LimitDegrees=15, swing2LimitDegrees=25, twistLimitDegrees=0}
 {bone1Name=dyn_hair_3_c, bone2Name=dyn_hair_2_c, swing1LimitDegrees=10, swing2LimitDegrees=40, twistLimitDegrees=0}
 {bone1Name=dyn_hair_4_c, bone2Name=dyn_hair_3_c, swing1LimitDegrees=10, swing2LimitDegrees=35, twistLimitDegrees=0}
 {bone1Name=dyn_hair_5_c, bone2Name=dyn_hair_4_c, swing1LimitDegrees=10, swing2LimitDegrees=28, twistLimitDegrees=0}
 {bone1Name=dyn_hair_6_c, bone2Name=dyn_hair_5_c, swing1LimitDegrees=10, swing2LimitDegrees=25, twistLimitDegrees=0}
 {bone1Name=dyn_hair_2_c, bone2Name=spine_05, swing1LimitDegrees=10, swing2LimitDegrees=25, twistLimitDegrees=0}
 {bone1Name=dyn_hair_2_l, bone2Name=spine_05, swing1LimitDegrees=5, swing2LimitDegrees=10, twistLimitDegrees=0}
 {bone1Name=dyn_hair_2_r, bone2Name=spine_05, swing1LimitDegrees=10, swing2LimitDegrees=25, twistLimitDegrees=0}
 {
  bone1Name=clavicle_l
  bone2Name=spine_05
  swing1LimitDegrees=45
  swing1Motion=Free
  swing2LimitDegrees=45
  swing2Motion=Free
  twistLimitDegrees=45
  twistMotion=Free
 }
 {
  bone1Name=clavicle_r
  bone2Name=spine_05
  swing1LimitDegrees=45
  swing1Motion=Free
  swing2LimitDegrees=45
  swing2Motion=Free
  twistLimitDegrees=45
  twistMotion=Free
 }
 {
  bone1Name=upperarm_l
  bone2Name=clavicle_l
  swing1LimitDegrees=45
  swing1Motion=Free
  swing2LimitDegrees=45
  swing2Motion=Free
  twistLimitDegrees=45
  twistMotion=Free
 }
 {
  bone1Name=upperarm_r
  bone2Name=clavicle_r
  swing1LimitDegrees=45
  swing1Motion=Free
  swing2LimitDegrees=45
  swing2Motion=Free
  twistLimitDegrees=45
  twistMotion=Free
 }
]
