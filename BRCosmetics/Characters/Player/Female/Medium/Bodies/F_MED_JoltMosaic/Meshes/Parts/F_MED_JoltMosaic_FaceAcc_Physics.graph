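# Physics asset for the face accessory mesh of the female JoltMosaic cosmetic, covering two side hair strands and a bow.
# Capsules on the spine, clavicles, upper arms, neck and head form a body proxy around the accessory bones.
# Left and right hair chains run from head-linked base bones through four dyn bones with narrow, asymmetric swing and twist limits.
# The bow chains off spine_05 with free angular constraints, as do the body and hair-base links.
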PhysicsAsset /BRCosmetics/Characters/Player/Female/Medium/Bodies/F_MED_JoltMosaic/Meshes/Parts/F_MED_JoltMosaic_FaceAcc_Physics
bodies {
 Bow_C_dyn [{
   shapeType=Capsule
   length=5
   radius=4
   shapeName=Bow_C_dyn_capsule
   center=<-2.3634 -0.3682 -1.6882>
   rotation=<rot -88.3194 109.0784 158.2903>
  }]
 Bow_r_dyn [{shapeType=Capsule, length=4, radius=3, shapeName=Bow_r_dyn_capsule, center=<-0.2352 2.6604 1.9356>, rotation=<rot 83.61 -93.1875 1.6249>}]
 clavicle_l [{
   shapeType=Capsule
   length=17.7735
   radius=5.5694
   shapeName=clavicle_l_capsule
   center=<3.1048 -1.2458 0.113>
   rotation=<rot -80.8819 179.9988 177.4577>
  }]
 clavicle_r [{shapeType=Capsule, length=17.1429, radius=4.5, shapeName=clavicle_r_capsule, center=<-3 -0.75 0>, rotation=<rot 0 -83 -90>}]
 hair_side_l_01_base [{shapeType=Capsule, length=4, radius=2.5, shapeName=hair_side_l_01_base_capsule, center=<2 0 0>, rotation=<rot 90 0 0>}]
 hair_side_l_01_dyn [{shapeType=Capsule, length=4, radius=2.5, shapeName=hair_side_l_01_dyn_capsule, center=<2 0 0>, rotation=<rot 90 0 0>}]
 hair_side_l_02_dyn [{shapeType=Capsule, length=4, radius=2.25, shapeName=hair_side_l_02_dyn_capsule, center=<1.5 0 0>, rotation=<rot 87.113 0 0>}]
 hair_side_l_03_dyn [{
   shapeType=Capsule
   length=4
   radius=2
   shapeName=hair_side_l_03_dyn_capsule
   center=<1.0287 0 -0.25>
   rotation=<rot 83.8484 -179.9999 -179.9999>
  }]
 hair_side_l_04_dyn [{shapeType=Capsule, length=7, radius=2, shapeName=hair_side_l_04_dyn_capsule, center=<3 0 1>, rotation=<rot 90 0 0>}]
 hair_side_r_01_base [{shapeType=Capsule, length=4, radius=2.5, shapeName=hair_side_r_01_base_capsule, center=<2 0 0>, rotation=<rot 90 0 0>}]
 hair_side_r_01_dyn [{shapeType=Capsule, length=4, radius=2.5, shapeName=hair_side_r_01_dyn_capsule, center=<2 0 0>, rotation=<rot 90 0 0>}]
 hair_side_r_02_dyn [{shapeType=Capsule, length=4.5, radius=1.75, shapeName=hair_side_r_02_dyn_capsule, center=<2 -0.35 0>, rotation=<rot 90 0 0>}]
 hair_side_r_03_dyn [{shapeType=Capsule, length=5, radius=1, shapeName=hair_side_r_03_dyn_capsule, center=<2 0 1>, rotation=<rot 90 0 0>}]
 hair_side_r_04_dyn [{shapeType=Capsule, length=7, radius=2.28, shapeName=hair_side_r_04_dyn_capsule, center=<3 -0.6 0.75>, rotation=<rot 90 0 0>}]
 head [{shapeType=Capsule, length=35, radius=6, shapeName=head_capsule, center=<-7.5233 -10.0631 0>, rotation=<rot 0 -90 0>}]
 neck_01 [{shapeType=Capsule, length=13.2067, radius=4.7, shapeName=neck_01_capsule, rotation=<rot 90 0 0>}]
 spine_02 [{shapeType=Capsule, length=4.887, radius=9.8602, shapeName=spine_02_capsule, center=<-0.2722 3.3234 -0.3025>, rotation=<rot 0 -90 0>}]
 spine_03 [{shapeType=Capsule, length=8.629, radius=9.6494, shapeName=spine_03_capsule, center=<-0.5142 1.5055 0>, rotation=<rot 0 -90 0>}]
 spine_04 [{shapeType=Capsule, length=3.4282, radius=9.5, shapeName=spine_04_capsule, center=<1.2499 3.811 0>, rotation=<rot 0 -90 0>}]
 spine_05 [{shapeType=Capsule, length=12.7998, radius=8.2002, shapeName=spine_04_capsule, center=<0.4465 -0.5399 0>, rotation=<rot 0 -90 0>}]
 upperarm_l [{shapeType=Capsule, length=17, radius=3.8, shapeName=upperarm_l_capsule, center=<5 -0.465 0>, rotation=<rot 90 2.2184 0>}]
 upperarm_r [{shapeType=Capsule, length=32.5, radius=3.8, shapeName=upperarm_r_capsule, center=<-11.514 1.6859 0.1903>, rotation=<rot 90 -178.1297 0>}]
}
constraints [
 {
  bone1Name=spine_03
  bone2Name=spine_02
  swing1LimitDegrees=45
  swing1Motion=Free
  swing2LimitDegrees=45
  swing2Motion=Free
  twistLimitDegrees=45
  twistMotion=Free
 }
 {
  bone1Name=spine_04
  bone2Name=spine_03
  swing1LimitDegrees=45
  swing1Motion=Free
  swing2LimitDegrees=45
  swing2Motion=Free
  twistLimitDegrees=45
  twistMotion=Free
 }
 {
  bone1Name=spine_05
  bone2Name=spine_04
  swing1LimitDegrees=45
  swing1Motion=Free
  swing2LimitDegrees=45
  swing2Motion=Free
  twistLimitDegrees=45
  twistMotion=Free
 }
 {
  bone1Name=clavicle_l
  bone2Name=spine_05
  swing1LimitDegrees=45
  swing1Motion=Free
  swing2LimitDegrees=45
  swing2Motion=Free
  twistLimitDegrees=45
  twistMotion=Free
 }
 {
  bone1Name=clavicle_r
  bone2Name=spine_05
  swing1LimitDegrees=45
  swing1Motion=Free
  swing2LimitDegrees=45
  swing2Motion=Free
  twistLimitDegrees=45
  twistMotion=Free
 }
 {
  bone1Name=upperarm_l
  bone2Name=clavicle_l
  swing1LimitDegrees=45
  swing1Motion=Free
  swing2LimitDegrees=45
  swing2Motion=Free
  twistLimitDegrees=45
  twistMotion=Free
 }
 {
  bone1Name=upperarm_r
  bone2Name=clavicle_r
  swing1LimitDegrees=45
  swing1Motion=Free
  swing2LimitDegrees=45
  swing2Motion=Free
  twistLimitDegrees=45
  twistMotion=Free
 }
 {
  bone1Name=head
  bone2Name=neck_01
  swing1LimitDegrees=45
  swing1Motion=Free
  swing2LimitDegrees=45
  swing2Motion=Free
  twistLimitDegrees=45
  twistMotion=Free
 }
 {
  bone1Name=Bow_C_dyn
  bone2Name=spine_05
  swing1LimitDegrees=45
  swing1Motion=Free
  swing2LimitDegrees=45
  swing2Motion=Free
  twistLimitDegrees=45
  twistMotion=Free
 }
 {
  bone1Name=Bow_r_dyn
  bone2Name=Bow_C_dyn
  swing1LimitDegrees=45
  swing1Motion=Free
  swing2LimitDegrees=45
  swing2Motion=Free
  twistLimitDegrees=45
  twistMotion=Free
 }
 {
  bone1Name=hair_side_l_01_base
  bone2Name=head
  swing1LimitDegrees=45
  swing1Motion=Free
  swing2LimitDegrees=45
  swing2Motion=Free
  twistLimitDegrees=45
  twistMotion=Free
 }
 {bone1Name=hair_side_l_01_dyn, bone2Name=hair_side_l_01_base, swing1LimitDegrees=8, swing2LimitDegrees=15, twistLimitDegrees=2}
 {bone1Name=hair_side_l_02_dyn, bone2Name=hair_side_l_01_dyn, swing1LimitDegrees=15, swing2LimitDegrees=35, twistLimitDegrees=5}
 {bone1Name=hair_side_l_03_dyn, bone2Name=hair_side_l_02_dyn, swing1LimitDegrees=8, swing2LimitDegrees=35, twistLimitDegrees=5}
 {bone1Name=hair_side_l_04_dyn, bone2Name=hair_side_l_03_dyn, swing1LimitDegrees=9, swing2LimitDegrees=35, twistLimitDegrees=5}
 {
  bone1Name=hair_side_r_01_base
  bone2Name=head
  swing1LimitDegrees=45
  swing1Motion=Free
  swing2LimitDegrees=45
  swing2Motion=Free
  twistLimitDegrees=45
  twistMotion=Free
 }
 {bone1Name=hair_side_r_01_dyn, bone2Name=hair_side_r_01_base, swing1LimitDegrees=0, swing2LimitDegrees=7, twistLimitDegrees=5}
 {bone1Name=hair_side_r_02_dyn, bone2Name=hair_side_r_01_dyn, swing1LimitDegrees=10, swing2LimitDegrees=30, twistLimitDegrees=7}
 {bone1Name=hair_side_r_03_dyn, bone2Name=hair_side_r_02_dyn, swing1LimitDegrees=10, swing2LimitDegrees=35, twistLimitDegrees=8}
 {bone1Name=hair_side_r_04_dyn, bone2Name=hair_side_r_03_dyn, swing1LimitDegrees=10, swing2LimitDegrees=35, twistLimitDegrees=8}
]
