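# Anim Blueprint for the F_MED_JoltMosaic face accessory part, adding secondary motion to hair on a female player cosmetic.
# A parent-mesh pose copy feeds a Control Rig, base-hair bone offsets, a RigidBody sim and front-hair AnimDynamics.
# Bool blends then select that result or extra RigidBody passes with front or crouching physics assets.
# The event graph derives gravity, sim alpha, fall overlap and rig inputs from movement, weapon, backbling and emote state.
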AnimBlueprint /BRCosmetics/Characters/Player/Female/Medium/Bodies/F_MED_JoltMosaic/Meshes/Parts/F_MED_JoltMosaic_FaceAcc_AnimBP
target_skeleton=&/F_MED_JoltMosaic_FaceAcc_Skeleton
graphs [
 AnimGraph {nodes [
   Output Pose(Result=Use cached pose 'dynamicsCP'()) {layerGroup=None}
   n1 = Control Rig!(
    Source=Copy Pose From Mesh() {bUseAttachedParent=true}
    Alpha
    pub isRigidBodies Enabled:bool=Get bRigidBodyEnabled()
    pub PA Swing:bool=Get bIsPlayingMeleeAnim()
    pub isWearingBackBling:bool=Get isWearingBlackbling() -> (isWearingBlackbling:bool)
    pub isTargeting:bool=Get bIsTargeting()
    pub isAR Equipped:bool=Get isAR Equipped() -> (isAR Equipped:bool)
    pub isPistol Equipped:bool=Get isPistol Equipped() -> (isPistol Equipped:bool)
    is_OutLaw_SG:bool=Get Is_outLaw_SG() -> (Is_outLaw_SG:bool)
    SMG_P90:bool=Get SMG_SG() -> (SMG_SG:bool)
    is Parachuting:double=Get Is_Parachute_Open() -> (Is_Parachute_Open:double)
    is Sky Diving Floating:double=Get Is Sky Diving() -> (Is Sky Diving:double)
    is SkyDiving Mode:double=Get SkyDiving_Mode() -> (SkyDiving_Mode:double)
    is CelebratinEmote:double=Get IsCelebrating() -> (IsCelebrating:double)
    Is Moving Left:double=Get strafeLeft float() -> (strafeLeft float:double)
    Is Moving Right:double=Get strafeRight float() -> (strafeRight float:double)
    isCelbration Emote:double=Get IsCelebrationEmote() -> (IsCelebrationEmote:double)
    Not_Crouching:double=Get NotCrouching() -> (NotCrouching:double)
    Is_Crouching:bool=Get bIsCrouching()
    isFrontEnd:bool=Get bIsInFrontEnd()
   ) -> (Pose) {controlRigAssetReference {blueprintRigClass=&/F_MED_JoltMosaic_FaceAcc_CtrlRig_02.F_MED_JoltMosaic_FaceAcc_CtrlRig_02_C}}
   Save cached pose 'controlRigCP'(Pose=@n1) {cacheName=controlRigCP}
   n2 = RigidBody!(
    ComponentPose=Local To Component(LocalPose=Use cached pose 'AnimDynamicsACP'())
    Alpha
    OverrideWorldGravity:Vector="0,0,0"
    SimSpaceSettings:Sim Space Settings Structure=(WorldAlpha=0.65,VelocityScaleZ=1,DampingAlpha=1,MaxLinearVelocity=400,MaxAngularVelocity=3,MaxLinearAcceleration=800,MaxAngularAcceleration=100,ExternalLinearDragV=(X=1,Y=1,Z=1),ExternalLinearVelocity=(X=0,Y=0,Z=0),ExternalAngularVelocity=(X=0,Y=0,Z=0))
   ) -> (Pose) {
    bOverrideWorldGravity=true
    overridePhysicsAsset=&/F_MED_JoltMosaic_FaceAcc_Front_Physics
    simulationSpace=BaseBoneSpace
    alphaBoolBlend {blendOption=Linear}
    alphaScaleBiasClamp {bInterpResult=true, interpSpeedDecreasing=3, interpSpeedIncreasing=3}
    baseBoneRef {boneName=neck_02}
    componentAppliedLinearAccClamp=<3000 3000 3000>
    simSpaceSettings {
     maxAngularAcceleration=100
     maxAngularVelocity=3
     maxLinearAcceleration=800
     maxLinearVelocity=400
     worldAlpha=0.65
     externalLinearDragV=<1 1 1>
    }
   }
   n3 = Component To Local(ComponentPose=@n2)
   n4 = RigidBody(
    ComponentPose=Local To Component(LocalPose=Use cached pose 'AnimDynamicsACP'())
    OverrideWorldGravity="0,0,0"
    SimSpaceSettings=(WorldAlpha=0.65,VelocityScaleZ=1,DampingAlpha=1,MaxLinearVelocity=400,MaxAngularVelocity=3,MaxLinearAcceleration=800,MaxAngularAcceleration=100,ExternalLinearDragV=(X=1,Y=1,Z=1),ExternalLinearVelocity=(X=0,Y=0,Z=0),ExternalAngularVelocity=(X=0,Y=0,Z=0))
   ) {
    bOverrideWorldGravity=true
    overridePhysicsAsset=&/F_MED_JoltMosaic_FaceAcc_Crouching_Physics
    simulationSpace=BaseBoneSpace
    alphaBoolBlend {blendOption=Linear}
    alphaScaleBiasClamp {bInterpResult=true, interpSpeedDecreasing=3, interpSpeedIncreasing=3}
    baseBoneRef {boneName=neck_02}
    componentAppliedLinearAccClamp=<3000 3000 3000>
    simSpaceSettings {
     maxAngularAcceleration=100
     maxAngularVelocity=3
     maxLinearAcceleration=800
     maxLinearVelocity=400
     worldAlpha=0.65
     externalLinearDragV=<1 1 1>
    }
   }
   n5 = Component To Local(ComponentPose=@n4)
   Save cached pose 'dynamicsCP'(Pose=Blend Poses by bool(BlendPose_0=Use cached pose 'AnimDynamicsACP'(), BlendPose_1=Blend Poses by bool(BlendPose_0=@n5, BlendPose_1=@n3) {blendType=Linear}) {blendType=Linear}) {cacheName=dynamicsCP}
   n6 = RigidBody!(ComponentPose=Local To Component(LocalPose=Use cached pose 'preSimTweaks'()), Alpha, OverrideWorldGravity="0,0,0", ExternalForce:Vector) -> (Pose) {
    bOverrideWorldGravity=true
    overridePhysicsAsset=&/F_MED_JoltMosaic_FaceAcc_Physics
    simulationSpace=BaseBoneSpace
    baseBoneRef {boneName=neck_02}
    componentAppliedLinearAccClamp=<3000 3000 3000>
    componentLinearAccScale=<0.2 0.2 0.2>
    componentLinearVelScale=<0.2 0.2 0.2>
    simSpaceSettings {
     dampingAlpha=0
     maxAngularAcceleration=75
     maxAngularVelocity=40
     maxLinearAcceleration=600
     maxLinearVelocity=250
     velocityScaleZ=0.05
     worldAlpha=0.4
     externalLinearDragV=<0.3 0.3 0.4>
    }
   }
   n7 = Anim Dynamics - Bone: front_hair_dyn!(ComponentPose=@n6, Alpha=1, GravityOverride:Vector="0,0,-980", ExternalForce="0,0,0") -> (Pose) {
    angularDampingOverride=1
    angularSpringConstant=12000
    bAngularSpring=true
    bOverrideAngularDamping=true
    bUseGravityOverride=true
    gravityScale=0.35
    simulationSpace=BoneRelative
    boundBone {boneName=front_hair_dyn}
    componentAppliedLinearAccClamp=<10000 10000 10000>
    componentLinearAccScale=<0.4 0.4 0.4>
    componentLinearVelScale=<0.4 0.4 0.4>
    gravityOverride {z=-980}
    physicsBodyDefinitions [{
      boundBone {boneName=front_hair_dyn}
      boxExtents=<15 20 20>
      constraintSetup {angularLimitsMax=<4 5 6>, angularLimitsMin=<-4 -5 -4>, angularTarget=<1 0 0>}
      localJointOffset=<4 0 0>
     }]
    relativeSpaceBone {boneName=head}
   }
   n8 = "Transform (Modify) Bone - Bone: hair_side_r_03_dyn"(ComponentPose=@n7, Alpha=1, Translation="0,0,0", Rotation="0.115,0,0", Scale="1,1,1") {rotationMode=BMM_Additive, rotationSpace=BCS_BoneSpace, boneToModify {boneName=hair_side_r_03_dyn}, rotation {pitch=0.115}}
   Save cached pose 'AnimDynamicsACP'(Pose=Component To Local(ComponentPose=@n8)) {cacheName=AnimDynamicsACP}
   n9 = "Transform (Modify) Bone - Bone: base_hair_1_l"!(ComponentPose=Local To Component(LocalPose=Use cached pose 'controlRigCP'()), Alpha=1, Translation="0,0,0", Rotation="10,0,0") -> (Pose) {
    rotationMode=BMM_Additive
    rotationSpace=BCS_BoneSpace
    translationMode=BMM_Additive
    translationSpace=BCS_BoneSpace
    alphaBoolBlend {blendOption=Linear}
    boneToModify {boneName=base_hair_1_l}
    rotation {pitch=10}
   }
   n10 = "Transform (Modify) Bone - Bone: base_hair_1_c"!(ComponentPose=@n9, Alpha=1, Translation="0,0,0", Rotation="10,0,0") -> (Pose) {
    rotationMode=BMM_Additive
    rotationSpace=BCS_BoneSpace
    translationMode=BMM_Additive
    translationSpace=BCS_BoneSpace
    alphaBoolBlend {blendOption=Linear}
    boneToModify {boneName=base_hair_1_c}
    rotation {pitch=10}
   }
   n11 = "Transform (Modify) Bone - Bone: base_hair_1_r"!(ComponentPose=@n10, Alpha=1, Translation="0,0,0", Rotation="10,0,0") -> (Pose) {
    rotationMode=BMM_Additive
    rotationSpace=BCS_BoneSpace
    translationMode=BMM_Additive
    translationSpace=BCS_BoneSpace
    alphaBoolBlend {blendOption=Linear}
    boneToModify {boneName=base_hair_1_r}
    rotation {pitch=10}
   }
   Save cached pose 'preSimTweaks'(Pose=Component To Local(ComponentPose=@n11)) {cacheName=preSimTweaks}
  ]}
 EventGraph {nodes [
   Event BlueprintUpdateAnimation()
   > n1 = Sequence!(execute) -> (then_0, then_1, then_2:Exec, then_3:Exec, then_4:Exec)
   > n2 = "Gravity Override State Function (Parameter Asset) (Nativized)"(
    CharacterPartAnimInstance=Self-Reference()
    GameTimeInSeconds=GetGameTimeInSeconds()
    Parameters=&/F_MED_Jolt_Mosaic_FaceAcc_Parameters
   )
   n3 = Get bIsMoving2D()
   n4 = Get bIsJumping()
   n5 = Get bIsFalling()
   > n6 = Branch(Condition=AND Boolean(A=OR Boolean(A=@n4, B=@n5), B=NOT Boolean(A=@n3)))
   > Set GravityOverride(GravityOverride="0,-250,-650")
   Set GravityOverride(execute=@n6.else, GravityOverride=@n2)
   > Set RBAN Settings Out(RBAN Settings Out=@n2.OutRigidBodyAnimNodeParameters)
   n7 = Branch(execute=@n1.then_2, Condition=Get bIsPlayingMeleeAnim())
   > Set alpha hairRigidBody(alpha hairRigidBody:double=0.75) -> (Output_Get:double)
   n8 = Branch(execute=@n7.else, Condition=Get isTacticalJumpingFalling() -> (isTacticalJumpingFalling:bool))
   > Set alpha hairRigidBody(alpha hairRigidBody=0.6)
   n9 = Branch(execute=@n8.else, Condition=Get bIsDBNOCarrying() -> (bIsDBNOCarrying:bool))
   > Set alpha hairRigidBody(alpha hairRigidBody=0.4)
   Set alpha hairRigidBody(execute=@n9.else, alpha hairRigidBody=1)
   n10 = float * float!(
    A="Clamp (Float)"(Value=Get PawnVelocityZ(), Min=-1000, Max=0)
    B:bool=Get bRigidBodyEnabled()
    C:bool=NOT Boolean(A=Get isWearingBlackbling())
    D:bool=NOT Boolean(A=Get bIsSkydiveDiveMode())
    E:bool=NOT Boolean(A=Get bIsDBNOCarrying())
   ) -> (ReturnValue) {is=title K2Node_PromotableOperator}
   n11 = "Atan2 (Degrees)"!(Y:double=@n10, X:double=3800) -> (ReturnValue:double)
   Set hairBaseOverlap rotator(execute=@n1.then_3, hairBaseOverlap rotator:Rotator=MakeRotator(Pitch=float * float!(A=@n11, B:double=-1) -> (ReturnValue))) -> (Output_Get:Rotator)
   > Set hairBaseOverlap vector(hairBaseOverlap vector:Vector=MakeVector(Y=0, Z=Remap Float(Input=@n10, OldRange=(X=0,Y=-1000), NewRange=(X=0,Y=2), Clamp=true) -> (Output=~))) -> (Output_Get:Vector)
   n12 = Sequence!(execute=@n1.then_4) -> (then_0, then_1, then_2, then_3, then_4, then_5:Exec, then_6:Exec, then_7:Exec)
   > Set Is_Parachute_Open(Is_Parachute_Open:double="To Float (Boolean)"(InBool=Get bIsParachuteOpened())) -> (Output_Get:double)
   n13 = Get bIsCrouching()
   > Set NotCrouching(NotCrouching:double="To Float (Boolean)"(InBool=NOT Boolean(A=@n13))) -> (Output_Get:double)
   Set Is Sky Diving(execute=@n12.then_1, Is Sky Diving:double="To Float (Boolean)"(InBool=Get bIsSkydiveFloating())) -> (Output_Get:double)
   n14 = NOT Boolean(A=Get bIsParachuteOpened())
   n15 = OR Boolean(A=Get bIsSkyDiving(), B=Get bIsSkydiveDiveMode())
   Set SkyDiving_Mode(execute=@n12.then_2, SkyDiving_Mode:double="To Float (Boolean)"(InBool=AND Boolean(A=@n15, B=@n14))) -> (Output_Get:double)
   n16 = GetAnimInstance(self=GetOwnerSkeletalMeshComponent())
   Is Valid(exec=@n12.then_3, InputObject=@n16)
   > n17 = Set IsCelebrating_Emote(IsCelebrating_Emote:bool=Montage_IsPlaying(self=@n16, Montage=/BRCosmetics/Animation/Game/MainPlayer/Emotes/Emote_Celebration_Female_M)) -> (Output_Get:bool)
   > n18 = Branch(Condition=@n17)
   > Set IsCelebrating(IsCelebrating:double=1) -> (Output_Get:double)
   Set IsCelebrating(execute=@n18.else, IsCelebrating=0)
   n19 = GetAnimInstance(self=GetOwnerSkeletalMeshComponent())
   Is Valid(exec=@n12.then_5, InputObject=@n19)
   > n20 = Set Iscelebration(Iscelebration:bool=Montage_IsPlaying(self=@n19, Montage=&Emote_Celebration_Female_M)) -> (Output_Get:bool)
   > n21 = Branch(Condition=@n20)
   > Set IsCelebrationEmote(IsCelebrationEmote:double=1) -> (Output_Get:double)
   Set IsCelebrationEmote(execute=@n21.else, IsCelebrationEmote=0)
   n22 = Branch(execute=@n12.then_6, Condition=OR Boolean(A=Get bIsCrouching(), B=Get bIsCrouchSprinting()))
   > Set Crou_Or_CroSpring(Crou_Or_CroSpring:double=1) -> (Output_Get:double)
   Set Crou_Or_CroSpring(execute=@n22.else, Crou_Or_CroSpring=0)
   n23 = GetAnimInstance(self=GetOwnerSkeletalMeshComponent())
   Is Valid(exec=@n12.then_7, InputObject=@n23)
   > Set Idle_FE(Idle_FE:bool=Montage_IsPlaying(self=@n23, Montage=/BRCosmetics/Animation/Game/MainPlayer/Menu/BR/CustomSkins/_s39/JoltMosaicF/JoltMosaic_Idle_Montage2)) -> (Output_Get:bool)
   n24 = PlayerMovement(execute=@n12.then_4, AnimInstance=Self-Reference(), XDotProductCoefficient=1, YDotProductCoefficient=1, ZDotProductCoefficient=1)
   n25 = BreakVector(InVec=MakeVector(X=@n24, Y=@n24.YMovement))
   > Set strafeLeft float(strafeLeft float:double="Absolute (Float)"(A="Clamp (Float)"(Value=@n25, Min=-1, Max=0))) -> (Output_Get:double)
   > Set strafeRight float(strafeRight float:double="Clamp (Float)"(Value=@n25, Min=0, Max=1)) -> (Output_Get:double)
   > Set moveForward(moveForward:double="Clamp (Float)"(Value=@n25.Y, Min=0, Max=1)) -> (Output_Get:double)
   > Set moveBackward(moveBackward:double="Absolute (Float)"(A="Clamp (Float)"(Value=@n25.Y, Min=-1, Max=0))) -> (Output_Get:double)
   n26 = Get bIsFalling()
   n27 = Get bIsTacticalSprinting()
   n28 = Get bIsJumping()
   Set isTacticalJumpingFalling(execute=@n1.then_1, isTacticalJumpingFalling:bool=OR Boolean(A=AND Boolean(A=@n28, B=@n27), B=AND Boolean(A=@n27, B=@n26))) -> (Output_Get:bool)
   n29 = OR Boolean(A=Get bIsSkydiveFloating(), B=Get bIsSkyDiving())
   n30 = Get bIsJumping()
   n31 = Get bIsFalling()
   > n32 = Branch(Condition=AND Boolean(A=OR Boolean(A=@n30, B=@n31), B=NOT Boolean(A=@n29)))
   > Set Is Jumping(Is Jumping:double=1) -> (Output_Get:double)
   Set Is Jumping(execute=@n32.else, Is Jumping=0)
   Event OnParachuteOpened()
   > Switch on EGliderType(Selection=Get PlayerGliderType() -> (PlayerGliderType=HangGlider))
   > Set isUsingHangGlidingInParachute(isUsingHangGlidingInParachute:bool=true) -> (Output_Get:bool)
   Event OnParachuteClosed()
   > Set isUsingHangGlidingInParachute(isUsingHangGlidingInParachute=false)
   Event OnCharacterCustomizationLoaded()
   n33 = CheckForBackbling()
   > Set isWearingBlackbling(isWearingBlackbling:bool=OR Boolean!(A=@n33, B=@n33.bIsBackblingACape, C:bool=@n33.bIsBackblingLong) -> (ReturnValue)) -> (Output_Get:bool)
   > n34 = Branch(Condition=Get bIsInFrontEnd())
   > Set SimSpaceSettings output(SimSpaceSettings output:Anim Phys Sim Space Settings Structure=Get SimSpaceSettings frontEnd() -> (SimSpaceSettings frontEnd:Anim Phys Sim Space Settings Structure)) -> (Output_Get:Anim Phys Sim Space Settings Structure)
   Set SimSpaceSettings output(
    execute=@n34.else
    SimSpaceSettings output=Get SimSpaceSettings default() -> (SimSpaceSettings default:Anim Phys Sim Space Settings Structure)
   )
   Event OnWeaponChanged()
   > Set isAR Equipped(isAR Equipped:bool=false) -> (Output_Get:bool)
   > Set isPistol Equipped(isPistol Equipped:bool=false) -> (Output_Get:bool)
   > Set Is_outLaw_SG(Is_outLaw_SG:bool=false) -> (Output_Get:bool)
   > Set SMG_SG(SMG_SG:bool=false) -> (Output_Get:bool)
   > n35 = Switch on EFortWeaponCoreAnimation!(execute, Selection:EFortWeaponCoreAnimation Enum=Get WeaponCoreAnim!() -> (WeaponCoreAnim)) -> (
    Melee:Exec
    Pistol:Exec
    Shotgun:Exec
    PaperBlueprint:Exec
    Rifle:Exec
    MeleeOneHand:Exec
    MachinePistol:Exec
    RocketLauncher:Exec
    GrenadeLauncher:Exec
    GoingCommando:Exec
    AssaultRifle:Exec
    TacticalShotgun:Exec
    SniperRifle:Exec
    TrapPlacement:Exec
    ShoulderLauncher:Exec
    AbilityDecoTool:Exec
    Crossbow:Exec
    C4:Exec
    RemoteControl:Exec
    DualWield:Exec
    AR_BullPup:Exec
    AR_ForwardGrip:Exec
    MedPackPaddles:Exec
    SMG_P90:Exec
    AR_DrumGun:Exec
    Consumable_Small:Exec
    Consumable_Large:Exec
    Balloon:Exec
    MountedTurret:Exec
    CreativeTool:Exec
    ExplosiveBow:Exec
    AshtonIndigo:Exec
    AshtonChicago:Exec
    MeleeDualWield:Exec
    Unarmed:Exec
   )
   Set isAR Equipped(execute=@n35.AssaultRifle, isAR Equipped=true)
   Set isPistol Equipped(execute=@n35.Pistol, isPistol Equipped=true)
   Set SMG_SG(execute=@n35.SMG_P90, SMG_SG=true)
   Set Is_outLaw_SG(execute=@n35.TacticalShotgun, Is_outLaw_SG=true)
   Reroute Node(InputPin:Wildcard) -> (OutputPin:Wildcard)
  ]}
]
variables [
 alpha hairRigidBody {type=number}
 Crou_Or_CroSpring {type=number}
 Gravity Override Params {
  type=FortGravityOverrideParameterSet
  default=(Skydive_GoingUp=(GravityOverride=(WindFrequency=(X=1.500000,Y=7.000000,Z=8.000000),WindAmplitude=(X=200.000000,Y=-500.000000,Z=400.000000),GravityOverride=(X=0.000000,Y=-300.000000,Z=1500.000000),JointName="root")),Skydive_FloatingDown=(GravityOverride=(WindFrequency=(X=1.500000,Y=7.000000,Z=8.000000),WindAmplitude=(X=200.000000,Y=-500.000000,Z=400.000000),WindOffset=(X=0.000000,Y=-600.000000,Z=400.000000),GravityOverride=(X=0.000000,Y=-700.000000,Z=600.000000),JointName="root")),Skydive_DivingDown=(GravityOverride=(WindFrequency=(X=1.500000,Y=12.000000,Z=10.000000),WindAmplitude=(X=50.000000,Y=300.000000,Z=100.000000),GravityOverride=(X=-1500.000000,Y=0.000000,Z=0.000000),JointName="pelvis")),Skydive_Parachuting=(GravityOverride=(WindFrequency=(X=1.500000,Y=7.000000,Z=8.000000),WindAmplitude=(X=200.000000,Y=10.000000,Z=50.000000),GravityOverride=(X=0.000000,Y=0.000000,Z=-1200.000000),JointName="root")),OnGround_Standing=(GravityOverride=(GravityOverride=(X=0.000000,Y=0.000000,Z=-980.000000),JointName="root")),OnGround_Moving=(GravityOverride=(GravityOverride=(X=0.000000,Y=0.000000,Z=-980.000000),JointName="root")),OnGround_Moving_HighRotationRate=(GravityOverride=(GravityOverride=(X=0.000000,Y=0.000000,Z=-980.000000),JointName="root")),OnGround_Crouch_Moving=(GravityOverride=(GravityOverride=(X=0.000000,Y=0.000000,Z=-980.000000),JointName="root")),Emote_or_Melee=(GravityOverride=(GravityOverride=(X=0.000000,Y=0.000000,Z=-980.000000),JointName="root")),Swimming=(GravityOverride=(GravityOverride=(X=-800.000000,Y=0.000000,Z=0.000000),JointName="pelvis")),Front_End=(GravityOverride=(GravityOverride=(X=0.000000,Y=0.000000,Z=-980.000000),JointName="root")),DBNO=(GravityOverride=(GravityOverride=(X=-980.000000,Y=400.000000,Z=0.000000),JointName="pelvis")),DBNO_Carried=(GravityOverride=(GravityOverride=(X=-980.000000,Y=-500.000000,Z=-500.000000),JointName="pelvis")),OnGround_Falling=(GravityOverride=(GravityOverride=(X=0.000000,Y=0.000000,Z=-980.000000),JointName="root")),Ziplining=(GravityOverride=(GravityOverride=(X=0.000000,Y=0.000000,Z=-980.000000),JointName="root")),Skydive_Surfing=(GravityOverride=(WindFrequency=(X=1.500000,Y=7.000000,Z=8.000000),WindAmplitude=(X=200.000000,Y=-500.000000,Z=400.000000),GravityOverride=(X=500.000000,Y=-300.000000,Z=-980.000000),JointName="root")),Riding_Standing=(GravityOverride=(GravityOverride=(X=0.000000,Y=-350.000000,Z=-980.000000),JointName="root")),Riding_Moving=(GravityOverride=(GravityOverride=(X=0.000000,Y=-450.000000,Z=-450.000000),JointName="root")),Riding_Moving_Targeting=(GravityOverride=(GravityOverride=(X=0.000000,Y=-200.000000,Z=-980.000000),JointName="root")),Tactical_Sprint=(GravityOverride=(GravityOverride=(X=0.000000,Y=-450.000000,Z=-500.000000),JointName="root")),Rail_Grinding=(GravityOverride=(GravityOverride=(X=0.000000,Y=0.000000,Z=-980.000000),JointName="root")),)
 }
 GravityOverride {type=Vector}
 hairBaseOverlap rotator {type=Rotator}
 hairBaseOverlap vector {type=Vector}
 Idle_FE {type=boolean}
 Idle_FrontEnd {type=number}
 Is Jumping {type=number}
 Is Sky Diving {type=number}
 Is_outLaw_SG {type=boolean}
 Is_Parachute_Open {type=number}
 isAR Equipped {type=boolean}
 IsCelebrating {type=number}
 IsCelebrating_Emote {type=boolean}
 Iscelebration {type=boolean}
 IsCelebrationEmote {type=number}
 isPistol Equipped {type=boolean}
 isTacticalJumpingFalling {type=boolean}
 isUsingHangGlidingInParachute {type=boolean}
 isWearingBlackbling {type=boolean}
 moveBackward {type=number}
 moveForward {type=number}
 NewVar {type=boolean}
 NotCrouching {type=number}
 RBAN Settings Out {type=FortRigidBodyAnimNodeParameters}
 Sim Space Settings {
  type=SimSpaceSettings
  default {
   dampingAlpha=1
   maxAngularAcceleration=100
   maxAngularVelocity=3
   maxLinearAcceleration=800
   maxLinearVelocity=400
   velocityScaleZ=1
   worldAlpha=1
   externalLinearDragV=<1 1 1>
  }
 }
 SimSpaceSettings default {
  type=AnimPhysSimSpaceSettings
  default {maxAngularAcceleration=10000, maxAngularVelocity=10000, simSpaceAngularAlpha=0, externalAngularVelocity=<0 0 0>}
 }
 SimSpaceSettings frontEnd {
  type=AnimPhysSimSpaceSettings
  default {maxAngularAcceleration=10000, maxAngularVelocity=10000, simSpaceAngularAlpha=0, externalAngularVelocity=<0 0 0>}
 }
 SimSpaceSettings output {
  type=AnimPhysSimSpaceSettings
  default {maxAngularAcceleration=10000, maxAngularVelocity=10000, simSpaceAngularAlpha=0, externalAngularVelocity=<0 0 0>}
 }
 SkyDiving_Mode {type=number}
 SMG_SG {type=boolean}
 strafeLeft float {type=number}
 strafeRight float {type=number}
]
